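# Control Rig for the MuffinLadle GasChin player body that drives its custom arm, finger, leg and shoulder plate bones from the standard skeleton.
# Front arms copy arm rotations, add SphericalPoseReader-weighted shoulder correctives and solve PBIK to the hands, then drive back arms.
# Fingers are constrained to source digits, with extra thumbs following pinky bones.
# Legs use FABRIK toward the balls; shoulder plates rotate by remapped upper-arm angle.
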ControlRigBlueprint /BRCosmetics/Characters/Player/Male/Medium/Bodies/M_MED_MuffinLadle_GasChin/Meshes/M_MED_MuffinLadle_GasChin_CtrlRig
graphs [
 Add_5_ContainedGraph {nodes [
   n1 = RigVMFunctionEntryNode()
   RigVMFunctionReturnNode(Result=/Script/RigVM.RigVMFunction_MathVectorAdd(A=&RigVMFunction_MathVectorAdd(A=@n1.A, B=@n1.B) {template="Add::Execute(in A,in B,out Result)"}.Result, B=@n1.C) {is=struct RigVMUnitNode, template="Add::Execute(in A,in B,out Result)"}.Result)
  ]}
 Add_5_ContainedGraph {nodes [
   n1 = RigVMFunctionEntryNode()
   RigVMFunctionReturnNode(Result=&RigVMFunction_MathVectorAdd(A=&RigVMFunction_MathVectorAdd(A=@n1.A, B=@n1.B) {template="Add::Execute(in A,in B,out Result)"}.Result, B=@n1.C) {template="Add::Execute(in A,in B,out Result)"}.Result)
  ]}
 Add_5_ContainedGraph {nodes [
   n1 = RigVMFunctionEntryNode()
   RigVMFunctionReturnNode(Result=&RigVMFunction_MathVectorAdd(A=&RigVMFunction_MathVectorAdd(A=@n1.A, B=@n1.B) {template="Add::Execute(in A,in B,out Result)"}.Result, B=@n1.C) {template="Add::Execute(in A,in B,out Result)"}.Result)
  ]}
 RigVMFunctionLibrary {links [], nodes []}
 RigVMFunction_Sequence_2_ContainedGraph {nodes [
   n1 = RigVMFunctionEntryNode()
   RigVMFunction_Sequence = &RigVMFunction_Sequence(ExecuteContext=@n1.ExecuteContext)
   n2 = &RigVMFunction_Sequence(ExecuteContext=@RigVMFunction_Sequence.B)
   n3 = &RigVMFunction_Sequence(ExecuteContext=@n2.B)
   RigVMFunctionReturnNode(A=@RigVMFunction_Sequence.A, B=@n2.A, C=@n3.A, D=@n3.B)
  ]}
 RigVMFunction_Sequence_2_ContainedGraph {nodes [
   n1 = RigVMFunctionEntryNode()
   n2 = &RigVMFunction_Sequence(ExecuteContext=@n1.ExecuteContext)
   n3 = &RigVMFunction_Sequence(ExecuteContext=@n2.B)
   n4 = &RigVMFunction_Sequence(ExecuteContext=@n3.B)
   n5 = &RigVMFunction_Sequence(ExecuteContext=@n4.B)
   n6 = &RigVMFunction_Sequence(ExecuteContext=@n5.B)
   RigVMFunctionReturnNode(A=@n2.A, B=@n3.A, C=@n4.A, D=@n5.A, E=@n6.A, F=@n6.B)
  ]}
 RigVMModel {nodes [
   For_Each_2 = RigVMDispatchNode(ExecuteContext=(), Array=RigVMVariableNode() {variable=OriginalThumbs}.Value) {template="DISPATCH_RigVMDispatch_ArrayIterator(in Array,out Element,out Index,out Count,out Ratio)"}
   > RigVMVariableNode(Value=@For_Each_2.Element.Keys) {variable=OriginalThumbs Index}
   > RigVMVariableNode(Value=RigVMDispatchNode(Index=@For_Each_2.Index, Array=RigVMVariableNode() {variable=CustomOuterThumbs}.Value) {template="DISPATCH_RigVMDispatch_ArrayGetAtIndex(in Array,in Index,out Element)"}.Element.Keys) {variable=CustomOuterThumbs Index}
   n1 = RigVMDispatchNode(Index=0, Array=RigVMVariableNode() {variable=OriginalThumbs Index}.Value) {template="DISPATCH_RigVMDispatch_ArrayGetAtIndex(in Array,in Index,out Element)"}
   n2 = &RigUnit_GetTransform(Space=LocalSpace, bInitial=false, Item=@n1.Element)
   n3 = /Script/RigVM.RigVMFunction_MathQuaternionInverse(Value=&RigUnit_GetTransform(Space=LocalSpace, bInitial=true, Item=@n1.Element).Transform.Rotation) {is=struct RigVMUnitNode, template="Inverse::Execute(in Value,out Result)"}
   > &RigUnit_ModifyTransforms(
    ItemToModify=((Item=(Type=Bone,Name="None"),Transform=(Rotation=(X=0,Y=0,Z=0,W=1),Translation=(X=0,Y=0,Z=0),Scale3D=(X=1,Y=1,Z=1))))
    Weight=1
    WeightMinimum=0
    ItemToModify.0.Transform.Rotation=/Script/RigVM.RigVMFunction_MathQuaternionMul(
     B=(X=-0.0168,Y=-0.0266,Z=0.0257,W=0.9992)
     A=&RigVMFunction_MathQuaternionMul(A=@n2.Transform.Rotation, B=@n3.Result) {template="Multiply::Execute(in A,in B,out Result)"}.Result
    ) {is=struct RigVMUnitNode, template="Multiply::Execute(in A,in B,out Result)"}.Result
    ItemToModify.0.Item=RigVMDispatchNode(Index=0, Array=RigVMVariableNode() {variable=CustomOuterThumbs Index}.Value) {template="DISPATCH_RigVMDispatch_ArrayGetAtIndex(in Array,in Index,out Element)"}.Element
   )
   n4 = RigVMDispatchNode(Index=1, Array=RigVMVariableNode() {variable=OriginalThumbs Index}.Value) {template="DISPATCH_RigVMDispatch_ArrayGetAtIndex(in Array,in Index,out Element)"}
   n5 = &RigUnit_GetTransform(Space=LocalSpace, bInitial=false, Item=@n4.Element)
   n6 = &RigVMFunction_MathQuaternionInverse(Value=&RigUnit_GetTransform(Space=LocalSpace, bInitial=true, Item=@n4.Element).Transform.Rotation) {template="Inverse::Execute(in Value,out Result)"}
   > &RigUnit_ModifyTransforms(
    ItemToModify=((Item=(Type=Bone,Name="None"),Transform=(Rotation=(X=0,Y=0,Z=0,W=1),Translation=(X=0,Y=0,Z=0),Scale3D=(X=1,Y=1,Z=1))))
    Weight=1
    WeightMinimum=0
    ItemToModify.0.Transform.Rotation=&RigVMFunction_MathQuaternionMul(
     B=(X=-0.0168,Y=-0.0266,Z=0.0257,W=0.9992)
     A=&RigVMFunction_MathQuaternionMul(A=@n5.Transform.Rotation, B=@n6.Result) {template="Multiply::Execute(in A,in B,out Result)"}.Result
    ) {template="Multiply::Execute(in A,in B,out Result)"}.Result
    ItemToModify.0.Item=RigVMDispatchNode(Index=1, Array=RigVMVariableNode() {variable=CustomOuterThumbs Index}.Value) {template="DISPATCH_RigVMDispatch_ArrayGetAtIndex(in Array,in Index,out Element)"}.Element
   )
   n7 = RigVMDispatchNode(Index=2, Array=RigVMVariableNode() {variable=OriginalThumbs Index}.Value) {template="DISPATCH_RigVMDispatch_ArrayGetAtIndex(in Array,in Index,out Element)"}
   n8 = &RigUnit_GetTransform(Space=LocalSpace, bInitial=false, Item=@n7.Element)
   n9 = &RigVMFunction_MathQuaternionInverse(Value=&RigUnit_GetTransform(Space=LocalSpace, bInitial=true, Item=@n7.Element).Transform.Rotation) {template="Inverse::Execute(in Value,out Result)"}
   > &RigUnit_ModifyTransforms(
    ItemToModify=((Item=(Type=Bone,Name="None"),Transform=(Rotation=(X=0,Y=0,Z=0,W=1),Translation=(X=0,Y=0,Z=0),Scale3D=(X=1,Y=1,Z=1))))
    Weight=1
    WeightMinimum=0
    ItemToModify.0.Transform.Rotation=&RigVMFunction_MathQuaternionMul(
     B=(X=-0.0168,Y=-0.0266,Z=0.0257,W=0.9992)
     A=&RigVMFunction_MathQuaternionMul(A=@n8.Transform.Rotation, B=@n9.Result) {template="Multiply::Execute(in A,in B,out Result)"}.Result
    ) {template="Multiply::Execute(in A,in B,out Result)"}.Result
    ItemToModify.0.Item=RigVMDispatchNode(Index=2, Array=RigVMVariableNode() {variable=CustomOuterThumbs Index}.Value) {template="DISPATCH_RigVMDispatch_ArrayGetAtIndex(in Array,in Index,out Element)"}.Element
   )
   Get Delta Ref_2 = RigVMFunctionReferenceNode(ExecuteContext=(), Item=(Type=Bone,Name="upperarm_l"))
   RigVMFunction_MathQuaternionFromRotatorV2 = /Script/RigVM.RigVMFunction_MathQuaternionFromRotatorV2(Value=@"Get Delta Ref_2".Result.Rotation) {is=struct RigVMUnitNode, template="Cast::Execute(in Value,out Result)"}
   > RigVMDispatchNode(
    Enabled=True
    ScreenColor=(R=1,G=1,B=1,A=1)
    ScreenDuration=0.05
    Value=&RigVMFunction_MathQuaternionToEuler(RotationOrder=ZYX, Value=@RigVMFunction_MathQuaternionFromRotatorV2.Result).Result.Y
   ) {template="DISPATCH_RigVMDispatch_Print(in Prefix,in Value,in Enabled,in ScreenDuration,in ScreenColor)"}
   n10 = &RigVMFunction_MathQuaternionToEuler(
    RotationOrder=ZYX
    Value=&RigUnit_GetTransform(Item=(Type=Bone,Name="thumb_01_r"), Space=LocalSpace, bInitial=false).Transform.Rotation
   )
   n11 = &RigVMFunction_MathDoubleSub(A=36.9092, B=@n10.Result.Y)
   n12 = &RigVMFunction_MathDoubleSub(A=33.801, B=@n10.Result.Z)
   n13 = &RigVMFunction_MathDoubleSub(A=-103.281, B=@n10.Result.X)
   &RigUnit_ModifyTransforms(
    ExecutePin=()
    ItemToModify=((Item=(Type=Bone,Name="dyn_R_custom_extraThumb_1"),Transform=(Rotation=(X=0,Y=0,Z=0,W=1),Translation=(X=0,Y=0,Z=0),Scale3D=(X=1,Y=1,Z=1))))
    Weight=1
    WeightMinimum=0
    ItemToModify.0.Transform.Rotation=&RigVMFunction_MathQuaternionFromEuler(
     Euler=(X=0,Y=0,Z=0)
     Euler.Z=&RigVMFunction_MathDoubleMul(A=1, B=@n11.Result).Result
     Euler.Y=&RigVMFunction_MathDoubleMul(A=1, B=@n12.Result).Result
     Euler.X=&RigVMFunction_MathDoubleMul(A=-1, B=@n13.Result).Result
    ).Result
   )
   n14 = &RigVMFunction_MathQuaternionToEuler(
    RotationOrder=ZYX
    Value=&RigUnit_GetTransform(Item=(Type=Bone,Name="thumb_02_r"), Space=LocalSpace, bInitial=false).Transform.Rotation
   )
   n15 = &RigVMFunction_MathDoubleSub(A=-0.3226, B=@n14.Result.X)
   n16 = &RigVMFunction_MathDoubleSub(A=-2.4582, B=@n14.Result.Y)
   n17 = &RigVMFunction_MathDoubleSub(A=-34.6932, B=@n14.Result.Z)
   > &RigUnit_ModifyTransforms(
    ItemToModify=((Item=(Type=Bone,Name="dyn_R_custom_extraThumb_2"),Transform=(Rotation=(X=0,Y=0,Z=0,W=1),Translation=(X=0,Y=0,Z=0),Scale3D=(X=1,Y=1,Z=1))))
    Weight=1
    WeightMinimum=0
    ItemToModify.0.Transform.Rotation=&RigVMFunction_MathQuaternionFromEuler(
     Euler=(X=0,Y=0,Z=0)
     Euler.X=&RigVMFunction_MathDoubleMul(A=-1, B=@n15.Result).Result
     Euler.Z=&RigVMFunction_MathDoubleMul(A=1, B=@n16.Result).Result
     Euler.Y=&RigVMFunction_MathDoubleMul(A=1, B=@n17.Result).Result
    ).Result
   )
   n18 = &RigVMFunction_MathQuaternionToEuler(
    RotationOrder=ZYX
    Value=&RigUnit_GetTransform(Item=(Type=Bone,Name="thumb_03_r"), Space=LocalSpace, bInitial=false).Transform.Rotation
   )
   n19 = &RigVMFunction_MathDoubleSub(A=-1.3439, B=@n18.Result.X)
   n20 = &RigVMFunction_MathDoubleSub(A=1.4939, B=@n18.Result.Y)
   n21 = &RigVMFunction_MathDoubleSub(A=-0.7334, B=@n18.Result.Z)
   > n22 = &RigUnit_ModifyTransforms(
    ItemToModify=((Item=(Type=Bone,Name="dyn_R_custom_extraThumb_3"),Transform=(Rotation=(X=0,Y=0,Z=0,W=1),Translation=(X=0,Y=0,Z=0),Scale3D=(X=1,Y=1,Z=1))))
    Weight=1
    WeightMinimum=0
    ItemToModify.0.Transform.Rotation=&RigVMFunction_MathQuaternionFromEuler(
     Euler=(X=0,Y=0,Z=0)
     Euler.X=&RigVMFunction_MathDoubleMul(A=-1, B=@n19.Result).Result
     Euler.Z=&RigVMFunction_MathDoubleMul(A=1, B=@n20.Result).Result
     Euler.Y=&RigVMFunction_MathDoubleMul(A=1, B=@n21.Result).Result
    ).Result
   )
   RigVMRerouteNode(Value=@n22.ExecutePin)
   n23 = &RigVMFunction_MathQuaternionToEuler(
    RotationOrder=ZYX
    Value=&RigUnit_GetTransform(Item=(Type=Bone,Name="thumb_01_l"), Space=LocalSpace, bInitial=false).Transform.Rotation
   )
   n24 = &RigVMFunction_MathDoubleSub(A=-103.281, B=@n23.Result.X)
   n25 = &RigVMFunction_MathDoubleSub(A=36.9092, B=@n23.Result.Y)
   n26 = &RigVMFunction_MathDoubleSub(A=33.801, B=@n23.Result.Z)
   &RigUnit_ModifyTransforms(
    ExecutePin=()
    ItemToModify=((Item=(Type=Bone,Name="dyn_L_custom_extraThumb_1"),Transform=(Rotation=(X=0,Y=0,Z=0,W=1),Translation=(X=0,Y=0,Z=0),Scale3D=(X=1,Y=1,Z=1))))
    Weight=1
    WeightMinimum=0
    ItemToModify.0.Transform.Rotation=&RigVMFunction_MathQuaternionFromEuler(
     Euler=(X=0,Y=0,Z=0)
     Euler.X=&RigVMFunction_MathDoubleMul(A=-1, B=@n24.Result).Result
     Euler.Z=&RigVMFunction_MathDoubleMul(A=1, B=@n25.Result).Result
     Euler.Y=&RigVMFunction_MathDoubleMul(A=1, B=@n26.Result).Result
    ).Result
   )
   n27 = &RigVMFunction_MathQuaternionToEuler(
    RotationOrder=ZYX
    Value=&RigUnit_GetTransform(Item=(Type=Bone,Name="thumb_02_l"), Space=LocalSpace, bInitial=false).Transform.Rotation
   )
   n28 = &RigVMFunction_MathDoubleSub(B=0, A=@n27.Result.X)
   n29 = &RigVMFunction_MathDoubleSub(A=-2.4582, B=@n27.Result.Y)
   n30 = &RigVMFunction_MathDoubleSub(A=-34.6932, B=@n27.Result.Z)
   > &RigUnit_ModifyTransforms(
    ItemToModify=((Item=(Type=Bone,Name="dyn_L_custom_extraThumb_2"),Transform=(Rotation=(X=0,Y=0,Z=0,W=1),Translation=(X=0,Y=0,Z=0),Scale3D=(X=1,Y=1,Z=1))))
    Weight=1
    WeightMinimum=0
    ItemToModify.0.Transform.Rotation=&RigVMFunction_MathQuaternionFromEuler(
     Euler=(X=0,Y=0,Z=0)
     Euler.X=&RigVMFunction_MathDoubleMul(A=-1, B=@n28.Result).Result
     Euler.Z=&RigVMFunction_MathDoubleMul(A=1, B=@n29.Result).Result
     Euler.Y=&RigVMFunction_MathDoubleMul(A=1, B=@n30.Result).Result
    ).Result
   )
   n31 = &RigVMFunction_MathQuaternionToEuler(
    RotationOrder=ZYX
    Value=&RigUnit_GetTransform(Item=(Type=Bone,Name="thumb_03_l"), Space=LocalSpace, bInitial=false).Transform.Rotation
   )
   n32 = &RigVMFunction_MathDoubleSub(A=-1.3439, B=@n31.Result.X)
   n33 = &RigVMFunction_MathDoubleSub(A=1.4939, B=@n31.Result.Y)
   n34 = &RigVMFunction_MathDoubleSub(A=-0.7334, B=@n31.Result.Z)
   > &RigUnit_ModifyTransforms(
    ItemToModify=((Item=(Type=Bone,Name="dyn_L_custom_extraThumb_3"),Transform=(Rotation=(X=0,Y=0,Z=0,W=1),Translation=(X=0,Y=0,Z=0),Scale3D=(X=1,Y=1,Z=1))))
    Weight=1
    WeightMinimum=0
    ItemToModify.0.Transform.Rotation=&RigVMFunction_MathQuaternionFromEuler(
     Euler=(X=0,Y=0,Z=0)
     Euler.X=&RigVMFunction_MathDoubleMul(A=-1, B=@n32.Result).Result
     Euler.Z=&RigVMFunction_MathDoubleMul(A=1, B=@n33.Result).Result
     Euler.Y=&RigVMFunction_MathDoubleMul(A=1, B=@n34.Result).Result
    ).Result
   )
   RigUnit_BeginExecution = &RigUnit_BeginExecution()
   RigVMFunction_Sequence = RigVMAggregateNode(ExecuteContext=@RigUnit_BeginExecution.ExecutePin)
   &RigUnit_RotationConstraintLocalSpaceOffset(
    AdvancedSettings=(InterpolationType=Average,RotationOrderForFilter=XZY)
    Child=(Type=Bone,Name="dyn_L_customFrontArm_clavicle")
    Filter=(bX=True,bY=True,bZ=True)
    Parents=((Item=(Type=Bone,Name="clavicle_l"),Weight=1))
    Weight=0.75
    ExecutePin=@RigVMFunction_Sequence.A
   )
   > &RigUnit_RotationConstraintLocalSpaceOffset(
    AdvancedSettings=(InterpolationType=Average,RotationOrderForFilter=XZY)
    Child=(Type=Bone,Name="dyn_L_customFrontArm_upperarm")
    Filter=(bX=true,bY=true,bZ=true)
    Parents=((Item=(Type=Bone,Name="upperarm_l"),Weight=1))
    Weight=1
   )
   > &RigUnit_RotationConstraintLocalSpaceOffset(
    AdvancedSettings=(InterpolationType=Average,RotationOrderForFilter=XZY)
    Child=(Type=Bone,Name="dyn_L_customFrontArm_lowerarm")
    Filter=(bX=True,bY=True,bZ=True)
    Parents=((Item=(Type=Bone,Name="lowerarm_l"),Weight=1))
    Weight=1
   )
   > n35 = &RigUnit_SphericalPoseReader(
    ActiveRegionScaleFactors=(PositiveWidth=0.5,NegativeWidth=0.5,PositiveHeight=0.5,NegativeHeight=1)
    ActiveRegionSize=0.1
    Debug=(bDrawDebug=false,bDraw2D=False,bDrawLocalAxes=False,DebugScale=25,DebugSegments=20,DebugThickness=0.25)
    DriverAxis=(X=1,Y=0,Z=0)
    DriverItem=(Type=Bone,Name="upperarm_l")
    FalloffRegionScaleFactors=(PositiveWidth=1,NegativeWidth=1,PositiveHeight=1,NegativeHeight=1)
    FalloffSize=0.3
    FlipHeightScaling=False
    FlipWidthScaling=False
    OptionalParentItem=(Type=Bone,Name="None")
    RotationOffset=(X=75,Y=0,Z=-30)
   )
   > &RigUnit_ModifyTransforms(
    ItemToModify=((Item=(Type=Bone,Name="dyn_L_customFrontArm_clavicle"),Transform=(Rotation=(X=-0.0871,Y=0.0435,Z=0.0038,W=0.9952),Translation=(X=1,Y=-2,Z=-1),Scale3D=(X=1,Y=1,Z=1))))
    WeightMinimum=0
    Weight=@n35.OutputParam
   )
   > &RigUnit_ModifyTransforms(
    ItemToModify=((Item=(Type=Bone,Name="dyn_L_customFrontArm_upperarm"),Transform=(Rotation=(X=0,Y=0,Z=0,W=1),Translation=(X=0,Y=0,Z=0),Scale3D=(X=1,Y=1,Z=1))))
    WeightMinimum=0
    Weight=@n35.OutputParam
   )
   > n36 = &RigUnit_SphericalPoseReader(
    ActiveRegionScaleFactors=(PositiveWidth=0.5,NegativeWidth=0.5,PositiveHeight=0.5,NegativeHeight=1)
    ActiveRegionSize=0.1
    Debug=(bDrawDebug=false,bDraw2D=False,bDrawLocalAxes=False,DebugScale=25,DebugSegments=20,DebugThickness=0.25)
    DriverAxis=(X=1,Y=0,Z=0)
    DriverItem=(Type=Bone,Name="upperarm_l")
    FalloffRegionScaleFactors=(PositiveWidth=1,NegativeWidth=1,PositiveHeight=1,NegativeHeight=1)
    FalloffSize=0.3
    FlipHeightScaling=False
    FlipWidthScaling=False
    OptionalParentItem=(Type=Bone,Name="None")
    RotationOffset=(X=0,Y=-50,Z=0)
   )
   > &RigUnit_ModifyTransforms(
    ItemToModify=((Item=(Type=Bone,Name="dyn_L_customFrontArm_clavicle"),Transform=(Rotation=(X=-0.1305,Y=0,Z=0,W=0.9914),Translation=(X=3,Y=-2,Z=-2),Scale3D=(X=1,Y=1,Z=1))))
    WeightMinimum=0
    Weight=@n36.OutputParam
   )
   > &RigUnit_ModifyTransforms(
    ItemToModify=((Item=(Type=Bone,Name="dyn_L_customFrontArm_upperarm"),Transform=(Rotation=(X=0,Y=0,Z=0,W=1),Translation=(X=3,Y=0,Z=0),Scale3D=(X=1,Y=1,Z=1))))
    WeightMinimum=0
    Weight=@n36.OutputParam
   )
   RigUnit_GetTransform = &RigUnit_GetTransform(Item=(Type=Bone,Name="hand_l"))
   > /Script/PBIK.RigUnit_PBIK(
    BoneSettings=()
    Debug=(DrawScale=1)
    Effectors=((Bone="dyn_L_customFrontArm_wrist",Transform=(Rotation=(X=0,Y=0,Z=0,W=1),Translation=(X=0,Y=0,Z=0),Scale3D=(X=1,Y=1,Z=1)),PositionAlpha=1,RotationAlpha=1,StrengthAlpha=1,PullChainAlpha=0,PinRotation=1,ChainDepth=0))
    ExcludedBones=()
    Root=dyn_L_customFrontArm_clavicle
    Settings="(Iterations=20,SubIterations=0,MassMultiplier=1,bAllowStretch=false,RootBehavior=PinToInput,PrePullRootSettings=(RotationAlpha=0,RotationAlphaX=1,RotationAlphaY=1,RotationAlphaZ=1,PositionAlpha=1,PositionAlphaX=1,PositionAlphaY=1,PositionAlphaZ=1),GlobalPu ...(truncated)"
    Effectors.0.Transform=@RigUnit_GetTransform.Transform
   ) {is=struct RigVMUnitNode}
   > &RigUnit_ParentConstraint(
    AdvancedSettings=(InterpolationType=Average,RotationOrderForFilter=XZY)
    Child=(Type=Bone,Name="dyn_L_customFrontArm_wrist")
    Filter=(TranslationFilter=(bX=true,bY=true,bZ=true),RotationFilter=(bX=True,bY=True,bZ=True),ScaleFilter=(bX=True,bY=True,bZ=True))
    Parents=((Item=(Type=Bone,Name="hand_l"),Weight=1))
    Weight=1
   )
   > &RigUnit_ParentConstraint(
    AdvancedSettings=(InterpolationType=Average,RotationOrderForFilter=XZY)
    Child=(Type=Bone,Name="dyn_L_customBackArm_clavicle")
    Parents=((Item=(Type=Bone,Name="dyn_L_customFrontArm_clavicle"),Weight=1))
    Weight=1
   )
   > &RigUnit_ParentConstraint(
    AdvancedSettings=(InterpolationType=Average,RotationOrderForFilter=XZY)
    Child=(Type=Bone,Name="dyn_L_customBackArm_upperarm")
    Parents=((Item=(Type=Bone,Name="dyn_L_customFrontArm_upperarm"),Weight=1))
    Weight=1
   )
   > &RigUnit_ParentConstraint(
    AdvancedSettings=(InterpolationType=Average,RotationOrderForFilter=XZY)
    Child=(Type=Bone,Name="dyn_L_customBackArm_lowerarm")
    Parents=((Item=(Type=Bone,Name="dyn_L_customFrontArm_lowerarm"),Weight=1))
    Weight=1
   )
   > &RigUnit_ParentConstraint(
    AdvancedSettings=(InterpolationType=Average,RotationOrderForFilter=XZY)
    Child=(Type=Bone,Name="dyn_L_customBackArm_wrist")
    Parents=((Item=(Type=Bone,Name="dyn_L_customFrontArm_wrist"),Weight=1))
    Weight=1
   )
   > &RigUnit_ParentConstraint(
    AdvancedSettings=(InterpolationType=Average,RotationOrderForFilter=XZY)
    Child=(Type=Bone,Name="dyn_L_custom_thumb_1")
    Filter=(TranslationFilter=(bX=false,bY=false,bZ=false),RotationFilter=(bX=True,bY=True,bZ=True),ScaleFilter=(bX=True,bY=True,bZ=True))
    Parents=((Item=(Type=Bone,Name="thumb_01_l"),Weight=1))
    Weight=1
   )
   > &RigUnit_ParentConstraint(
    AdvancedSettings=(InterpolationType=Average,RotationOrderForFilter=XZY)
    Child=(Type=Bone,Name="dyn_L_custom_thumb_2")
    Filter=(TranslationFilter=(bX=false,bY=false,bZ=false),RotationFilter=(bX=True,bY=True,bZ=True),ScaleFilter=(bX=True,bY=True,bZ=True))
    Parents=((Item=(Type=Bone,Name="thumb_02_l"),Weight=1))
    Weight=1
   )
   > &RigUnit_ParentConstraint(
    AdvancedSettings=(InterpolationType=Average,RotationOrderForFilter=XZY)
    Child=(Type=Bone,Name="dyn_L_custom_thumb_3")
    Filter=(TranslationFilter=(bX=false,bY=false,bZ=false),RotationFilter=(bX=True,bY=True,bZ=True),ScaleFilter=(bX=True,bY=True,bZ=True))
    Parents=((Item=(Type=Bone,Name="thumb_03_l"),Weight=1))
    Weight=1
   )
   > &RigUnit_ParentConstraint(
    AdvancedSettings=(InterpolationType=Average,RotationOrderForFilter=XZY)
    Child=(Type=Bone,Name="dyn_L_custom_index_1")
    Filter=(TranslationFilter=(bX=false,bY=false,bZ=false),RotationFilter=(bX=True,bY=True,bZ=True),ScaleFilter=(bX=True,bY=True,bZ=True))
    Parents=((Item=(Type=Bone,Name="index_01_l"),Weight=1),(Item=(Type=Bone,Name="middle_01_l"),Weight=0))
    Weight=1
   )
   > &RigUnit_ParentConstraint(
    AdvancedSettings=(InterpolationType=Average,RotationOrderForFilter=XZY)
    Child=(Type=Bone,Name="dyn_L_custom_index_2")
    Filter=(TranslationFilter=(bX=false,bY=false,bZ=false),RotationFilter=(bX=True,bY=True,bZ=True),ScaleFilter=(bX=True,bY=True,bZ=True))
    Parents=((Item=(Type=Bone,Name="index_02_l"),Weight=1),(Item=(Type=Bone,Name="middle_02_l"),Weight=0))
    Weight=1
   )
   > &RigUnit_ParentConstraint(
    AdvancedSettings=(InterpolationType=Average,RotationOrderForFilter=XZY)
    Child=(Type=Bone,Name="dyn_L_custom_index_3")
    Filter=(TranslationFilter=(bX=false,bY=false,bZ=false),RotationFilter=(bX=True,bY=True,bZ=True),ScaleFilter=(bX=True,bY=True,bZ=True))
    Parents=((Item=(Type=Bone,Name="index_03_l"),Weight=1),(Item=(Type=Bone,Name="middle_03_l"),Weight=0))
    Weight=1
   )
   > &RigUnit_ParentConstraint(
    AdvancedSettings=(InterpolationType=Average,RotationOrderForFilter=XZY)
    Child=(Type=Bone,Name="dyn_L_custom_ring_1")
    Filter=(TranslationFilter=(bX=false,bY=false,bZ=false),RotationFilter=(bX=True,bY=True,bZ=True),ScaleFilter=(bX=True,bY=True,bZ=True))
    Parents=((Item=(Type=Bone,Name="ring_01_l"),Weight=1),(Item=(Type=Bone,Name="middle_01_l"),Weight=0))
    Weight=1
   )
   > &RigUnit_ParentConstraint(
    AdvancedSettings=(InterpolationType=Average,RotationOrderForFilter=XZY)
    Child=(Type=Bone,Name="dyn_L_custom_ring_2")
    Filter=(TranslationFilter=(bX=false,bY=false,bZ=false),RotationFilter=(bX=True,bY=True,bZ=True),ScaleFilter=(bX=True,bY=True,bZ=True))
    Parents=((Item=(Type=Bone,Name="ring_02_l"),Weight=1),(Item=(Type=Bone,Name="middle_02_l"),Weight=0))
    Weight=1
   )
   > &RigUnit_ParentConstraint(
    AdvancedSettings=(InterpolationType=Average,RotationOrderForFilter=XZY)
    Child=(Type=Bone,Name="dyn_L_custom_ring_3")
    Filter=(TranslationFilter=(bX=false,bY=false,bZ=false),RotationFilter=(bX=True,bY=True,bZ=True),ScaleFilter=(bX=True,bY=True,bZ=True))
    Parents=((Item=(Type=Bone,Name="ring_03_l"),Weight=1),(Item=(Type=Bone,Name="middle_03_l"),Weight=0))
    Weight=1
   )
   > &RigUnit_ParentConstraint(
    AdvancedSettings=(InterpolationType=Average,RotationOrderForFilter=XZY)
    Child=(Type=Bone,Name="dyn_L_custom_extraThumb_1")
    Filter=(TranslationFilter=(bX=false,bY=false,bZ=false),RotationFilter=(bX=True,bY=True,bZ=True),ScaleFilter=(bX=True,bY=True,bZ=True))
    Parents=((Item=(Type=Bone,Name="pinky_metacarpal_l"),Weight=1))
    Weight=1
    bMaintainOffset=false
   )
   > &RigUnit_ParentConstraint(
    AdvancedSettings=(InterpolationType=Average,RotationOrderForFilter=XZY)
    Child=(Type=Bone,Name="dyn_L_custom_extraThumb_2")
    Filter=(TranslationFilter=(bX=false,bY=false,bZ=false),RotationFilter=(bX=True,bY=True,bZ=True),ScaleFilter=(bX=True,bY=True,bZ=True))
    Parents=((Item=(Type=Bone,Name="pinky_01_l"),Weight=1))
    Weight=1
    bMaintainOffset=false
   )
   > &RigUnit_ParentConstraint(
    AdvancedSettings=(InterpolationType=Average,RotationOrderForFilter=XZY)
    Child=(Type=Bone,Name="dyn_L_custom_extraThumb_3")
    Filter=(TranslationFilter=(bX=false,bY=false,bZ=false),RotationFilter=(bX=True,bY=True,bZ=True),ScaleFilter=(bX=True,bY=True,bZ=True))
    Parents=((Item=(Type=Bone,Name="pinky_02_l"),Weight=1))
    Weight=1
    bMaintainOffset=false
   )
   > &RigUnit_ModifyTransforms(
    ItemToModify="((Item=(Type=Bone,Name=\"dyn_L_custom_extraThumb_1\"),Transform=(Rotation=(X=0,Y=-0.1305,Z=0,W=0.9914),Translation=(X=0,Y=-0.2,Z=0),Scale3D=(X=1,Y=1,Z=1))),(Item=(Type=Bone,Name=\"dyn_L_custom_extraThumb_2\"),Transform=(Rotation=(X=0,Y=0,Z=0.2588,W=0.9659),Tra ...(truncated)"
    Weight=1
    WeightMinimum=0
   )
   > &RigUnit_RotationConstraintLocalSpaceOffset(
    AdvancedSettings=(InterpolationType=Average,RotationOrderForFilter=XZY)
    Child=(Type=Bone,Name="dyn_R_customFrontArm_clavicle")
    Filter=(bX=True,bY=True,bZ=True)
    Parents=((Item=(Type=Bone,Name="clavicle_r"),Weight=1))
    Weight=1
   )
   > &RigUnit_RotationConstraintLocalSpaceOffset(
    AdvancedSettings=(InterpolationType=Average,RotationOrderForFilter=XZY)
    Child=(Type=Bone,Name="dyn_R_customFrontArm_upperarm")
    Filter=(bX=True,bY=True,bZ=True)
    Parents=((Item=(Type=Bone,Name="upperarm_r"),Weight=1))
    Weight=1
   )
   > &RigUnit_RotationConstraintLocalSpaceOffset(
    AdvancedSettings=(InterpolationType=Average,RotationOrderForFilter=XZY)
    Child=(Type=Bone,Name="dyn_R_customFrontArm_lowerarm")
    Filter=(bX=True,bY=True,bZ=True)
    Parents=((Item=(Type=Bone,Name="lowerarm_r"),Weight=1))
    Weight=1
   )
   > n37 = &RigUnit_SphericalPoseReader(
    ActiveRegionScaleFactors=(PositiveWidth=1,NegativeWidth=0.75,PositiveHeight=0.75,NegativeHeight=1)
    ActiveRegionSize=0.1
    Debug=(bDrawDebug=false,bDraw2D=False,bDrawLocalAxes=False,DebugScale=25,DebugSegments=20,DebugThickness=0.25)
    DriverAxis=(X=-1,Y=0,Z=0)
    DriverItem=(Type=Bone,Name="upperarm_r")
    FalloffRegionScaleFactors=(PositiveWidth=1,NegativeWidth=1,PositiveHeight=1,NegativeHeight=1)
    FalloffSize=0.3
    FlipHeightScaling=False
    FlipWidthScaling=False
    OptionalParentItem=(Type=Bone,Name="None")
    RotationOffset=(X=-100,Y=0,Z=20)
   )
   > &RigUnit_ModifyTransforms(
    ItemToModify=((Item=(Type=Bone,Name="dyn_R_customFrontArm_clavicle"),Transform=(Rotation=(X=0,Y=0,Z=0,W=1),Translation=(X=-2,Y=-1,Z=1),Scale3D=(X=1,Y=1,Z=1))))
    WeightMinimum=0
    Weight=@n37.OutputParam
   )
   > &RigUnit_ModifyTransforms(
    ItemToModify=((Item=(Type=Bone,Name="dyn_R_customFrontArm_upperarm"),Transform=(Rotation=(X=-0.1305,Y=0,Z=0,W=0.9914),Translation=(X=2,Y=0,Z=-1.5),Scale3D=(X=1,Y=1,Z=1))))
    WeightMinimum=0
    Weight=@n37.OutputParam
   )
   > n38 = &RigUnit_SphericalPoseReader(
    ActiveRegionScaleFactors=(PositiveWidth=0.5,NegativeWidth=0.5,PositiveHeight=0.5,NegativeHeight=0.5)
    ActiveRegionSize=0.1
    Debug=(bDrawDebug=false,bDraw2D=False,bDrawLocalAxes=False,DebugScale=25,DebugSegments=20,DebugThickness=0.25)
    DriverAxis=(X=-1,Y=0,Z=0)
    DriverItem=(Type=Bone,Name="upperarm_r")
    FalloffRegionScaleFactors=(PositiveWidth=1,NegativeWidth=1,PositiveHeight=1,NegativeHeight=1)
    FalloffSize=0.3
    FlipHeightScaling=False
    FlipWidthScaling=False
    OptionalParentItem=(Type=Bone,Name="None")
    RotationOffset=(X=-180,Y=-45,Z=0)
   )
   > &RigUnit_ModifyTransforms(
    ItemToModify=((Item=(Type=Bone,Name="dyn_R_customFrontArm_clavicle"),Transform=(Rotation=(X=-0.0872,Y=0,Z=0,W=0.9962),Translation=(X=-1.5,Y=1.5,Z=2),Scale3D=(X=1,Y=1,Z=1))))
    WeightMinimum=0
    Weight=@n38.OutputParam
   )
   > &RigUnit_ModifyTransforms(
    ItemToModify=((Item=(Type=Bone,Name="dyn_R_customFrontArm_upperarm"),Transform=(Rotation=(X=0.1305,Y=0,Z=0,W=0.9914),Translation=(X=-3,Y=0,Z=0),Scale3D=(X=1,Y=1,Z=1))))
    WeightMinimum=0
    Weight=@n38.OutputParam
   )
   > n39 = &RigUnit_SphericalPoseReader(
    ActiveRegionScaleFactors=(PositiveWidth=0.5,NegativeWidth=0.5,PositiveHeight=0.5,NegativeHeight=0.5)
    ActiveRegionSize=0.1
    Debug=(bDrawDebug=false,bDraw2D=False,bDrawLocalAxes=False,DebugScale=25,DebugSegments=20,DebugThickness=0.25)
    DriverAxis=(X=-1,Y=0,Z=0)
    DriverItem=(Type=Bone,Name="upperarm_r")
    FalloffRegionScaleFactors=(PositiveWidth=1,NegativeWidth=1,PositiveHeight=1,NegativeHeight=1)
    FalloffSize=0.2
    FlipHeightScaling=False
    FlipWidthScaling=False
    OptionalParentItem=(Type=Bone,Name="None")
    RotationOffset=(X=-45,Y=30,Z=0)
   )
   > &RigUnit_ModifyTransforms(
    ItemToModify=((Item=(Type=Bone,Name="dyn_R_customFrontArm_clavicle"),Transform=(Rotation=(X=0,Y=0,Z=0,W=1),Translation=(X=0,Y=-2,Z=0),Scale3D=(X=1,Y=1,Z=1))))
    WeightMinimum=0
    Weight=@n39.OutputParam
   )
   > &RigUnit_PBIK(
    BoneSettings=()
    Debug=(DrawScale=1)
    Effectors=((Bone="dyn_R_customFrontArm_wrist",Transform=(Rotation=(X=0,Y=0,Z=0,W=1),Translation=(X=0,Y=0,Z=0),Scale3D=(X=1,Y=1,Z=1)),PositionAlpha=1,RotationAlpha=1,StrengthAlpha=1,PullChainAlpha=1,PinRotation=1,ChainDepth=0))
    ExcludedBones=()
    Root=dyn_R_customFrontArm_clavicle
    Settings="(Iterations=20,SubIterations=0,MassMultiplier=1,bAllowStretch=False,RootBehavior=PinToInput,PrePullRootSettings=(RotationAlpha=0,RotationAlphaX=1,RotationAlphaY=1,RotationAlphaZ=1,PositionAlpha=1,PositionAlphaX=1,PositionAlphaY=1,PositionAlphaZ=1),GlobalPu ...(truncated)"
    Effectors.0.Transform=&RigUnit_GetTransform(Item=(Type=Bone,Name="hand_r")).Transform
   )
   > &RigUnit_ParentConstraint(
    AdvancedSettings=(InterpolationType=Average,RotationOrderForFilter=XZY)
    Child=(Type=Bone,Name="dyn_R_customFrontArm_wrist")
    Filter=(TranslationFilter=(bX=true,bY=true,bZ=true),RotationFilter=(bX=True,bY=True,bZ=True),ScaleFilter=(bX=True,bY=True,bZ=True))
    Parents=((Item=(Type=Bone,Name="hand_r"),Weight=1))
    Weight=1
   )
   > &RigUnit_ParentConstraint(
    AdvancedSettings=(InterpolationType=Average,RotationOrderForFilter=XZY)
    Child=(Type=Bone,Name="dyn_R_customBackArm_clavicle")
    Parents=((Item=(Type=Bone,Name="dyn_R_customFrontArm_clavicle"),Weight=1))
    Weight=1
   )
   > &RigUnit_ParentConstraint(
    AdvancedSettings=(InterpolationType=Average,RotationOrderForFilter=XZY)
    Child=(Type=Bone,Name="dyn_R_customBackArm_upperarm")
    Parents=((Item=(Type=Bone,Name="dyn_R_customFrontArm_upperarm"),Weight=1))
    Weight=1
   )
   > &RigUnit_ParentConstraint(
    AdvancedSettings=(InterpolationType=Average,RotationOrderForFilter=XZY)
    Child=(Type=Bone,Name="dyn_R_customBackArm_lowerarm")
    Parents=((Item=(Type=Bone,Name="dyn_R_customFrontArm_lowerarm"),Weight=1))
    Weight=1
   )
   > &RigUnit_ParentConstraint(
    AdvancedSettings=(InterpolationType=Average,RotationOrderForFilter=XZY)
    Child=(Type=Bone,Name="dyn_R_customBackArm_wrist")
    Parents=((Item=(Type=Bone,Name="dyn_R_customFrontArm_wrist"),Weight=1))
    Weight=1
   )
   > &RigUnit_ParentConstraint(
    AdvancedSettings=(InterpolationType=Average,RotationOrderForFilter=XZY)
    Child=(Type=Bone,Name="dyn_R_custom_thumb_1")
    Filter=(TranslationFilter=(bX=false,bY=false,bZ=false),RotationFilter=(bX=True,bY=True,bZ=True),ScaleFilter=(bX=True,bY=True,bZ=True))
    Parents=((Item=(Type=Bone,Name="thumb_01_r"),Weight=1))
    Weight=1
   )
   > &RigUnit_ParentConstraint(
    AdvancedSettings=(InterpolationType=Average,RotationOrderForFilter=XZY)
    Child=(Type=Bone,Name="dyn_R_custom_thumb_2")
    Filter=(TranslationFilter=(bX=false,bY=false,bZ=false),RotationFilter=(bX=True,bY=True,bZ=True),ScaleFilter=(bX=True,bY=True,bZ=True))
    Parents=((Item=(Type=Bone,Name="thumb_02_r"),Weight=1))
    Weight=1
   )
   > &RigUnit_ParentConstraint(
    AdvancedSettings=(InterpolationType=Average,RotationOrderForFilter=XZY)
    Child=(Type=Bone,Name="dyn_R_custom_thumb_3")
    Filter=(TranslationFilter=(bX=false,bY=false,bZ=false),RotationFilter=(bX=True,bY=True,bZ=True),ScaleFilter=(bX=True,bY=True,bZ=True))
    Parents=((Item=(Type=Bone,Name="thumb_03_r"),Weight=1))
    Weight=1
   )
   > &RigUnit_ParentConstraint(
    AdvancedSettings=(InterpolationType=Average,RotationOrderForFilter=XZY)
    Child=(Type=Bone,Name="dyn_R_custom_index_1")
    Filter=(TranslationFilter=(bX=false,bY=false,bZ=false),RotationFilter=(bX=True,bY=True,bZ=True),ScaleFilter=(bX=True,bY=True,bZ=True))
    Parents=((Item=(Type=Bone,Name="index_01_r"),Weight=1),(Item=(Type=Bone,Name="middle_01_r"),Weight=0))
    Weight=1
   )
   > &RigUnit_ParentConstraint(
    AdvancedSettings=(InterpolationType=Average,RotationOrderForFilter=XZY)
    Child=(Type=Bone,Name="dyn_R_custom_index_2")
    Filter=(TranslationFilter=(bX=false,bY=false,bZ=false),RotationFilter=(bX=True,bY=True,bZ=True),ScaleFilter=(bX=True,bY=True,bZ=True))
    Parents=((Item=(Type=Bone,Name="index_02_r"),Weight=1),(Item=(Type=Bone,Name="middle_02_r"),Weight=0))
    Weight=1
   )
   > &RigUnit_ParentConstraint(
    AdvancedSettings=(InterpolationType=Average,RotationOrderForFilter=XZY)
    Child=(Type=Bone,Name="dyn_R_custom_index_3")
    Filter=(TranslationFilter=(bX=false,bY=false,bZ=false),RotationFilter=(bX=True,bY=True,bZ=True),ScaleFilter=(bX=True,bY=True,bZ=True))
    Parents=((Item=(Type=Bone,Name="index_03_r"),Weight=1),(Item=(Type=Bone,Name="middle_03_r"),Weight=0))
    Weight=1
   )
   > &RigUnit_ParentConstraint(
    AdvancedSettings=(InterpolationType=Average,RotationOrderForFilter=XZY)
    Child=(Type=Bone,Name="dyn_R_custom_ring_1")
    Filter=(TranslationFilter=(bX=false,bY=false,bZ=false),RotationFilter=(bX=True,bY=True,bZ=True),ScaleFilter=(bX=True,bY=True,bZ=True))
    Parents=((Item=(Type=Bone,Name="ring_01_r"),Weight=1),(Item=(Type=Bone,Name="middle_01_r"),Weight=0))
    Weight=1
   )
   > &RigUnit_ParentConstraint(
    AdvancedSettings=(InterpolationType=Average,RotationOrderForFilter=XZY)
    Child=(Type=Bone,Name="dyn_R_custom_ring_2")
    Filter=(TranslationFilter=(bX=false,bY=false,bZ=false),RotationFilter=(bX=True,bY=True,bZ=True),ScaleFilter=(bX=True,bY=True,bZ=True))
    Parents=((Item=(Type=Bone,Name="ring_02_r"),Weight=1),(Item=(Type=Bone,Name="middle_02_r"),Weight=0))
    Weight=1
   )
   > &RigUnit_ParentConstraint(
    AdvancedSettings=(InterpolationType=Average,RotationOrderForFilter=XZY)
    Child=(Type=Bone,Name="dyn_R_custom_ring_3")
    Filter=(TranslationFilter=(bX=false,bY=false,bZ=false),RotationFilter=(bX=True,bY=True,bZ=True),ScaleFilter=(bX=True,bY=True,bZ=True))
    Parents=((Item=(Type=Bone,Name="ring_03_r"),Weight=1),(Item=(Type=Bone,Name="middle_03_r"),Weight=0))
    Weight=1
   )
   > &RigUnit_ParentConstraint(
    AdvancedSettings=(InterpolationType=Average,RotationOrderForFilter=XZY)
    Child=(Type=Bone,Name="dyn_R_custom_extraThumb_1")
    Filter=(TranslationFilter=(bX=false,bY=false,bZ=false),RotationFilter=(bX=True,bY=True,bZ=True),ScaleFilter=(bX=True,bY=True,bZ=True))
    Parents=((Item=(Type=Bone,Name="pinky_metacarpal_r"),Weight=1))
    Weight=1
    bMaintainOffset=false
   )
   > &RigUnit_ParentConstraint(
    AdvancedSettings=(InterpolationType=Average,RotationOrderForFilter=XZY)
    Child=(Type=Bone,Name="dyn_R_custom_extraThumb_2")
    Filter=(TranslationFilter=(bX=false,bY=false,bZ=false),RotationFilter=(bX=True,bY=True,bZ=True),ScaleFilter=(bX=True,bY=True,bZ=True))
    Parents=((Item=(Type=Bone,Name="pinky_01_r"),Weight=1))
    Weight=1
    bMaintainOffset=false
   )
   > &RigUnit_ParentConstraint(
    AdvancedSettings=(InterpolationType=Average,RotationOrderForFilter=XZY)
    Child=(Type=Bone,Name="dyn_R_custom_extraThumb_3")
    Filter=(TranslationFilter=(bX=false,bY=false,bZ=false),RotationFilter=(bX=True,bY=True,bZ=True),ScaleFilter=(bX=True,bY=True,bZ=True))
    Parents=((Item=(Type=Bone,Name="pinky_02_r"),Weight=1))
    Weight=1
    bMaintainOffset=false
   )
   > &RigUnit_ModifyTransforms(
    ItemToModify="((Item=(Type=Bone,Name=\"dyn_R_custom_extraThumb_1\"),Transform=(Rotation=(X=0,Y=-0.1305,Z=0,W=0.9914),Translation=(X=0,Y=0.2,Z=0),Scale3D=(X=1,Y=1,Z=1))),(Item=(Type=Bone,Name=\"dyn_R_custom_extraThumb_2\"),Transform=(Rotation=(X=0,Y=0,Z=0.2588,W=0.9659),Tran ...(truncated)"
    Weight=1
    WeightMinimum=0
   )
   n40 = &RigUnit_GetTransform(Item=(Type=Bone,Name="ball_l"), bInitial=false)
   > /Script/ControlRig.RigUnit_FABRIKItemArray(
    Items=((Type=Bone,Name="dyn_L_custom_calf"),(Type=Bone,Name="dyn_L_custom_ball"))
    MaxIterations=10
    Precision=1
    Weight=1
    bPropagateToChildren=True
    bSetEffectorTransform=true
    EffectorTransform=@n40.Transform
   ) {is=struct RigVMUnitNode}
   > &RigUnit_ModifyTransforms(
    ItemToModify="((Item=(Type=Bone,Name=\"thigh_l\"),Transform=(Rotation=(X=0,Y=0,Z=0,W=1),Translation=(X=0,Y=0,Z=0),Scale3D=(X=1,Y=1,Z=1))),(Item=(Type=Bone,Name=\"dyn_L_custom_calf\"),Transform=(Rotation=(X=0,Y=0,Z=0,W=1),Translation=(X=0,Y=0,Z=0),Scale3D=(X=1,Y=1,Z=1))),(It ...(truncated)"
    Weight=1
    WeightMinimum=0
   )
   > &RigUnit_ParentConstraint(
    AdvancedSettings=(InterpolationType=Average,RotationOrderForFilter=XZY)
    Child=(Type=Bone,Name="thigh_l_IK")
    Filter=(TranslationFilter=(bX=True,bY=True,bZ=True),RotationFilter=(bX=true,bY=true,bZ=true),ScaleFilter=(bX=True,bY=True,bZ=True))
    Parents=((Item=(Type=Bone,Name="thigh_l"),Weight=1))
    Weight=1
   )
   > &RigUnit_ParentConstraint(
    AdvancedSettings=(InterpolationType=Average,RotationOrderForFilter=XZY)
    Child=(Type=Bone,Name="dyn_L_custom_foot")
    Filter=(TranslationFilter=(bX=true,bY=true,bZ=true),RotationFilter=(bX=true,bY=true,bZ=true),ScaleFilter=(bX=True,bY=True,bZ=True))
    Parents=((Item=(Type=Bone,Name="foot_l"),Weight=1))
    Weight=1
    bMaintainOffset=true
   )
   > &RigUnit_ParentConstraint(
    AdvancedSettings=(InterpolationType=Average,RotationOrderForFilter=XZY)
    Child=(Type=Bone,Name="dyn_L_custom_ball")
    Filter=(TranslationFilter=(bX=true,bY=false,bZ=false),RotationFilter=(bX=true,bY=true,bZ=true),ScaleFilter=(bX=True,bY=True,bZ=True))
    Parents=((Item=(Type=Bone,Name="ball_l"),Weight=1))
    Weight=1
    bMaintainOffset=true
   )
   n41 = &RigUnit_GetTransform(Item=(Type=Bone,Name="ball_r"), bInitial=false)
   > &RigUnit_FABRIKItemArray(
    Items=((Type=Bone,Name="dyn_R_custom_calf"),(Type=Bone,Name="dyn_R_custom_ball"))
    MaxIterations=10
    Precision=1
    Weight=1
    bPropagateToChildren=True
    bSetEffectorTransform=true
    EffectorTransform=@n41.Transform
   )
   > &RigUnit_ModifyTransforms(
    ItemToModify="((Item=(Type=Bone,Name=\"thigh_r\"),Transform=(Rotation=(X=0,Y=0,Z=0,W=1),Translation=(X=0,Y=0,Z=0),Scale3D=(X=1,Y=1,Z=1))),(Item=(Type=Bone,Name=\"dyn_R_custom_calf\"),Transform=(Rotation=(X=0,Y=0,Z=0,W=1),Translation=(X=0,Y=0,Z=0),Scale3D=(X=1,Y=1,Z=1))),(It ...(truncated)"
    Weight=1
    WeightMinimum=0
   )
   > &RigUnit_ParentConstraint(
    AdvancedSettings=(InterpolationType=Average,RotationOrderForFilter=XZY)
    Child=(Type=Bone,Name="thigh_r_IK")
    Filter=(TranslationFilter=(bX=True,bY=True,bZ=True),RotationFilter=(bX=true,bY=true,bZ=true),ScaleFilter=(bX=True,bY=True,bZ=True))
    Parents=((Item=(Type=Bone,Name="thigh_r"),Weight=1))
    Weight=1
   )
   > &RigUnit_ParentConstraint(
    AdvancedSettings=(InterpolationType=Average,RotationOrderForFilter=XZY)
    Child=(Type=Bone,Name="dyn_R_custom_foot")
    Filter=(TranslationFilter=(bX=true,bY=True,bZ=True),RotationFilter=(bX=true,bY=true,bZ=true),ScaleFilter=(bX=True,bY=True,bZ=True))
    Parents=((Item=(Type=Bone,Name="foot_r"),Weight=1))
    Weight=1
   )
   > &RigUnit_ParentConstraint(
    AdvancedSettings=(InterpolationType=Average,RotationOrderForFilter=XZY)
    Child=(Type=Bone,Name="dyn_R_custom_ball")
    Filter=(TranslationFilter=(bX=true,bY=false,bZ=false),RotationFilter=(bX=true,bY=true,bZ=true),ScaleFilter=(bX=True,bY=True,bZ=True))
    Parents=((Item=(Type=Bone,Name="ball_r"),Weight=1))
    Weight=1
    bMaintainOffset=true
   )
   n42 = &RigUnit_GetRelativeTransformForItem(Child=(Type=Bone,Name="upperarm_l"), Parent=(Type=Bone,Name="clavicle_l"), bChildInitial=False, bParentInitial=False)
   > &RigUnit_ModifyTransforms(
    ItemToModify=((Item=(Type=Bone,Name="dyn_L_shoulderPlate_main"),Transform=(Rotation=(X=-0.4988,Y=-0.0349,Z=0.0604,W=0.8639),Translation=(X=0,Y=0,Z=0),Scale3D=(X=1,Y=1,Z=1))))
    WeightMinimum=0
    Weight=&RigVMFunction_MathDoubleRemap(
     SourceMaximum=60
     SourceMinimum=40
     TargetMaximum=1
     TargetMinimum=0
     Value=&RigVMFunction_MathQuaternionToEuler(RotationOrder=ZXY, Value=@n42.RelativeTransform.Rotation).Result.X
    ).Result
   )
   n43 = &RigUnit_GetRelativeTransformForItem(Child=(Type=Bone,Name="upperarm_r"), Parent=(Type=Bone,Name="clavicle_r"), bChildInitial=False, bParentInitial=False)
   > &RigUnit_ModifyTransforms(
    ItemToModify=((Item=(Type=Bone,Name="dyn_R_shoulderPlate_main"),Transform=(Rotation=(X=-0.4988,Y=-0.0349,Z=0.0604,W=0.8639),Translation=(X=0,Y=0,Z=0),Scale3D=(X=1,Y=1,Z=1))))
    WeightMinimum=0
    Weight=&RigVMFunction_MathDoubleRemap(
     SourceMaximum=60
     SourceMinimum=40
     TargetMaximum=1
     TargetMinimum=0
     Value=&RigVMFunction_MathQuaternionToEuler(RotationOrder=ZXY, Value=@n43.RelativeTransform.Rotation).Result.X
    ).Result
   )
   n44 = &RigUnit_GetRelativeTransformForItem(Child=(Type=Bone,Name="lowerarm_l"), Parent=(Type=Bone,Name="upperarm_l"), bChildInitial=False, bParentInitial=False)
   > &RigUnit_ModifyTransforms(
    ItemToModify="((Item=(Type=Bone,Name=\"dyn_L_customBackArm_wireAdjuster\"),Transform=(Rotation=(X=0,Y=0,Z=0,W=1),Translation=(X=-4,Y=0,Z=0),Scale3D=(X=1.2,Y=1,Z=1))),(Item=(Type=Bone,Name=\"dyn_L_customFrontArm_wireAdjuster\"),Transform=(Rotation=(X=0,Y=0,Z=0,W=1),Translati ...(truncated)"
    WeightMinimum=0
    Weight=&RigVMFunction_MathDoubleRemap(
     SourceMaximum=85
     SourceMinimum=50
     TargetMaximum=1
     TargetMinimum=0
     Value=&RigVMFunction_MathQuaternionToEuler(RotationOrder=XYZ, Value=@n44.RelativeTransform.Rotation).Result.Z
    ).Result
   )
   n45 = &RigUnit_GetRelativeTransformForItem(Child=(Type=Bone,Name="lowerarm_r"), Parent=(Type=Bone,Name="upperarm_r"), bChildInitial=False, bParentInitial=False)
   > &RigUnit_ModifyTransforms(
    ItemToModify="((Item=(Type=Bone,Name=\"dyn_R_customBackArm_wireAdjuster\"),Transform=(Rotation=(X=0,Y=0,Z=0,W=1),Translation=(X=4,Y=0,Z=0),Scale3D=(X=1.2,Y=1,Z=1))),(Item=(Type=Bone,Name=\"dyn_R_customFrontArm_wireAdjuster\"),Transform=(Rotation=(X=0,Y=0,Z=0,W=1),Translatio ...(truncated)"
    WeightMinimum=0
    Weight=&RigVMFunction_MathDoubleRemap(
     SourceMaximum=85
     SourceMinimum=50
     TargetMaximum=1
     TargetMinimum=0
     Value=&RigVMFunction_MathQuaternionToEuler(RotationOrder=XYZ, Value=@n45.RelativeTransform.Rotation).Result.Z
    ).Result
   )
   > &RigUnit_RotationConstraintLocalSpaceOffset(
    AdvancedSettings=(InterpolationType=Average,RotationOrderForFilter=XZY)
    Child=(Type=Bone,Name="dyn_L_customFrontArm_clavicle_adjuster")
    Filter=(bX=True,bY=True,bZ=True)
    Parents=((Item=(Type=Bone,Name="dyn_L_customFrontArm_clavicle"),Weight=0.5),(Item=(Type=Bone,Name="dyn_L_customFrontArm_upperarm"),Weight=0.5))
    Weight=1
   )
   > &RigUnit_RotationConstraintLocalSpaceOffset(
    AdvancedSettings=(InterpolationType=Average,RotationOrderForFilter=XZY)
    Child=(Type=Bone,Name="dyn_L_customBackArm_clavicle_adjuster")
    Filter=(bX=True,bY=True,bZ=True)
    Parents=((Item=(Type=Bone,Name="dyn_L_customBackArm_clavicle"),Weight=0.5),(Item=(Type=Bone,Name="dyn_L_customBackArm_upperarm"),Weight=0.5))
    Weight=1
   )
   > &RigUnit_RotationConstraintLocalSpaceOffset(
    AdvancedSettings=(InterpolationType=Average,RotationOrderForFilter=XZY)
    Child=(Type=Bone,Name="dyn_R_customFrontArm_clavicle_adjuster")
    Filter=(bX=True,bY=True,bZ=True)
    Parents=((Item=(Type=Bone,Name="dyn_R_customFrontArm_clavicle"),Weight=0.5),(Item=(Type=Bone,Name="dyn_R_customFrontArm_upperarm"),Weight=0.5))
    Weight=1
   )
   > &RigUnit_RotationConstraintLocalSpaceOffset(
    AdvancedSettings=(InterpolationType=Average,RotationOrderForFilter=XZY)
    Child=(Type=Bone,Name="dyn_R_customBackArm_clavicle_adjuster")
    Filter=(bX=True,bY=True,bZ=True)
    Parents=((Item=(Type=Bone,Name="dyn_R_customBackArm_clavicle"),Weight=0.5),(Item=(Type=Bone,Name="dyn_R_customBackArm_upperarm"),Weight=0.5))
    Weight=1
   )
   n46 = &RigUnit_GetRelativeTransformForItem(Child=(Type=Bone,Name="upperarm_l"), Parent=(Type=Bone,Name="clavicle_l"), bChildInitial=False, bParentInitial=False)
   > &RigUnit_ModifyTransforms(
    ItemToModify=((Item=(Type=Bone,Name="dyn_L_customBackArm_clavicle_adjuster"),Transform=(Rotation=(X=-0.4226,Y=0,Z=0,W=0.9063),Translation=(X=0,Y=-2,Z=0),Scale3D=(X=1,Y=1,Z=1))))
    WeightMinimum=0
    Weight=&RigVMFunction_MathDoubleRemap(
     SourceMaximum=90
     SourceMinimum=50
     TargetMaximum=1
     TargetMinimum=0
     Value=&RigVMFunction_MathQuaternionToEuler(RotationOrder=ZXY, Value=@n46.RelativeTransform.Rotation).Result.Z
    ).Result
   )
   n47 = &RigUnit_GetRelativeTransformForItem(Child=(Type=Bone,Name="upperarm_r"), Parent=(Type=Bone,Name="clavicle_r"), bChildInitial=False, bParentInitial=False)
   > &RigUnit_ModifyTransforms(
    ItemToModify=((Item=(Type=Bone,Name="dyn_R_customBackArm_clavicle_adjuster"),Transform=(Rotation=(X=-0.4226,Y=0,Z=0,W=0.9063),Translation=(X=0,Y=2,Z=0),Scale3D=(X=1,Y=1,Z=1))))
    WeightMinimum=0
    Weight=&RigVMFunction_MathDoubleRemap(
     SourceMaximum=90
     SourceMinimum=50
     TargetMaximum=1
     TargetMinimum=0
     Value=&RigVMFunction_MathQuaternionToEuler(RotationOrder=ZXY, Value=@n47.RelativeTransform.Rotation).Result.Z
    ).Result
   )
   n48 = &RigVMFunction_MathQuaternionToEuler(
    RotationOrder=ZYX
    Value=&RigUnit_GetTransform(Item=(Type=Bone,Name="head"), Space=LocalSpace, bInitial=False).Transform.Rotation
   )
   RigUnit_SetRotation = &RigUnit_SetRotation(
    Item=(Type=Bone,Name="head")
    Weight=1
    bInitial=False
    bPropagateToChildren=true
    Value=&RigVMFunction_MathQuaternionFromEuler(
     Euler=(X=0,Y=0,Z=0)
     Euler.Z=&RigVMFunction_MathDoubleClamp(Maximum=2.5, Minimum=-45, Value=@n48.Result.Z).Result
     Euler.X=&RigVMFunction_MathDoubleMul(B=0.8, A=@n48.Result.X).Result
     Euler.Y=&RigVMFunction_MathDoubleMul(B=0.8, A=@n48.Result.Y).Result
    ).Result
    ExecutePin=@RigVMFunction_Sequence.D
   )
   n49 = &RigVMFunction_MathQuaternionToEuler(
    RotationOrder=ZYX
    Value=&RigUnit_GetTransform(Item=(Type=Bone,Name="neck_02"), Space=LocalSpace, bInitial=False).Transform.Rotation
   )
   > &RigUnit_SetRotation(
    Item=(Type=Bone,Name="neck_02")
    Weight=1
    bInitial=False
    Value=&RigVMFunction_MathQuaternionFromEuler(
     Euler=(X=0,Y=0,Z=0)
     Euler.Z=&RigVMFunction_MathDoubleClamp(Maximum=45, Minimum=-10, Value=@n49.Result.Z).Result
     Euler.X=&RigVMFunction_MathDoubleMul(B=0.8, A=@n49.Result.X).Result
     Euler.Y=&RigVMFunction_MathDoubleMul(B=0.8, A=@n49.Result.Y).Result
    ).Result
   )
   n50 = &RigVMFunction_MathQuaternionToEuler(
    RotationOrder=ZYX
    Value=&RigUnit_GetTransform(Item=(Type=Bone,Name="neck_01"), Space=LocalSpace, bInitial=False).Transform.Rotation
   )
   > &RigUnit_SetRotation(
    Item=(Type=Bone,Name="neck_01")
    Weight=1
    bInitial=False
    Value=&RigVMFunction_MathQuaternionFromEuler(
     Euler=(X=0,Y=0,Z=0)
     Euler.Z=&RigVMFunction_MathDoubleClamp(Maximum=2.5, Minimum=-30, Value=@n50.Result.Z).Result
     Euler.X=&RigVMFunction_MathDoubleMul(B=0.6, A=@n50.Result.X).Result
     Euler.Y=&RigVMFunction_MathDoubleMul(B=0.8, A=@n50.Result.Y).Result
    ).Result
   )
   RigVMFunction_Sequence_1 = RigVMAggregateNode(ExecuteContext=@RigVMFunction_Sequence.B)
   &RigUnit_ParentConstraint(
    AdvancedSettings=(RotationOrderForFilter=XZY)
    Child=(Type=Bone,Name="deform_belt_2")
    Parents="((Item=(Type=Bone,Name=\"spine_05\"),Weight=0.0131),(Item=(Type=Bone,Name=\"pelvis\"),Weight=0.0289),(Item=(Type=Bone,Name=\"spine_01\"),Weight=0.0903),(Item=(Type=Bone,Name=\"spine_04\"),Weight=0.1698),(Item=(Type=Bone,Name=\"spine_02\"),Weight=0.3321),(Item=(Type= ...(truncated)"
    Weight=1
    ExecutePin=@RigVMFunction_Sequence_1.A
   )
   > &RigUnit_ParentConstraint(
    AdvancedSettings=(RotationOrderForFilter=XZY)
    Child=(Type=Bone,Name="deform_belt_1")
    Parents=((Item=(Type=Bone,Name="spine_02"),Weight=0.0549),(Item=(Type=Bone,Name="spine_03"),Weight=0.1879),(Item=(Type=Bone,Name="spine_05"),Weight=0.2069),(Item=(Type=Bone,Name="spine_04"),Weight=0.5504))
    Weight=1
   )
   > &RigUnit_ParentConstraint(
    AdvancedSettings=(RotationOrderForFilter=XZY)
    Child=(Type=Bone,Name="deform_belt_f")
    Parents=((Item=(Type=Bone,Name="spine_01"),Weight=1.0))
    Weight=1
   )
   > &RigUnit_ParentConstraint(
    AdvancedSettings=(RotationOrderForFilter=XZY)
    Child=(Type=Bone,Name="deform_belt_b")
    Parents=((Item=(Type=Bone,Name="pelvis"),Weight=1.0))
    Weight=1
   )
   > &RigUnit_ParentConstraint(
    AdvancedSettings=(RotationOrderForFilter=XZY)
    Child=(Type=Bone,Name="deform_grenades")
    Parents=((Item=(Type=Bone,Name="spine_01"),Weight=1.0))
    Weight=1
   )
   > &RigUnit_ParentConstraint(
    AdvancedSettings=(RotationOrderForFilter=XZY)
    Child=(Type=Bone,Name="dyn_belt_1_f")
    Parents=((Item=(Type=Bone,Name="spine_02"),Weight=0.0001),(Item=(Type=Bone,Name="spine_01"),Weight=0.0008),(Item=(Type=Bone,Name="pelvis"),Weight=0.0018),(Item=(Type=Bone,Name="dyn_belt_f"),Weight=0.0665),(Item=(Type=Bone,Name="dyn_belt_1_f"),Weight=0.9308))
    Weight=1
   )
   > &RigUnit_ParentConstraint(
    AdvancedSettings=(RotationOrderForFilter=XZY)
    Child=(Type=Bone,Name="dyn_belt_1_b")
    Parents="((Item=(Type=Bone,Name=\"dyn_R_customBackArm_clavicle\"),Weight=0.001),(Item=(Type=Bone,Name=\"spine_04\"),Weight=0.001),(Item=(Type=Bone,Name=\"spine_01\"),Weight=0.0038),(Item=(Type=Bone,Name=\"thigh_l_IK\"),Weight=0.0272),(Item=(Type=Bone,Name=\"thigh_twist_01_r ...(truncated)"
    Weight=1
   )
   > &RigUnit_ParentConstraint(
    AdvancedSettings=(RotationOrderForFilter=XZY)
    Child=(Type=Bone,Name="dyn_belt_2_f")
    Parents=((Item=(Type=Bone,Name="dyn_belt_b"),Weight=0.0028),(Item=(Type=Bone,Name="dyn_belt_1_b"),Weight=0.1494),(Item=(Type=Bone,Name="dyn_belt_1_f"),Weight=0.2535),(Item=(Type=Bone,Name="dyn_belt_2_b"),Weight=0.5943))
    Weight=1
   )
   > &RigUnit_ParentConstraint(
    AdvancedSettings=(RotationOrderForFilter=XZY)
    Child=(Type=Bone,Name="dyn_belt_2_b")
    Parents=((Item=(Type=Bone,Name="dyn_belt_b"),Weight=0.0028),(Item=(Type=Bone,Name="dyn_belt_1_b"),Weight=0.1494),(Item=(Type=Bone,Name="dyn_belt_1_f"),Weight=0.2535),(Item=(Type=Bone,Name="dyn_belt_2_b"),Weight=0.5943))
    Weight=1
   )
   Get Delta Ref = RigVMFunctionReferenceNode(Item=(Type=Bone,Name="dyn_L_custom_calf"), ExecuteContext=@RigVMFunction_Sequence_1.B)
   > Get Delta Ref_1 = RigVMFunctionReferenceNode(Item=(Type=Bone,Name="dyn_R_custom_calf"))
   > &RigUnit_ParentConstraint(
    AdvancedSettings=(RotationOrderForFilter=XZY)
    Child=(Type=Bone,Name="deform_knee_in_base_l")
    Parents=((Item=(Type=Bone,Name="dyn_L_custom_calf"),Weight=0.5),(Item=(Type=Bone,Name="thigh_l_IK"),Weight=0.5))
    Weight=1
   )
   > &RigUnit_ParentConstraint(
    AdvancedSettings=(RotationOrderForFilter=XZY)
    Child=(Type=Bone,Name="deform_knee_in_base_r")
    Parents=((Item=(Type=Bone,Name="dyn_R_custom_calf"),Weight=0.5),(Item=(Type=Bone,Name="thigh_r_IK"),Weight=0.5))
    Weight=1
   )
   > &RigUnit_ParentConstraint(
    AdvancedSettings=(RotationOrderForFilter=XZY)
    Child=(Type=Bone,Name="deform_knee_out_base_l")
    Parents=((Item=(Type=Bone,Name="dyn_L_custom_calf"),Weight=0.1),(Item=(Type=Bone,Name="thigh_l_IK"),Weight=0.9))
    Weight=1
   )
   > &RigUnit_ParentConstraint(
    AdvancedSettings=(RotationOrderForFilter=XZY)
    Child=(Type=Bone,Name="deform_knee_out_base_r")
    Parents=((Item=(Type=Bone,Name="dyn_R_custom_calf"),Weight=0.1),(Item=(Type=Bone,Name="thigh_r_IK"),Weight=0.9))
    Weight=1
   )
   n51 = &RigVMFunction_MathFloatRemap(SourceMaximum=100, SourceMinimum=10, bClamp=true, Value=@"Get Delta Ref".Result.Rotation.Yaw)
   > &RigUnit_ModifyTransforms(
    ItemToModify=((Item=(Type=Bone,Name="deform_knee_in_l"),Transform=(Rotation=(X=0,Y=0,Z=0,W=1),Translation=(X=0,Y=-12,Z=0),Scale3D=(X=0.5,Y=1,Z=1))))
    Weight=@n51.Result
   )
   > &RigUnit_ModifyTransforms(
    ItemToModify=((Item=(Type=Bone,Name="deform_knee_out_l"),Transform=(Rotation=(X=0,Y=0,Z=0,W=1),Translation=(X=0,Y=3,Z=0),Scale3D=(X=1.5,Y=1,Z=1.5))))
    Weight=@n51.Result
   )
   n52 = &RigVMFunction_MathFloatRemap(SourceMaximum=100, SourceMinimum=10, bClamp=true, Value=@"Get Delta Ref_1".Result.Rotation.Yaw)
   > &RigUnit_ModifyTransforms(
    ItemToModify=((Item=(Type=Bone,Name="deform_knee_in_r"),Transform=(Rotation=(X=0,Y=0,Z=0,W=1),Translation=(X=0,Y=12,Z=0),Scale3D=(X=0.5,Y=1,Z=1))))
    Weight=@n52.Result
   )
   > &RigUnit_ModifyTransforms(
    ItemToModify=((Item=(Type=Bone,Name="deform_knee_out_r"),Transform=(Rotation=(X=0,Y=0,Z=0,W=1),Translation=(X=0,Y=-3,Z=0),Scale3D=(X=1.5,Y=1,Z=1.5))))
    Weight=@n52.Result
   )
   n53 = &RigUnit_SphericalPoseReader(
    ActiveRegionScaleFactors=(PositiveWidth=3.5,NegativeWidth=0.5,PositiveHeight=3,NegativeHeight=3)
    ActiveRegionSize=0.1
    Debug=(bDrawDebug=false,DebugScale=25,DebugSegments=20,DebugThickness=0.25)
    DriverAxis=(X=-1,Y=0,Z=0)
    DriverItem=(Type=Bone,Name="thigh_l")
    FalloffRegionScaleFactors=(PositiveWidth=1,NegativeWidth=1,PositiveHeight=1,NegativeHeight=1)
    FalloffSize=0.45
    OptionalParentItem=(Type=Bone)
    RotationOffset=(X=90,Y=0,Z=0)
    ExecutePin=@RigVMFunction_Sequence_1.C
   )
   > n54 = &RigUnit_SphericalPoseReader(
    ActiveRegionScaleFactors=(PositiveWidth=3.5,NegativeWidth=0.5,PositiveHeight=3,NegativeHeight=3)
    ActiveRegionSize=0.1
    Debug=(bDrawDebug=false,DebugScale=25,DebugSegments=20,DebugThickness=0.25)
    DriverAxis=(X=1,Y=0,Z=0)
    DriverItem=(Type=Bone,Name="thigh_r")
    FalloffRegionScaleFactors=(PositiveWidth=1,NegativeWidth=1,PositiveHeight=1,NegativeHeight=1)
    FalloffSize=0.45
    FlipHeightScaling=false
    FlipWidthScaling=true
    OptionalParentItem=(Type=Bone)
    RotationOffset=(X=-90,Y=0,Z=0)
   )
   > &RigUnit_ParentConstraint(
    AdvancedSettings=(RotationOrderForFilter=XZY)
    Child=(Type=Bone,Name="deform_groin_base_l")
    Parents=((Item=(Type=Bone,Name="thigh_l_IK"),Weight=0.7),(Item=(Type=Bone,Name="pelvis"),Weight=0.3))
    Weight=1
    bMaintainOffset=true
   )
   > &RigUnit_ParentConstraint(
    AdvancedSettings=(RotationOrderForFilter=XZY)
    Child=(Type=Bone,Name="deform_groin_base_r")
    Parents=((Item=(Type=Bone,Name="thigh_r_IK"),Weight=0.7),(Item=(Type=Bone,Name="pelvis"),Weight=0.3))
    Weight=1
   )
   > &RigUnit_ParentConstraint(
    AdvancedSettings=(RotationOrderForFilter=XZY)
    Child=(Type=Bone,Name="deform_glute_base_l")
    Parents=((Item=(Type=Bone,Name="thigh_l_IK"),Weight=0.25),(Item=(Type=Bone,Name="pelvis"),Weight=0.75))
    Weight=1
   )
   > &RigUnit_ParentConstraint(
    AdvancedSettings=(RotationOrderForFilter=XZY)
    Child=(Type=Bone,Name="deform_glute_base_r")
    Parents=((Item=(Type=Bone,Name="thigh_r_IK"),Weight=0.25),(Item=(Type=Bone,Name="pelvis"),Weight=0.75))
    Weight=1
   )
   > &RigUnit_ModifyTransforms(
    ItemToModify=((Item=(Type=Bone,Name="deform_groin_l"),Transform=(Rotation=(X=0,Y=0,Z=0,W=1),Translation=(X=-3,Y=12,Z=0),Scale3D=(X=0.5,Y=1,Z=1))))
    Weight=@n53.OutputParam
   )
   > &RigUnit_ModifyTransforms(
    ItemToModify=((Item=(Type=Bone,Name="deform_glute_l"),Transform=(Rotation=(X=0,Y=0,Z=0,W=1),Translation=(X=-7,Y=0,Z=0),Scale3D=(X=1,Y=1.5,Z=1))))
    Weight=@n53.OutputParam
   )
   > &RigUnit_ModifyTransforms(
    ItemToModify=((Item=(Type=Bone,Name="deform_groin_r"),Transform=(Rotation=(X=0,Y=0,Z=0,W=1),Translation=(X=3,Y=-12,Z=0),Scale3D=(X=0.5,Y=1,Z=1))))
    Weight=@n54.OutputParam
   )
   > &RigUnit_ModifyTransforms(
    ItemToModify=((Item=(Type=Bone,Name="deform_glute_r"),Transform=(Rotation=(X=0,Y=0,Z=0,W=1),Translation=(X=7,Y=0,Z=0),Scale3D=(X=1,Y=1.5,Z=1))))
    Weight=@n54.OutputParam
   )
   Add = &RigVMFunction_MathDoubleAdd(A=@n53.OutputParam, B=@n54.OutputParam)
   Remap = &RigVMFunction_MathDoubleRemap(SourceMaximum=2, SourceMinimum=0, TargetMaximum=1, TargetMinimum=0, Value=@Add.Result)
   > &RigUnit_ModifyTransforms(
    ItemToModify=((Item=(Type=Bone,Name="deform_skirt_b_l"),Transform=(Rotation=(X=0,Y=0,Z=-0.0872,W=0.9962),Translation=(X=0,Y=-1,Z=0),Scale3D=(X=1,Y=1,Z=1))))
    WeightMinimum=0
    Weight=@Remap.Result
   )
   > &RigUnit_ModifyTransforms(
    ItemToModify=((Item=(Type=Bone,Name="deform_skirt_b_r"),Transform=(Rotation=(X=0,Y=0,Z=-0.0872,W=0.9962),Translation=(X=0,Y=-1,Z=0),Scale3D=(X=1,Y=1,Z=1))))
    WeightMinimum=0
    Weight=@Remap.Result
   )
   RigVMFunction_ControlFlowBranch = &RigVMFunction_ControlFlowBranch(ExecuteContext=@RigVMFunction_Sequence_1.F, Condition=RigVMVariableNode() {variable=IsRigidBodyOn}.Value)
   &RigUnit_AimConstraintLocalSpaceOffset(
    Child=(Type=Bone,Name="deform_skirt_l")
    Filter=(bX=True,bY=True,bZ=True)
    Parents=((Item=(Type=Bone,Name="dyn_L_custom_calf"),Weight=1))
    UpAxis=(X=0,Y=-1,Z=0)
    Weight=1
    WorldUp=(Target=(X=0,Y=0,Z=1),Kind=Location,Space=(Type=Bone,Name="deform_skirt_l"))
    bMaintainOffset=true
    ExecutePin=@RigVMFunction_ControlFlowBranch.False
   )
   > &RigUnit_AimConstraintLocalSpaceOffset(
    Child=(Type=Bone,Name="deform_skirt_r")
    Filter=(bX=True,bY=True,bZ=True)
    Parents=((Item=(Type=Bone,Name="dyn_R_custom_calf"),Weight=1))
    UpAxis=(X=0,Y=-1,Z=0)
    Weight=1
    WorldUp=(Target=(X=0,Y=0,Z=1),Kind=Location,Space=(Type=Bone,Name="deform_skirt_r"))
    bMaintainOffset=true
   )
   > &RigUnit_RotationConstraintLocalSpaceOffset(
    AdvancedSettings=(InterpolationType=Average,RotationOrderForFilter=XZY)
    Child=(Type=Bone,Name="deform_lace_l")
    Filter=(bX=True,bY=True,bZ=True)
    Parents=((Item=(Type=Bone,Name="deform_skirt_l"),Weight=1))
    Weight=1
    bMaintainOffset=true
   )
   > &RigUnit_RotationConstraintLocalSpaceOffset(
    AdvancedSettings=(InterpolationType=Average,RotationOrderForFilter=XZY)
    Child=(Type=Bone,Name="deform_lace_r")
    Filter=(bX=True,bY=True,bZ=True)
    Parents=((Item=(Type=Bone,Name="deform_skirt_r"),Weight=1))
    Weight=1
    bMaintainOffset=true
   )
   > &RigUnit_ModifyTransforms(
    ItemToModify="((Item=(Type=Bone,Name=\"deform_skirt_l\"),Transform=(Rotation=(X=0,Y=0,Z=-0.0436,W=0.999),Translation=(X=0,Y=0,Z=0),Scale3D=(X=1,Y=1,Z=1))),(Item=(Type=Bone,Name=\"deform_skirt_r\"),Transform=(Rotation=(X=0,Y=0,Z=-0.0436,W=0.999),Translation=(X=0,Y=0,Z=0),Sca ...(truncated)"
    Weight=1
    WeightMinimum=0
   )
   > &RigUnit_AimConstraintLocalSpaceOffset(
    Child=(Type=Bone,Name="deform_belt_f")
    Filter=(bX=True,bY=True,bZ=True)
    Parents=((Item=(Type=Bone,Name="dyn_R_custom_calf"),Weight=1))
    UpAxis=(X=0,Y=-1,Z=0)
    Weight=1
    WorldUp=(Target=(X=0,Y=0,Z=1),Kind=Location,Space=(Type=Bone,Name="deform_belt_f"))
    bMaintainOffset=true
   )
   > &RigUnit_AimConstraintLocalSpaceOffset(
    Child=(Type=Bone,Name="deform_belt_b")
    Filter=(bX=True,bY=True,bZ=True)
    Parents=((Item=(Type=Bone,Name="deform_glute_r"),Weight=1))
    UpAxis=(X=0,Y=-1,Z=0)
    Weight=1
    WorldUp=(Target=(X=0,Y=0,Z=1),Kind=Location,Space=(Type=Bone,Name="deform_belt_b"))
    bMaintainOffset=true
   )
   > &RigUnit_ModifyTransforms(
    ItemToModify="((Item=(Type=Bone,Name=\"deform_belt_f\"),Transform=(Rotation=(X=0,Y=0,Z=0,W=1),Translation=(X=0,Y=0,Z=0),Scale3D=(X=1,Y=1,Z=1))),(Item=(Type=Bone,Name=\"deform_belt_b\"),Transform=(Rotation=(X=0,Y=0,Z=-0.0436,W=0.999),Translation=(X=0,Y=0,Z=0),Scale3D=(X=1,Y= ...(truncated)"
    Weight=1
    WeightMinimum=0
   )
   > &RigUnit_AimConstraintLocalSpaceOffset(
    Child=(Type=Bone,Name="deform_skirt_b_l")
    Filter=(bX=True,bY=True,bZ=True)
    Parents=((Item=(Type=Bone,Name="deform_glute_l"),Weight=1))
    UpAxis=(X=0,Y=-1,Z=0)
    Weight=1
    WorldUp=(Target=(X=0,Y=0,Z=1),Kind=Location,Space=(Type=Bone,Name="deform_skirt_b_l"))
    bMaintainOffset=true
   )
   > &RigUnit_AimConstraintLocalSpaceOffset(
    Child=(Type=Bone,Name="deform_skirt_b_r")
    Filter=(bX=True,bY=True,bZ=True)
    Parents=((Item=(Type=Bone,Name="deform_glute_r"),Weight=1))
    UpAxis=(X=0,Y=-1,Z=0)
    Weight=1
    WorldUp=(Target=(X=0,Y=0,Z=1),Kind=Location,Space=(Type=Bone,Name="deform_skirt_b_r"))
    bMaintainOffset=true
   )
   > &RigUnit_ParentConstraint(
    AdvancedSettings=(InterpolationType=Average,RotationOrderForFilter=XZY)
    Child=(Type=Bone,Name="dyn_skirt_b_1_l")
    Parents=((Item=(Type=Bone,Name="deform_glute_l"),Weight=1))
    Weight=1
   )
   > &RigUnit_ParentConstraint(
    AdvancedSettings=(InterpolationType=Average,RotationOrderForFilter=XZY)
    Child=(Type=Bone,Name="dyn_skirt_b_1_r")
    Parents=((Item=(Type=Bone,Name="deform_glute_r"),Weight=1))
    Weight=1
   )
   > &RigUnit_ModifyTransforms(
    ItemToModify="((Item=(Type=Bone,Name=\"deform_skirt_b_l\"),Transform=(Rotation=(X=0,Y=0,Z=0,W=1),Translation=(X=0,Y=-1.2,Z=0),Scale3D=(X=1,Y=1,Z=1))),(Item=(Type=Bone,Name=\"deform_skirt_b_r\"),Transform=(Rotation=(X=0,Y=0,Z=0,W=1),Translation=(X=0,Y=-1.2,Z=0),Scale3D=(X=1, ...(truncated)"
    Weight=1
    WeightMinimum=0
   )
   &RigUnit_ParentConstraint(
    AdvancedSettings=(InterpolationType=Average,RotationOrderForFilter=XZY)
    Child=(Type=Bone,Name="dyn_L_shoulderPlate_main")
    Parents=((Item=(Type=Bone,Name="clavicle_l"),Weight=0.2),(Item=(Type=Bone,Name="dyn_L_customFrontArm_upperarm"),Weight=0.8))
    Weight=0.5
    ExecutePin=@RigVMFunction_Sequence_1.D
   )
   > &RigUnit_ParentConstraint(
    AdvancedSettings=(InterpolationType=Average,RotationOrderForFilter=XZY)
    Child=(Type=Bone,Name="dyn_R_shoulderPlate_main")
    Parents=((Item=(Type=Bone,Name="clavicle_r"),Weight=0.2),(Item=(Type=Bone,Name="dyn_R_customFrontArm_upperarm"),Weight=0.8))
    Weight=0.5
   )
   > Set Transform_1 = /Script/ControlRig.RigUnit_SetTranslation(
    Item=(Type=Bone,Name="dyn_L_shoulderPlate_main")
    Space=GlobalSpace
    Weight=0.6
    bInitial=false
    bPropagateToChildren=true
    Value=/Script/RigVM.RigVMFunction_MathTransformClampSpatially(
     Axis=X
     DebugColor=(R=1,G=0,B=0,A=1)
     DebugThickness=1
     Maximum=30
     Minimum=15
     Type=Plane
     bDrawDebug=true
     Value=&RigUnit_GetTransform(Item=(Type=Bone,Name="dyn_L_shoulderPlate_main"), bInitial=false).Transform
     Space=&RigUnit_GetTransform(Item=(Type=Bone,Name="clavicle_l"), bInitial=false).Transform
    ) {
     is=struct RigVMUnitNode
     template="ClampSpatially::Execute(in Value,in Axis,in Type,in Minimum,in Maximum,in Space,in bDrawDebug,in DebugColor,in DebugThickness,out Result)"
    }.Result.Translation
   ) {is=struct RigVMUnitNode, template="Set Transform::Execute(in Item,in Space,in bInitial,in Value,in Weight,in bPropagateToChildren)"}
   n55 = &RigUnit_GetTransform(Item=(Type=Bone,Name="clavicle_r"), bInitial=false)
   RigVMFunction_MathQuaternionToEuler = &RigVMFunction_MathQuaternionToEuler(RotationOrder=ZYX, Value=@n55.Transform.Rotation)
   RigVMFunction_MathQuaternionFromEuler = &RigVMFunction_MathQuaternionFromEuler(
    Euler=(X=0,Y=0,Z=0)
    Euler.X=@RigVMFunction_MathQuaternionToEuler.Result.X
    Euler.Y=@RigVMFunction_MathQuaternionToEuler.Result.Y
    Euler.Z=@RigVMFunction_MathQuaternionToEuler.Result.Z
   )
   n56 = &RigVMFunction_MathTransformClampSpatially(
    Axis=X
    DebugColor=(R=1,G=0,B=0,A=1)
    DebugThickness=1
    Maximum=30
    Minimum=16
    Space=(Rotation=(X=0.7071,Y=0,Z=0,W=0.7071),Translation=(X=0,Y=0,Z=0),Scale3D=(X=-1,Y=1,Z=1))
    Type=Plane
    bDrawDebug=true
    Space.Translation=@n55.Transform.Translation
    Value=&RigUnit_GetTransform(Item=(Type=Bone,Name="dyn_R_shoulderPlate_main"), bInitial=false).Transform
    Space.Rotation=@RigVMFunction_MathQuaternionFromEuler.Result
   ) {template="ClampSpatially::Execute(in Value,in Axis,in Type,in Minimum,in Maximum,in Space,in bDrawDebug,in DebugColor,in DebugThickness,out Result)"}
   > Set Transform_3 = &RigUnit_SetTranslation(
    Item=(Type=Bone,Name="dyn_R_shoulderPlate_main")
    Space=GlobalSpace
    Weight=0.6
    bInitial=false
    bPropagateToChildren=true
    Value=@n56.Result.Translation
   ) {template="Set Transform::Execute(in Item,in Space,in bInitial,in Value,in Weight,in bPropagateToChildren)"}
   > n57 = &RigUnit_SphericalPoseReader(
    ActiveRegionScaleFactors=(PositiveWidth=1,NegativeWidth=1,PositiveHeight=1,NegativeHeight=1)
    ActiveRegionSize=0.15
    Debug=(bDrawDebug=true,DebugScale=25,DebugSegments=20,DebugThickness=0.25)
    DriverAxis=(X=0,Y=0,Z=1)
    DriverItem=(Type=Bone,Name="upperarm_l")
    FalloffRegionScaleFactors=(PositiveWidth=1,NegativeWidth=1,PositiveHeight=1,NegativeHeight=1)
    FalloffSize=0.5
    OptionalParentItem=(Type=Bone,Name="spine_05")
    RotationOffset=(X=-110,Y=0,Z=30)
   )
   > &RigUnit_ModifyTransforms(
    ItemToModify=((Item=(Type=Bone,Name="dyn_L_shoulderPlate_main"),Transform=(Rotation=(X=-0.5736,Y=0,Z=0,W=0.8192),Translation=(X=1,Y=3,Z=-3),Scale3D=(X=1,Y=1,Z=1))))
    WeightMinimum=0
    Weight=@n57.OutputParam
   )
   > n58 = &RigUnit_SphericalPoseReader(
    ActiveRegionScaleFactors=(PositiveWidth=1,NegativeWidth=1,PositiveHeight=1,NegativeHeight=1)
    ActiveRegionSize=0.15
    Debug=(bDrawDebug=true,DebugScale=25,DebugSegments=20,DebugThickness=0.25)
    DriverAxis=(X=0,Y=0,Z=1)
    DriverItem=(Type=Bone,Name="upperarm_r")
    FalloffRegionScaleFactors=(PositiveWidth=1,NegativeWidth=1,PositiveHeight=1,NegativeHeight=1)
    FalloffSize=0.5
    OptionalParentItem=(Type=Bone,Name="spine_05")
    RotationOffset=(X=-110,Y=0,Z=30)
   )
   > &RigUnit_ModifyTransforms(
    ItemToModify=((Item=(Type=Bone,Name="dyn_R_shoulderPlate_main"),Transform=(Rotation=(X=-0.5736,Y=0,Z=0,W=0.8192),Translation=(X=-1,Y=-2,Z=3),Scale3D=(X=1,Y=1,Z=1))))
    WeightMinimum=0
    Weight=@n58.OutputParam
   )
   &RigUnit_AimConstraintLocalSpaceOffset(
    Child=(Type=Bone,Name="deform_skirt_l")
    Filter=(bX=True,bY=True,bZ=True)
    Parents=((Item=(Type=Bone,Name="dyn_L_custom_calf"),Weight=1))
    UpAxis=(X=0,Y=-1,Z=0)
    Weight=0.75
    WorldUp=(Target=(X=0,Y=0,Z=1),Kind=Location,Space=(Type=Bone,Name="deform_skirt_l"))
    bMaintainOffset=true
    ExecutePin=@RigVMFunction_Sequence_1.E
   )
   > &RigUnit_AimConstraintLocalSpaceOffset(
    Child=(Type=Bone,Name="deform_skirt_r")
    Filter=(bX=True,bY=True,bZ=True)
    Parents=((Item=(Type=Bone,Name="dyn_R_custom_calf"),Weight=1))
    UpAxis=(X=0,Y=-1,Z=0)
    Weight=0.75
    WorldUp=(Target=(X=0,Y=0,Z=1),Kind=Location,Space=(Type=Bone,Name="deform_skirt_r"))
    bMaintainOffset=true
   )
   > &RigUnit_AimConstraintLocalSpaceOffset(
    Child=(Type=Bone,Name="deform_lace_l")
    Filter=(bX=True,bY=True,bZ=True)
    Parents=((Item=(Type=Bone,Name="dyn_L_custom_calf"),Weight=1))
    UpAxis=(X=0,Y=-1,Z=0)
    Weight=0.75
    WorldUp=(Target=(X=0,Y=0,Z=1),Kind=Location,Space=(Type=Bone,Name="deform_skirt_l"))
    bMaintainOffset=true
   )
   > &RigUnit_AimConstraintLocalSpaceOffset(
    Child=(Type=Bone,Name="deform_lace_r")
    Filter=(bX=True,bY=True,bZ=True)
    Parents=((Item=(Type=Bone,Name="dyn_R_custom_calf"),Weight=1))
    UpAxis=(X=0,Y=-1,Z=0)
    Weight=0.8
    WorldUp=(Target=(X=0,Y=0,Z=1),Kind=Location,Space=(Type=Bone,Name="deform_skirt_r"))
    bMaintainOffset=true
   )
   > n59 = &RigUnit_SphericalPoseReader(
    ActiveRegionScaleFactors=(PositiveWidth=0.25,NegativeWidth=2,PositiveHeight=0.25,NegativeHeight=2)
    ActiveRegionSize=0.1
    Debug=(bDrawDebug=false,bDraw2D=False,bDrawLocalAxes=False,DebugScale=25,DebugSegments=20,DebugThickness=0.25)
    DriverAxis=(X=1,Y=0,Z=0)
    DriverItem=(Type=Bone,Name="thigh_r")
    FalloffRegionScaleFactors=(PositiveWidth=1,NegativeWidth=1,PositiveHeight=1,NegativeHeight=1)
    FalloffSize=0.25
    FlipHeightScaling=False
    FlipWidthScaling=False
    OptionalParentItem=(Type=Bone,Name="None")
    RotationOffset=(X=-90,Y=0,Z=0)
   )
   AlphaInterp = &RigVMFunction_AlphaInterp(InterpSpeedDecreasing=0.1, InterpSpeedIncreasing=0.1, Scale=1, bClampResult=False, bInterpResult=False, Value=@n59.OutputParam)
   > &RigUnit_ModifyTransforms(
    ItemToModify="((Item=(Type=Bone,Name=\"deform_belt_f\"),Transform=(Rotation=(X=0,Y=0,Z=-0.3827,W=0.9239),Translation=(X=0,Y=0,Z=0),Scale3D=(X=1,Y=1,Z=1))),(Item=(Type=Bone,Name=\"deform_belt_b\"),Transform=(Rotation=(X=0,Y=0.3827,Z=0,W=0.9239),Translation=(X=0,Y=0,Z=0),Scal ...(truncated)"
    WeightMinimum=0
    Weight=@AlphaInterp.Result
   )
   > n60 = &RigUnit_SphericalPoseReader(
    ActiveRegionScaleFactors=(PositiveWidth=1,NegativeWidth=3,PositiveHeight=3,NegativeHeight=1)
    ActiveRegionSize=0.1
    Debug=(bDrawDebug=false,bDraw2D=False,bDrawLocalAxes=False,DebugScale=25,DebugSegments=20,DebugThickness=0.25)
    DriverAxis=(X=1,Y=0,Z=0)
    DriverItem=(Type=Bone,Name="thigh_r")
    FalloffRegionScaleFactors=(PositiveWidth=1,NegativeWidth=1,PositiveHeight=1,NegativeHeight=1)
    FalloffSize=0.4
    FlipHeightScaling=False
    FlipWidthScaling=False
    OptionalParentItem=(Type=Bone,Name="None")
    RotationOffset=(X=-10,Y=-10,Z=0)
   )
   n61 = &RigVMFunction_AlphaInterp(InterpSpeedDecreasing=1, InterpSpeedIncreasing=1, Scale=1, bClampResult=False, bInterpResult=False, Value=@n60.OutputParam)
   > &RigUnit_ModifyTransforms(
    ItemToModify="((Item=(Type=Bone,Name=\"deform_belt_f\"),Transform=(Rotation=(X=-0.2896,Y=-0.096,Z=-0.2896,W=0.9072),Translation=(X=0,Y=0,Z=0),Scale3D=(X=1,Y=1,Z=1))),(Item=(Type=Bone,Name=\"deform_belt_b\"),Transform=(Rotation=(X=-0.0277,Y=0.1766,Z=-0.0782,W=0.9808),Transla ...(truncated)"
    WeightMinimum=0
    Weight=@n61.Result
   )
   n62 = &RigUnit_GetTransform(Item=(Type=Bone,Name="clavicle_l"), bInitial=false)
   Set Transform_2 = &RigUnit_SetTranslation(
    ExecutePin=()
    Item=(Type=Bone,Name="dyn_L_shoulderPlate_main")
    Space=GlobalSpace
    Weight=1
    bInitial=false
    bPropagateToChildren=true
    Value=&RigVMFunction_MathTransformClampSpatially(
     Axis=X
     DebugColor=(R=1,G=0,B=0,A=1)
     DebugThickness=1
     Maximum=30
     Minimum=15
     Type=Plane
     bDrawDebug=true
     Space=@n62.Transform
     Value=&RigUnit_GetTransform(Item=(Type=Bone,Name="dyn_L_shoulderPlate_back"), bInitial=false).Transform
    ) {template="ClampSpatially::Execute(in Value,in Axis,in Type,in Minimum,in Maximum,in Space,in bDrawDebug,in DebugColor,in DebugThickness,out Result)"}.Result.Translation
   ) {template="Set Transform::Execute(in Item,in Space,in bInitial,in Value,in Weight,in bPropagateToChildren)"}
   n63 = &RigUnit_GetTransform(Item=(Type=Bone,Name="dyn_L_customBackArm_upperarm"), bInitial=false)
   Add_1 = RigVMAggregateNode(
    A=@n63.Transform.Translation
    B=/Script/RigVM.RigVMFunction_MathVectorMul(
     B=(X=0,Y=0,Z=0)
     A=/Script/RigVM.RigVMFunction_MathQuaternionGetAxis(Axis=X, Quaternion=@n63.Transform.Rotation) {is=struct RigVMUnitNode, template="Axis::Execute(in Quaternion,in Axis,out Result)"}.Result
    ) {is=struct RigVMUnitNode, template="Multiply::Execute(in A,in B,out Result)"}.Result
    C=&RigVMFunction_MathVectorMul(
     B=(X=-5,Y=-5,Z=-5)
     A=&RigVMFunction_MathQuaternionGetAxis(Axis=Z, Quaternion=@n63.Transform.Rotation) {template="Axis::Execute(in Quaternion,in Axis,out Result)"}.Result
    ) {template="Multiply::Execute(in A,in B,out Result)"}.Result
   )
   n64 = &RigVMFunction_MathTransformClampSpatially(
    Axis=Z
    DebugColor=(R=1,G=0,B=0,A=1)
    DebugThickness=1
    Maximum=50
    Minimum=6
    Space=(Rotation=(X=0.7071,Y=0,Z=0,W=0.7071),Translation=(X=0,Y=0,Z=0),Scale3D=(X=1,Y=1,Z=1))
    Type=Plane
    bDrawDebug=true
    Space.Translation=@Add_1.Result
    Value=&RigUnit_GetTransform(Item=(Type=Bone,Name="dyn_L_shoulderPlate_main"), bInitial=false).Transform
    Space.Rotation=@n63.Transform.Rotation
   ) {template="ClampSpatially::Execute(in Value,in Axis,in Type,in Minimum,in Maximum,in Space,in bDrawDebug,in DebugColor,in DebugThickness,out Result)"}
   > Set Transform_4 = &RigUnit_SetTranslation(
    Item=(Type=Bone,Name="dyn_L_shoulderPlate_main")
    Space=GlobalSpace
    Weight=1
    bInitial=false
    bPropagateToChildren=true
    Value=@n64.Result.Translation
   ) {template="Set Transform::Execute(in Item,in Space,in bInitial,in Value,in Weight,in bPropagateToChildren)"}
   n65 = &RigUnit_GetTransform(Item=(Type=Bone,Name="dyn_L_customBackArm_upperarm"), bInitial=false)
   Add_6 = RigVMAggregateNode(
    A=@n65.Transform.Translation
    B=&RigVMFunction_MathVectorMul(
     B=(X=0,Y=0,Z=0)
     A=&RigVMFunction_MathQuaternionGetAxis(Axis=X, Quaternion=@n65.Transform.Rotation) {template="Axis::Execute(in Quaternion,in Axis,out Result)"}.Result
    ) {template="Multiply::Execute(in A,in B,out Result)"}.Result
    C=&RigVMFunction_MathVectorMul(
     B=(X=-5,Y=-5,Z=-5)
     A=&RigVMFunction_MathQuaternionGetAxis(Axis=Z, Quaternion=@n65.Transform.Rotation) {template="Axis::Execute(in Quaternion,in Axis,out Result)"}.Result
    ) {template="Multiply::Execute(in A,in B,out Result)"}.Result
   )
   n66 = &RigVMFunction_MathTransformClampSpatially(
    Axis=Y
    DebugColor=(R=1,G=0,B=0,A=1)
    DebugThickness=1
    Maximum=2
    Minimum=0
    Space=(Rotation=(X=0.7071,Y=0,Z=0,W=0.7071),Translation=(X=0,Y=0,Z=0),Scale3D=(X=1,Y=1,Z=1))
    Type=Plane
    bDrawDebug=true
    Space.Translation=@Add_6.Result
    Value=&RigUnit_GetTransform(Item=(Type=Bone,Name="dyn_L_shoulderPlate_main"), bInitial=false).Transform
    Space.Rotation=@n65.Transform.Rotation
   ) {template="ClampSpatially::Execute(in Value,in Axis,in Type,in Minimum,in Maximum,in Space,in bDrawDebug,in DebugColor,in DebugThickness,out Result)"}
   > Set Transform_6 = &RigUnit_SetTranslation(
    Item=(Type=Bone,Name="dyn_L_shoulderPlate_main")
    Space=GlobalSpace
    Weight=1
    bInitial=false
    bPropagateToChildren=true
    Value=@n66.Result.Translation
   ) {template="Set Transform::Execute(in Item,in Space,in bInitial,in Value,in Weight,in bPropagateToChildren)"}
   Add_4 = RigVMAggregateNode(
    A=@n62.Transform.Translation
    C=&RigVMFunction_MathVectorMul(
     B=(X=0,Y=0,Z=0)
     A=&RigVMFunction_MathQuaternionGetAxis(Axis=Z, Quaternion=@n62.Transform.Rotation) {template="Axis::Execute(in Quaternion,in Axis,out Result)"}.Result
    ) {template="Multiply::Execute(in A,in B,out Result)"}.Result
    B=&RigVMFunction_MathVectorMul(
     B=(X=10,Y=10,Z=10)
     A=&RigVMFunction_MathQuaternionGetAxis(Axis=X, Quaternion=@n62.Transform.Rotation) {template="Axis::Execute(in Quaternion,in Axis,out Result)"}.Result
    ) {template="Multiply::Execute(in A,in B,out Result)"}.Result
   )
   n67 = RigVMVariableNode() {variable=Dyn_is_crouch}
   Clamp = &RigVMFunction_MathDoubleClamp(Value=&RigVMFunction_MathDoubleAdd(A=@n61.Result, B=&RigVMFunction_MathDoubleMul(A=-0.5, B=@n67.Value).Result).Result)
   n68 = RigVMVariableNode() {variable=Dyn_is_crouch}
   &RigVMFunction_MathDoubleClamp(Value=&RigVMFunction_MathDoubleAdd(A=@AlphaInterp.Result, B=&RigVMFunction_MathDoubleMul(A=-0.3, B=@n68.Value).Result).Result)
   Multiply = &RigVMFunction_MathTransformMul(
    B=@n40.Transform
    A=&RigVMFunction_MathTransformMakeRelative(
     Parent=&RigUnit_GetTransform(Item=(Type=Bone,Name="ball_l"), bInitial=true).Transform
     Global=&RigUnit_GetTransform(Item=(Type=Bone,Name="dyn_L_custom_ball"), bInitial=true).Transform
    ).Local
   ) {template="Multiply::Execute(in A,in B,out Result)"}
   &RigVMFunction_MathTransformMul(
    B=@n41.Transform
    A=&RigVMFunction_MathTransformMakeRelative(
     Parent=&RigUnit_GetTransform(Item=(Type=Bone,Name="ball_r"), bInitial=true).Transform
     Global=&RigUnit_GetTransform(Item=(Type=Bone,Name="dyn_R_custom_ball"), bInitial=true).Transform
    ).Local
   ) {template="Multiply::Execute(in A,in B,out Result)"}
   &RigVMFunction_MathDoubleMul(A=0.75, B=1)
   &RigVMFunction_MathDoubleMul(B=0.5, A=@n48.Result.Z)
   &RigVMFunction_MathDoubleMul(B=0.5, A=@n49.Result.Z)
   RigVMRerouteNode(Value=@RigVMFunction_Sequence.C)
   Subtract = &RigVMFunction_MathDoubleSub(B=@AlphaInterp.Result, A=@n61.Result)
   Unit = /Script/RigVM.RigVMFunction_MathVectorUnit(Value=@"Get Delta Ref_2".Result.Location) {is=struct RigVMUnitNode, template="Unit::Execute(in Value,out Result)"}
  ]}
]
hierarchy [
 root {loc=<0 0 0>, rot=<0 0 0 1>}
 pelvis {parent=root, loc=<0 -0.7071 93.1191>, rot=<0.0127 0.707 -0.0127 -0.707>}
 spine_01 {parent=pelvis, loc=<6.7494 -0.6388 0>, rot=<0 0 0.034 -0.9994>}
 spine_02 {parent=spine_01, loc=<7.252 -0.035 0>, rot=<0 0 0.0238 -0.9997>}
 spine_03 {parent=spine_02, loc=<6.0106 0 0>, rot=<0 0 0.0149 -0.9999>}
 spine_04 {parent=spine_03, loc=<8.1023 0 0>, rot=<0 0 -0.0132 -0.9999>}
 spine_05 {parent=spine_04, loc=<10.0142 0 0>, rot=<0 0 0 1>}
 clavicle_l {parent=spine_05, loc=<4.3579 -0.1336 -2.3223>, rot=<0.0476 -0.7289 -0.01 -0.6829>}
 upperarm_l {parent=clavicle_l, loc=<15.5111 0 0>, rot=<0.0226 -0.3234 -0.0088 -0.9459>}
 lowerarm_l {parent=upperarm_l, loc=<26.6514 0 0>, rot=<0.0391 -0.0099 -0.2453 -0.9686>}
 hand_l {parent=lowerarm_l, loc=<22.3491 0 0>, rot=<-0.7081 -0.0507 0.0975 -0.6975>}
 hand_attach_l {parent=hand_l, loc=<8.5 -3 0>, rot=<0.7071 0.7071 0 0>}
 thumb_01_l {parent=hand_l, loc=<1.3118 -0.3191 -2.459>, rot=<0.6546 -0.4042 0.0664 -0.6354>}
 thumb_02_l {parent=thumb_01_l, loc=<5.9336 -0.0448 0.0401>, rot=<-0.0037 0.0213 0.298 -0.9543>}
 thumb_03_l {parent=thumb_02_l, loc=<2.2223 0.3682 -0.0907>, rot=<0.0118 -0.013 0.0066 -0.9998>}
 index_metacarpal_l {parent=hand_l, loc=<1.6535 0.0183 -2.96>, rot=<0.0158 -0.0124 -0.0173 -0.9996>}
 index_01_l {parent=index_metacarpal_l, loc=<8.3008 0.0357 -0.0524>, rot=<-0.0044 0.0485 0.0924 -0.9945>}
 index_02_l {parent=index_01_l, loc=<3.3587 -0.3004 0.0452>, rot=<-0.002 0.0027 0.3225 -0.9466>}
 index_03_l {parent=index_02_l, loc=<2.5578 -0.123 -0.0252>, rot=<-0.002 0.0067 0.2558 -0.9667>}
 middle_metacarpal_l {parent=hand_l, loc=<2.1856 -0.1725 -0.5984>, rot=<0.006 0.0168 -0.0329 -0.9993>}
 middle_01_l {parent=middle_metacarpal_l, loc=<7.3275 -0.2447 0.0129>, rot=<-0.0208 0.0042 0.0858 -0.9961>}
 middle_02_l {parent=middle_01_l, loc=<4.0123 0.1401 0.1175>, rot=<-0.0218 0.0258 0.3512 -0.9357>}
 middle_03_l {parent=middle_02_l, loc=<2.6077 0.0534 0.1373>, rot=<0.0082 0.0188 0.1708 -0.9851>}
 ring_metacarpal_l {parent=hand_l, loc=<2.2074 -0.3625 1.3638>, rot=<-0.0284 0.0722 -0.0458 -0.9959>}
 ring_01_l {parent=ring_metacarpal_l, loc=<7.0295 -0.6197 -0.1446>, rot=<-0.0489 -0.0252 0.1892 -0.9804>}
 ring_02_l {parent=ring_01_l, loc=<3.6174 0.4647 -0.1027>, rot=<0 0 0.2806 -0.9598>}
 ring_03_l {parent=ring_02_l, loc=<2.4152 0.114 0.0199>, rot=<0 0 0.1189 -0.9929>}
 pinky_metacarpal_l {parent=hand_l, loc=<1.7616 -0.9753 3.3814>, rot=<-0.0397 0.0814 -0.069 -0.9935>}
 pinky_01_l {parent=pinky_metacarpal_l, loc=<7.171 -0.483 -0.1094>, rot=<-0.0654 -0.0035 0.2242 -0.9723>}
 pinky_02_l {parent=pinky_01_l, loc=<2.8274 0.0892 -0.1043>, rot=<0.0037 0.0011 0.2854 -0.9584>}
 pinky_03_l {parent=pinky_02_l, loc=<2.2046 0.0307 -0.0078>, rot=<-0.0002 0.0098 0.1445 -0.9895>}
 weapon_l {parent=hand_l, loc=<-0.1974 -3.4111 0.3811>, rot=<0 0 0 1>}
 lowerarm_twist_02_l {parent=lowerarm_l, loc=<9.6199 0 0>, rot=<0 0 0 -1>}
 lowerarm_twist_01_l {parent=lowerarm_l, loc=<17.8549 0 0>, rot=<0 0 0 -1>}
 upperarm_twist_01_l {parent=upperarm_l, loc=<0.0041 0 0>, rot=<0 0 0 -1>}
 upperarm_twist_02_l {parent=upperarm_l, loc=<13.7941 0 0>, rot=<0 0 0 -1>}
 clavicle_r {parent=spine_05, loc=<4.3579 -0.1335 2.3223>, rot=<0.7289 0.0476 -0.6829 0.01>}
 upperarm_r {parent=clavicle_r, loc=<-15.5111 0 0>, rot=<0.0226 -0.3234 -0.0088 -0.9459>}
 lowerarm_r {parent=upperarm_r, loc=<-26.6514 0 0>, rot=<0.0391 -0.0099 -0.2453 -0.9686>}
 hand_r {parent=lowerarm_r, loc=<-22.3491 0 0>, rot=<-0.7081 -0.0507 0.0975 -0.6975>}
 thumb_01_r {parent=hand_r, loc=<-1.3118 0.3191 2.459>, rot=<0.6546 -0.4042 0.0664 -0.6354>}
 thumb_02_r {parent=thumb_01_r, loc=<-5.9336 0.0448 -0.0401>, rot=<-0.0037 0.0213 0.298 -0.9543>}
 thumb_03_r {parent=thumb_02_r, loc=<-2.2223 -0.3682 0.0907>, rot=<0.0118 -0.013 0.0066 -0.9998>}
 index_metacarpal_r {parent=hand_r, loc=<-1.6535 -0.0183 2.96>, rot=<0.0158 -0.0124 -0.0173 -0.9996>}
 index_01_r {parent=index_metacarpal_r, loc=<-8.3008 -0.0357 0.0524>, rot=<-0.0044 0.0485 0.0924 -0.9945>}
 index_02_r {parent=index_01_r, loc=<-3.3587 0.3004 -0.0452>, rot=<-0.002 0.0027 0.3225 -0.9466>}
 index_03_r {parent=index_02_r, loc=<-2.5578 0.123 0.0252>, rot=<-0.002 0.0067 0.2558 -0.9667>}
 middle_metacarpal_r {parent=hand_r, loc=<-2.1856 0.1725 0.5984>, rot=<0.006 0.0168 -0.0329 -0.9993>}
 middle_01_r {parent=middle_metacarpal_r, loc=<-7.3275 0.2447 -0.0129>, rot=<-0.0208 0.0042 0.0858 -0.9961>}
 middle_02_r {parent=middle_01_r, loc=<-4.0123 -0.1401 -0.1175>, rot=<-0.0218 0.0258 0.3512 -0.9357>}
 middle_03_r {parent=middle_02_r, loc=<-2.6077 -0.0534 -0.1373>, rot=<0.0082 0.0188 0.1708 -0.9851>}
 ring_metacarpal_r {parent=hand_r, loc=<-2.2074 0.3625 -1.3638>, rot=<-0.0284 0.0722 -0.0458 -0.9959>}
 ring_01_r {parent=ring_metacarpal_r, loc=<-7.0295 0.6197 0.1446>, rot=<-0.0489 -0.0252 0.1892 -0.9804>}
 ring_02_r {parent=ring_01_r, loc=<-3.6174 -0.4647 0.1027>, rot=<0 0 0.2806 -0.9598>}
 ring_03_r {parent=ring_02_r, loc=<-2.4152 -0.114 -0.0199>, rot=<0 0 0.1189 -0.9929>}
 pinky_metacarpal_r {parent=hand_r, loc=<-1.7616 0.9753 -3.3814>, rot=<-0.0397 0.0814 -0.069 -0.9935>}
 pinky_01_r {parent=pinky_metacarpal_r, loc=<-7.1709 0.4831 0.1094>, rot=<-0.0654 -0.0035 0.2242 -0.9723>}
 pinky_02_r {parent=pinky_01_r, loc=<-2.8274 -0.0893 0.1043>, rot=<0.0037 0.0011 0.2854 -0.9584>}
 pinky_03_r {parent=pinky_02_r, loc=<-2.2046 -0.0307 0.0078>, rot=<-0.0002 0.0098 0.1445 -0.9895>}
 weapon_r {parent=hand_r, loc=<0.1974 3.4111 -0.3811>, rot=<0 0 0 1>}
 hand_attach_r {parent=hand_r, loc=<-8.5001 2.9999 0>, rot=<0 0 -0.7071 -0.7071>}
 lowerarm_twist_02_r {parent=lowerarm_r, loc=<-9.6199 0 0>, rot=<0 0 0 -1>}
 lowerarm_twist_01_r {parent=lowerarm_r, loc=<-17.8549 0 0>, rot=<0 0 0 -1>}
 upperarm_twist_01_r {parent=upperarm_r, loc=<-0.0041 0 0>, rot=<0 0 0 -1>}
 upperarm_twist_02_r {parent=upperarm_r, loc=<-13.7941 0 0>, rot=<0 0 0 -1>}
 deform_belt_3 {parent=clavicle_r, loc=<-10.2753 1.3132 -6.9594>, rot=<-0.5575 0.0063 0.8299 0.0206>}
 neck_01 {parent=spine_05, loc=<11.3133 0.1867 0>, rot=<0 0 -0.2017 -0.9794>}
 neck_02 {parent=neck_01, loc=<4.5627 0 0>, rot=<0 0 0.0504 -0.9987>}
 head {parent=neck_02, loc=<5.3914 0 0>, rot=<0 0 0.115 -0.9934>}
 deform_strap_b {parent=head, loc=<1.251 -4.7692 2.9598>, rot=<0.2092 -0.9599 0.1861 0.0127>}
 dyn_strap_b {parent=deform_strap_b, loc=<0 0 0>, rot=<0 0 0 1>}
 dyn_strap_1_b {parent=dyn_strap_b, loc=<4.7443 0.0077 1.2477>, rot=<0.0188 0.0851 0.0104 -0.9961>}
 dyn_strap_2_b {parent=dyn_strap_1_b, loc=<3.2633 0 0>, rot=<0.0209 0.0195 -0.0727 -0.9969>}
 dyn_strap_3_b {parent=dyn_strap_b, loc=<4.7857 -0.0443 -0.396>, rot=<-0.0338 0.0101 -0.0034 -0.9994>}
 dyn_strap_4_b {parent=dyn_strap_3_b, loc=<5.5502 0.0007 0.5192>, rot=<-0.0047 0.0067 -0.0082 -0.9999>}
 dyn_strap_5_b {parent=dyn_strap_3_b, loc=<5.5988 -0.2737 -0.8203>, rot=<-0.0938 0.0022 0.0308 -0.9951>}
 deform_strap_r {parent=head, loc=<2.2707 3.5588 6.0928>, rot=<0.0092 -0.6182 0.7826 0.0728>}
 dyn_strap_r {parent=deform_strap_r, loc=<0 0 0>, rot=<0 0 0 1>}
 dyn_strap_1_r {parent=dyn_strap_r, loc=<5.8218 0 0>, rot=<-0.0776 -0.0595 0.0185 -0.995>}
 dyn_strap_2_r {parent=dyn_strap_1_r, loc=<6.615 0 0>, rot=<-0.0826 -0.0803 0.065 -0.9912>}
 dyn_strap_3_r {parent=dyn_strap_2_r, loc=<7.2589 -0.0002 -0.0001>, rot=<-0.2744 -0.0163 -0.0254 -0.9611>}
 deform_strap_l {parent=head, loc=<2.3624 3.5495 -6.1035>, rot=<-0.0092 0.6182 0.7826 0.0728>}
 dyn_strap_l {parent=deform_strap_l, loc=<0 0 0>, rot=<0 0 0 1>}
 dyn_strap_1_l {parent=dyn_strap_l, loc=<5.8217 0 0>, rot=<0.0768 0.0491 0.028 -0.9954>}
 dyn_strap_2_l {parent=dyn_strap_1_l, loc=<6.6148 0 0>, rot=<0.0848 0.078 0.0372 -0.9926>}
 dyn_strap_3_l {parent=dyn_strap_2_l, loc=<8.8147 0.0901 0.0271>, rot=<0.2744 0.0163 -0.0254 -0.9611>}
 attach_cape {parent=spine_05, loc=<9.052 -11.687 0>, rot=<0.0627 -0.7043 -0.0627 -0.7043>}
 cape {parent=attach_cape, type=Null, loc=<0 1.5303 -1.3086>, rot=<0 0 0 1>}
 attach_backpack {parent=spine_05, loc=<-0.0827 -13.4733 0>, rot=<-0.0082 -0.7071 0.0082 -0.7071>}
 Backpack_BR {parent=attach_backpack, type=Null, loc=<0 0.713 0>, rot=<0 0 0 1>}
 dyn_L_customFrontArm_clavicle {parent=spine_05, loc=<-3.6963 1.4162 -9.5394>, rot=<0.0288 -0.3736 -0.0599 -0.9252>}
 dyn_L_customFrontArm_upperarm {parent=dyn_L_customFrontArm_clavicle, loc=<10.2442 0 0>, rot=<0 -0.6806 0 -0.7327>}
 dyn_L_customFrontArm_lowerarm {parent=dyn_L_customFrontArm_upperarm, loc=<26.397 0 0>, rot=<0.1856 -0.0422 -0.2176 -0.9573>}
 dyn_L_customFrontArm_wrist {parent=dyn_L_customFrontArm_lowerarm, loc=<23.1045 -0.7127 -0.3162>, rot=<-0.7071 0 0 -0.7071>}
 dyn_L_custom_index_1 {parent=dyn_L_customFrontArm_wrist, loc=<8.8135 1.0897 -0.6224>, rot=<0.7178 -0.1239 -0.0064 -0.6851>}
 dyn_L_custom_index_2 {parent=dyn_L_custom_index_1, loc=<3.5083 0 0>, rot=<0 -0.2835 0 -0.959>}
 dyn_L_custom_index_3 {parent=dyn_L_custom_index_2, loc=<2.9925 0 0>, rot=<-0.1582 -0.4125 -0.073 -0.8941>}
 dyn_L_custom_thumb_1 {parent=dyn_L_customFrontArm_wrist, loc=<1.9169 0.5869 -1.1471>, rot=<-0.1438 -0.4135 0.3197 -0.8403>}
 dyn_L_custom_thumb_2 {parent=dyn_L_custom_thumb_1, loc=<4.6371 0 0>, rot=<0 0.3526 0 -0.9358>}
 dyn_L_custom_thumb_3 {parent=dyn_L_custom_thumb_2, loc=<2.5336 0 0>, rot=<0 0.2518 0 -0.9678>}
 dyn_L_customFrontArm_wireAdjuster {parent=dyn_L_customFrontArm_lowerarm, loc=<8.2109 0.0484 0.1462>, rot=<0 0 0 1>}
 dyn_L_customBackArm_clavicle {parent=spine_05, loc=<-4.2849 -2.1385 -9.2883>, rot=<0.0206 -0.3627 -0.0346 -0.9311>}
 dyn_L_customBackArm_upperarm {parent=dyn_L_customBackArm_clavicle, loc=<11.8541 0 0>, rot=<0 -0.695 0 -0.719>}
 dyn_L_customBackArm_lowerarm {parent=dyn_L_customBackArm_upperarm, loc=<28.246 0 0>, rot=<0.1791 -0.0455 -0.2421 -0.9525>}
 dyn_L_customBackArm_wrist {parent=dyn_L_customBackArm_lowerarm, loc=<22.4142 -0.5164 -0.442>, rot=<-0.7071 0 0 -0.7071>}
 dyn_L_custom_ring_1 {parent=dyn_L_customBackArm_wrist, loc=<8.6761 0.0888 1.8777>, rot=<0.5269 0.0272 0.0974 -0.8439>}
 dyn_L_custom_ring_2 {parent=dyn_L_custom_ring_1, loc=<3.2343 0 0>, rot=<0 -0.3309 0 -0.9437>}
 dyn_L_custom_ring_3 {parent=dyn_L_custom_ring_2, loc=<3.1061 0 0>, rot=<0.0274 -0.4048 0.0121 -0.9139>}
 dyn_L_custom_extraThumb_1 {parent=dyn_L_customBackArm_wrist, loc=<1.5302 0.0024 1.4683>, rot=<-0.0058 0.3074 0.2728 -0.9116>}
 dyn_L_custom_extraThumb_2 {parent=dyn_L_custom_extraThumb_1, loc=<4.3712 0 0>, rot=<0 -0.2723 0 -0.9622>}
 dyn_L_custom_extraThumb_3 {parent=dyn_L_custom_extraThumb_2, loc=<3.9198 0 0>, rot=<0 -0.3043 0 -0.9526>}
 dyn_L_customBackArm_wireAdjuster {parent=dyn_L_customBackArm_lowerarm, loc=<7.9802 0.16 -0.3091>, rot=<-0.2106 -0.0207 -0.0053 -0.9773>}
 dyn_L_shoulderPlate_front {parent=dyn_L_customBackArm_upperarm, loc=<-1.1536 8.0221 1.1063>, rot=<0.4473 -0.7362 -0.2605 0.436>}
 dyn_L_shoulderPlate_back {parent=dyn_L_shoulderPlate_front, loc=<11.3198 -10.057 -0.2791>, rot=<0.0894 0.0504 -0.7721 -0.6272>}
 dyn_L_shoulderPlate_main {parent=dyn_L_shoulderPlate_back, loc=<6.2584 4.7544 1.5821>, rot=<-0.3177 -0.7658 0.2649 -0.4925>}
 dyn_L_customBackArm_clavicle_adjuster {parent=dyn_L_customBackArm_clavicle, loc=<4.7972 -1.3501 0.8084>, rot=<-0.0045 0.0709 0.016 -0.9973>}
 dyn_R_customFrontArm_clavicle {parent=spine_05, loc=<-3.6962 -0.1836 9.6423>, rot=<-0.3736 -0.0288 0.9252 -0.0599>}
 dyn_R_customFrontArm_upperarm {parent=dyn_R_customFrontArm_clavicle, loc=<-10.2441 0 0.0002>, rot=<0 -0.6806 0 -0.7327>}
 dyn_R_customFrontArm_lowerarm {parent=dyn_R_customFrontArm_upperarm, loc=<-26.3968 0 -0.0001>, rot=<0.1856 -0.0422 -0.2176 -0.9573>}
 dyn_R_customFrontArm_wrist {parent=dyn_R_customFrontArm_lowerarm, loc=<-23.5231 -0.6802 -0.3579>, rot=<-0.7071 0 0 -0.7071>}
 dyn_R_custom_index_1 {parent=dyn_R_customFrontArm_wrist, loc=<-8.8136 -1.0897 0.6224>, rot=<0.7178 -0.1239 -0.0064 -0.6851>}
 dyn_R_custom_index_2 {parent=dyn_R_custom_index_1, loc=<-3.5084 0 0>, rot=<0 -0.2835 0 -0.959>}
 dyn_R_custom_index_3 {parent=dyn_R_custom_index_2, loc=<-2.9925 0 0>, rot=<-0.1582 -0.4125 -0.073 -0.8941>}
 dyn_R_custom_thumb_1 {parent=dyn_R_customFrontArm_wrist, loc=<-1.9174 -0.5865 1.1471>, rot=<-0.1438 -0.4135 0.3197 -0.8403>}
 dyn_R_custom_thumb_2 {parent=dyn_R_custom_thumb_1, loc=<-4.6368 0 0.0004>, rot=<0 0.3526 0 -0.9358>}
 dyn_R_custom_thumb_3 {parent=dyn_R_custom_thumb_2, loc=<-2.5336 0 -0.0001>, rot=<0 0.2518 0 -0.9678>}
 dyn_R_customFrontArm_wireAdjuster {parent=dyn_R_customFrontArm_lowerarm, loc=<-8.211 -0.0484 -0.1462>, rot=<0 0 0 1>}
 dyn_R_customFrontArm_clavicle_adjuster {parent=dyn_R_customFrontArm_clavicle, loc=<-4.7948 -1.0094 -1.168>, rot=<-0.0137 0.0839 -0.0332 -0.9958>}
 dyn_R_customBackArm_clavicle {parent=spine_05, loc=<-4.2845 -3.6476 8.8057>, rot=<-0.3627 -0.0206 0.9311 -0.0346>}
 dyn_R_customBackArm_upperarm {parent=dyn_R_customBackArm_clavicle, loc=<-11.854 0 0.0001>, rot=<0 -0.695 0 -0.719>}
 dyn_R_customBackArm_lowerarm {parent=dyn_R_customBackArm_upperarm, loc=<-28.2462 0 0.0002>, rot=<0.1791 -0.0455 -0.2421 -0.9525>}
 dyn_R_customBackArm_wrist {parent=dyn_R_customBackArm_lowerarm, loc=<-23.2584 -0.4646 -0.4717>, rot=<-0.7071 0 0 -0.7071>}
 dyn_R_custom_ring_1 {parent=dyn_R_customBackArm_wrist, loc=<-8.6758 -0.0892 -1.8776>, rot=<0.5269 0.0272 0.0974 -0.8439>}
 dyn_R_custom_ring_2 {parent=dyn_R_custom_ring_1, loc=<-3.2344 0 0>, rot=<0 -0.3309 0 -0.9437>}
 dyn_R_custom_ring_3 {parent=dyn_R_custom_ring_2, loc=<-3.1061 0 0>, rot=<0.0274 -0.4048 0.0121 -0.9139>}
 dyn_R_custom_extraThumb_1 {parent=dyn_R_customBackArm_wrist, loc=<-1.5299 -0.0028 -1.4683>, rot=<-0.0058 0.3074 0.2728 -0.9116>}
 dyn_R_custom_extraThumb_2 {parent=dyn_R_custom_extraThumb_1, loc=<-4.3711 0.0001 0>, rot=<0 -0.2723 0 -0.9622>}
 dyn_R_custom_extraThumb_3 {parent=dyn_R_custom_extraThumb_2, loc=<-3.9199 0 0>, rot=<0 -0.3043 0 -0.9526>}
 dyn_R_customBackArm_wireAdjuster {parent=dyn_R_customBackArm_lowerarm, loc=<-7.9801 -0.16 0.309>, rot=<-0.2106 -0.0207 -0.0053 -0.9773>}
 dyn_R_shoulderPlate_front {parent=dyn_R_customBackArm_upperarm, loc=<0.8802 -9.5328 -1.4978>, rot=<0.4473 -0.7362 -0.2605 0.436>}
 dyn_R_shoulderPlate_back {parent=dyn_R_shoulderPlate_front, loc=<-11.3199 10.0567 0.279>, rot=<0.0894 0.0504 -0.7721 -0.6272>}
 dyn_R_shoulderPlate_main {parent=dyn_R_shoulderPlate_back, loc=<-6.2586 -4.7543 -1.5821>, rot=<-0.3177 -0.7658 0.2649 -0.4925>}
 dyn_R_customBackArm_clavicle_adjuster {parent=dyn_R_customBackArm_clavicle, loc=<-4.7966 1.3501 -0.8079>, rot=<-0.0045 0.0709 0.016 -0.9973>}
 dyn_L_customFrontArm_clavicle_IK {parent=spine_05, loc=<-3.6963 0.6185 -9.6241>, rot=<0.0288 -0.3736 -0.0599 -0.9252>}
 dyn_L_customFrontArm_upperarm_IK {parent=dyn_L_customFrontArm_clavicle_IK, loc=<10.2442 0 0>, rot=<0 -0.6806 0 -0.7327>}
 dyn_L_customFrontArm_lowerarm_IK {parent=dyn_L_customFrontArm_upperarm_IK, loc=<26.397 0 0>, rot=<0.1856 -0.0422 -0.2176 -0.9573>}
 dyn_L_customFrontArm_wrist_IK {parent=dyn_L_customFrontArm_lowerarm_IK, loc=<23.2934 0 0>, rot=<-0.7071 0 0 -0.7071>}
 dyn_L_customFrontArm_clavicle_FK {parent=spine_05, loc=<-3.6963 0.6185 -9.6241>, rot=<0.0288 -0.3736 -0.0599 -0.9252>}
 dyn_L_customFrontArm_upperarm_FK {parent=dyn_L_customFrontArm_clavicle_FK, loc=<10.2442 0 0>, rot=<0 -0.6806 0 -0.7327>}
 dyn_L_customFrontArm_lowerarm_FK {parent=dyn_L_customFrontArm_upperarm_FK, loc=<26.397 0 0>, rot=<0.1856 -0.0422 -0.2176 -0.9573>}
 dyn_L_customFrontArm_wrist_FK {parent=dyn_L_customFrontArm_lowerarm_FK, loc=<23.2934 0 0>, rot=<-0.7071 0 0 -0.7071>}
 dyn_L_customBackArm_clavicle_FK {parent=spine_05, loc=<-4.2849 -2.9031 -9.0784>, rot=<0.0206 -0.3627 -0.0346 -0.9311>}
 dyn_L_customBackArm_upperarm_FK {parent=dyn_L_customBackArm_clavicle_FK, loc=<11.8541 0 0>, rot=<0 -0.695 0 -0.719>}
 dyn_L_customBackArm_lowerarm_FK {parent=dyn_L_customBackArm_upperarm_FK, loc=<28.246 0 0>, rot=<0.1791 -0.0455 -0.2421 -0.9525>}
 dyn_L_customBackArm_wrist_FK {parent=dyn_L_customBackArm_lowerarm_FK, loc=<22.8224 0 0>, rot=<-0.7071 0 0 -0.7071>}
 dyn_L_customBackArm_clavicle_IK {parent=spine_05, loc=<-4.2849 -2.9031 -9.0784>, rot=<0.0206 -0.3627 -0.0346 -0.9311>}
 dyn_L_customBackArm_upperarm_IK {parent=dyn_L_customBackArm_clavicle_IK, loc=<11.8541 0 0>, rot=<0 -0.695 0 -0.719>}
 dyn_L_customBackArm_lowerarm_IK {parent=dyn_L_customBackArm_upperarm_IK, loc=<28.246 0 0>, rot=<0.1791 -0.0455 -0.2421 -0.9525>}
 dyn_L_customBackArm_wrist_IK {parent=dyn_L_customBackArm_lowerarm_IK, loc=<22.8224 0 0>, rot=<-0.7071 0 0 -0.7071>}
 dyn_R_customFrontArm_clavicle_IK {parent=spine_05, loc=<-3.6962 0.6185 9.6242>, rot=<-0.3736 -0.0288 0.9252 -0.0599>}
 dyn_R_customFrontArm_upperarm_IK {parent=dyn_R_customFrontArm_clavicle_IK, loc=<-10.2441 0 0.0002>, rot=<0 -0.6806 0 -0.7327>}
 dyn_R_customFrontArm_lowerarm_IK {parent=dyn_R_customFrontArm_upperarm_IK, loc=<-26.3968 0 -0.0001>, rot=<0.1856 -0.0422 -0.2176 -0.9573>}
 dyn_R_customFrontArm_wrist_IK {parent=dyn_R_customFrontArm_lowerarm_IK, loc=<-23.2934 0 0.0001>, rot=<-0.7071 0 0 -0.7071>}
 dyn_R_customBackArm_clavicle_IK {parent=spine_05, loc=<-4.2845 -2.9031 9.0784>, rot=<-0.3627 -0.0206 0.9311 -0.0346>}
 dyn_R_customBackArm_upperarm_IK {parent=dyn_R_customBackArm_clavicle_IK, loc=<-11.854 0 0.0001>, rot=<0 -0.695 0 -0.719>}
 dyn_R_customBackArm_lowerarm_IK {parent=dyn_R_customBackArm_upperarm_IK, loc=<-28.2462 0 0.0002>, rot=<0.1791 -0.0455 -0.2421 -0.9525>}
 dyn_R_customBackArm_wrist_IK {parent=dyn_R_customBackArm_lowerarm_IK, loc=<-22.8228 0 0.0004>, rot=<-0.7071 0 0 -0.7071>}
 dyn_R_customFrontArm_clavicle_FK {parent=spine_05, loc=<-3.6962 0.6185 9.6242>, rot=<-0.3736 -0.0288 0.9252 -0.0599>}
 dyn_R_customFrontArm_upperarm_FK {parent=dyn_R_customFrontArm_clavicle_FK, loc=<-10.2441 0 0.0002>, rot=<0 -0.6806 0 -0.7327>}
 dyn_R_customFrontArm_lowerarm_FK {parent=dyn_R_customFrontArm_upperarm_FK, loc=<-26.3968 0 -0.0001>, rot=<0.1856 -0.0422 -0.2176 -0.9573>}
 dyn_R_customFrontArm_wrist_FK {parent=dyn_R_customFrontArm_lowerarm_FK, loc=<-23.2934 0 0.0001>, rot=<-0.7071 0 0 -0.7071>}
 dyn_R_customBackArm_clavicle_FK {parent=spine_05, loc=<-4.2845 -2.9031 9.0784>, rot=<-0.3627 -0.0206 0.9311 -0.0346>}
 dyn_R_customBackArm_upperarm_FK {parent=dyn_R_customBackArm_clavicle_FK, loc=<-11.854 0 0.0001>, rot=<0 -0.695 0 -0.719>}
 dyn_R_customBackArm_lowerarm_FK {parent=dyn_R_customBackArm_upperarm_FK, loc=<-28.2462 0 0.0002>, rot=<0.1791 -0.0455 -0.2421 -0.9525>}
 dyn_R_customBackArm_wrist_FK {parent=dyn_R_customBackArm_lowerarm_FK, loc=<-22.8228 0 0.0004>, rot=<-0.7071 0 0 -0.7071>}
 spine_05_locator {parent=spine_05, loc=<0 0 0>, rot=<0 0 0 1>}
 deform_belt_1 {parent=spine_04, loc=<3.7358 13.46 5.1795>, rot=<-0.3391 -0.0389 0.9382 0.0574>}
 deform_belt_2 {parent=spine_04, loc=<-7.1267 11.2593 -5.7105>, rot=<-0.4093 0.117 0.88 -0.2107>}
 deform_grenades {parent=spine_01, loc=<2.6223 1.8802 11.2886>, rot=<0.0741 -0.618 0.7827 -0.0044>}
 deform_skirt_l {parent=spine_01, loc=<-0.4524 7.8642 -4.8166>, rot=<-0.0279 0.0009 0.9952 0.0934>}
 dyn_skirt_l {parent=deform_skirt_l, loc=<0 0 0>, rot=<0 0 0 1>}
 dyn_skirt_down_l {parent=dyn_skirt_l, loc=<9.8504 0 0>, rot=<0 0 0 1>}
 dyn_skirt_down_1_l {parent=dyn_skirt_down_l, loc=<15.9663 0 0>, rot=<0 0 0 1>}
 dyn_skirt_up_1_l {parent=dyn_skirt_down_l, loc=<9.4911 -0.8663 0>, rot=<0 0 0 1>}
 deform_lace_l {parent=spine_01, loc=<-0.4317 9.348 -1.8569>, rot=<-0.0279 0.0002 0.9926 0.1183>}
 dyn_lace_l {parent=deform_lace_l, loc=<0 0 0>, rot=<0 0 0 1>}
 dyn_lace_1_l {parent=dyn_lace_l, loc=<6.7675 0 0>, rot=<0 0 0 1>}
 dyn_lace_2_l {parent=dyn_lace_1_l, loc=<7.2229 0 0>, rot=<0 0 0 1>}
 deform_belt_f {parent=spine_01, loc=<0.0904 6.2037 8.6609>, rot=<0.1998 -0.3896 0.8937 0.0978>}
 dyn_belt_f {parent=deform_belt_f, loc=<0 0 0>, rot=<0 0 0 1>}
 dyn_belt_1_f {parent=dyn_belt_f, loc=<10.4366 0 0>, rot=<0.0112 0.2388 -0.2396 -0.941>}
 dyn_belt_2_f {parent=dyn_belt_1_f, loc=<12.0598 -0.0195 0.21>, rot=<0.1242 0.3902 -0.2546 -0.8761>}
 deform_belt_b {parent=spine_01, loc=<2.9788 -7.9291 7.6479>, rot=<0.104 -0.8861 0.404 0.2019>}
 dyn_belt_b {parent=deform_belt_b, loc=<0 0 0>, rot=<0 0 0 1>}
 dyn_belt_1_b {parent=dyn_belt_b, loc=<13.6181 0 0>, rot=<0.0246 -0.2169 -0.2487 -0.9437>}
 dyn_belt_2_b {parent=dyn_belt_1_b, loc=<11.6126 0.0064 -0.0395>, rot=<0.2164 -0.8969 -0.1513 0.3547>}
 deform_lace_r {parent=spine_01, loc=<-0.432 9.348 1.857>, rot=<0.0279 -0.0002 0.9926 0.1183>}
 dyn_lace_r {parent=deform_lace_r, loc=<0 0 0>, rot=<0 0 0 1>}
 dyn_lace_1_r {parent=dyn_lace_r, loc=<6.7671 0 0>, rot=<0 0 0 1>}
 dyn_lace_2_r {parent=dyn_lace_1_r, loc=<7.2229 0 0>, rot=<0 0 0 1>}
 deform_skirt_r {parent=spine_01, loc=<-0.4524 7.8642 4.8166>, rot=<0.0279 -0.0009 0.9952 0.0934>}
 dyn_skirt_r {parent=deform_skirt_r, loc=<0 0 0>, rot=<0 0 0 1>}
 dyn_skirt_down_r {parent=dyn_skirt_r, loc=<9.8504 0 0>, rot=<0 0 0 1>}
 dyn_skirt_up_1_r {parent=dyn_skirt_down_r, loc=<9.4911 -0.8663 0>, rot=<0 0 0 1>}
 dyn_skirt_down_1_r {parent=dyn_skirt_down_r, loc=<15.9663 0 0>, rot=<0 0 0 1>}
 deform_skirt_b_l {parent=spine_01, loc=<2.9175 -9.8179 -4.0547>, rot=<0.0941 -0.9939 -0.0569 -0.0123>}
 dyn_skirt_b_l {parent=deform_skirt_b_l, loc=<0 0 0>, rot=<0 0 0 1>}
 dyn_skirt_b_1_l {parent=dyn_skirt_b_l, loc=<12.8277 0 0>, rot=<0 0 -0.0794 -0.9968>}
 dyn_skirt_b_2_l {parent=dyn_skirt_b_1_l, loc=<11.3991 0 0>, rot=<0 0 0 1>}
 deform_skirt_b_r {parent=spine_01, loc=<2.9171 -9.8179 4.0547>, rot=<0.0941 -0.9939 0.0569 0.0123>}
 dyn_skirt_b_r {parent=deform_skirt_b_r, loc=<0 0 0>, rot=<0 0 0 -1>}
 dyn_skirt_b_1_r {parent=dyn_skirt_b_r, loc=<12.8273 -0.0001 0>, rot=<0 0 -0.0794 -0.9968>}
 dyn_skirt_b_2_r {parent=dyn_skirt_b_1_r, loc=<11.3991 0 0>, rot=<0 0 0 -1>}
 thigh_l {parent=pelvis, loc=<-2.5693 -0.0109 -7.7501>, rot=<0.0012 0.0572 0.0155 -0.9982>}
 calf_l {parent=thigh_l, loc=<-39.0484 0 0>, rot=<0.0006 -0.0161 -0.0391 -0.9991>}
 foot_l {parent=calf_l, loc=<-42.561 0.1077 0.4292>, rot=<0.0418 -0.0455 0.0343 -0.9975>}
 ball_l {parent=foot_l, loc=<-6.1599 13.8264 0.6806>, rot=<0.0001 -0.0001 -0.7063 -0.7079>}
 calf_twist_02_l {parent=calf_l, loc=<-18.8064 -0.0967 0.0397>, rot=<0.0237 0.0095 -0.0035 -0.9997>}
 calf_twist_01_l {parent=calf_l, loc=<-35.813 -0.0967 0.0397>, rot=<0.0237 0.0095 -0.0035 -0.9997>}
 thigh_twist_01_l {parent=thigh_l, loc=<-21.6303 0 0>, rot=<0.0013 0 0 -1>}
 dyn_L_custom_calf {parent=thigh_l, loc=<-37.4434 4.2041 0.7172>, rot=<0.0237 -0.0225 -0.3459 -0.9377>}
 dyn_L_custom_calf_1 {parent=dyn_L_custom_calf, loc=<-18.1799 0 0>, rot=<0 0 0.4115 -0.9114>}
 dyn_L_custom_foot {parent=dyn_L_custom_calf_1, loc=<-30.6903 2.6181 0>, rot=<-0.027 -0.0463 -0.0699 -0.9961>}
 dyn_L_custom_ball {parent=dyn_L_custom_foot, loc=<-6.6526 10.3851 -1.222>, rot=<0 0 -0.7071 -0.7071>}
 dyn_L_custom_ball_corrective {parent=dyn_L_custom_foot, loc=<-6.6526 10.3851 -1.222>, rot=<0 0 -0.7071 -0.7071>}
 deform_knee_out_base_l {parent=dyn_L_custom_calf, loc=<1.3519 4.4237 0.3216>, rot=<-0.0283 0.0061 0.2373 -0.971>}
 deform_knee_out_l {parent=deform_knee_out_base_l, loc=<0 0 0>, rot=<0 0 0 1>}
 deform_knee_in_base_l {parent=dyn_L_custom_calf, loc=<-1.7762 -4.4062 -0.2672>, rot=<-0.0283 0.0061 0.2373 -0.971>}
 deform_knee_in_l {parent=deform_knee_in_base_l, loc=<0 0 0>, rot=<0 0 0 1>}
 thigh_r {parent=pelvis, loc=<-2.5693 -0.0109 7.7501>, rot=<0.0572 -0.0012 0.9982 0.0155>}
 calf_r {parent=thigh_r, loc=<39.0484 0 0>, rot=<0.0006 -0.0161 -0.0391 -0.9991>}
 foot_r {parent=calf_r, loc=<42.561 -0.1077 -0.4292>, rot=<0.0418 -0.0455 0.0343 -0.9975>}
 ball_r {parent=foot_r, loc=<6.1599 -13.8264 -0.6806>, rot=<0.0001 -0.0001 -0.7063 -0.7079>}
 calf_twist_02_r {parent=calf_r, loc=<18.8064 0.0967 -0.0397>, rot=<0.0237 0.0095 -0.0035 -0.9997>}
 calf_twist_01_r {parent=calf_r, loc=<35.813 0.0967 -0.0397>, rot=<0.0237 0.0095 -0.0035 -0.9997>}
 thigh_twist_01_r {parent=thigh_r, loc=<21.6303 0 0>, rot=<0.0013 0 0 -1>}
 dyn_R_custom_calf {parent=thigh_r, loc=<37.4435 -4.2041 -0.7173>, rot=<0.0237 -0.0225 -0.3459 -0.9377>}
 dyn_R_custom_calf_1 {parent=dyn_R_custom_calf, loc=<18.1799 0 0>, rot=<0 0 0.4115 -0.9114>}
 dyn_R_custom_foot {parent=dyn_R_custom_calf_1, loc=<30.6903 -2.6181 0>, rot=<-0.0268 -0.0463 -0.0699 -0.9961>}
 dyn_R_custom_ball_corrective {parent=dyn_R_custom_foot, loc=<6.6526 -10.3851 1.2219>, rot=<0 0 -0.7071 -0.7071>}
 dyn_R_custom_ball {parent=dyn_R_custom_foot, loc=<6.6526 -10.3851 1.2219>, rot=<0 0 -0.7071 -0.7071>}
 deform_knee_in_base_r {parent=dyn_R_custom_calf, loc=<1.7761 4.4062 0.2668>, rot=<-0.0283 0.0061 0.2373 -0.971>}
 deform_knee_in_r {parent=deform_knee_in_base_r, loc=<0 0 0>, rot=<0 0 0 1>}
 deform_knee_out_base_r {parent=dyn_R_custom_calf, loc=<-1.3518 -4.4238 -0.3213>, rot=<-0.0283 0.0061 0.2373 -0.971>}
 deform_knee_out_r {parent=deform_knee_out_base_r, loc=<0 0 0>, rot=<0 0 0 1>}
 thigh_l_IK {parent=pelvis, loc=<-2.5693 -0.0109 -7.7501>, rot=<0.0285 0.0454 0.0727 -0.9959>}
 dyn_L_custom_calf_IK {parent=thigh_l_IK, loc=<-37.6855 0 0>, rot=<0 0 -0.3977 -0.9175>}
 dyn_L_custom_calf_1_IK {parent=dyn_L_custom_calf_IK, loc=<-18.1799 0 0>, rot=<0 0 0.4115 -0.9114>}
 dyn_L_custom_foot_IK {parent=dyn_L_custom_calf_1_IK, loc=<-30.6903 2.6181 0>, rot=<-0.027 -0.0463 -0.0699 -0.9961>}
 deform_groin_base_l {parent=thigh_l_IK, loc=<-3.0487 7.9165 -1.3102>, rot=<-0.0307 0.0112 -0.0555 -0.9979>}
 deform_groin_l {parent=deform_groin_base_l, loc=<0 0 0>, rot=<0 0 0 1>}
 thigh_twist_01_l_IK {parent=thigh_l_IK, loc=<-21.622 0 0>, rot=<0 0 0 1>}
 thigh_r_IK {parent=pelvis, loc=<-2.5694 -0.0109 7.7501>, rot=<0.0454 -0.0285 0.9959 0.0727>}
 dyn_R_custom_calf_IK {parent=thigh_r_IK, loc=<37.6855 0 0>, rot=<0.0002 0 -0.3977 -0.9175>}
 dyn_R_custom_calf_1_IK {parent=dyn_R_custom_calf_IK, loc=<18.1799 0 0>, rot=<0 0 0.4115 -0.9114>}
 dyn_R_custom_foot_IK {parent=dyn_R_custom_calf_1_IK, loc=<30.6903 -2.6181 0>, rot=<-0.027 -0.0463 -0.0699 -0.9961>}
 deform_groin_base_r {parent=thigh_r_IK, loc=<3.0486 -7.9165 1.3102>, rot=<-0.0307 0.0112 -0.0555 -0.9979>}
 deform_groin_r {parent=deform_groin_base_r, loc=<0 0 0>, rot=<0 0 0 1>}
 thigh_twist_01_r_IK {parent=thigh_r_IK, loc=<21.6219 0 0>, rot=<0 0 0 -1>}
 deform_glute_base_r {parent=pelvis, loc=<-8.8062 -8.8222 8.4287>, rot=<0.0501 -0.0275 0.8926 -0.4472>}
 deform_glute_r {parent=deform_glute_base_r, loc=<0 0 0>, rot=<0 0 0 1>}
 deform_glute_base_l {parent=pelvis, loc=<-8.8064 -8.8222 -8.4287>, rot=<0.0275 0.0501 -0.4472 -0.8926>}
 deform_glute_l {parent=deform_glute_base_l, loc=<0 0 0>, rot=<0 0 0 1>}
 attach {parent=root, loc=<0 0 0>, rot=<0 0 0 1>}
 ik_hand_root {parent=root, loc=<0 0 0>, rot=<0 0 0 1>}
 ik_hand_gun {parent=ik_hand_root, loc=<-51.9727 4.0883 102.6935>, rot=<0.6084 0.1793 0.4613 -0.6205>}
 ik_hand_r {parent=ik_hand_gun, loc=<0 0 0>, rot=<0 0 0 1>}
 ik_hand_l {parent=ik_hand_gun, loc=<53.0285 82.1781 35.2051>, rot=<-0.5102 -0.7906 -0.3387 0>}
 ik_foot_root {parent=root, loc=<0 0 0>, rot=<0 0 0 1>}
 ik_foot_l {parent=ik_foot_root, loc=<15.2608 -4.4176 9.4432>, rot=<0.0357 0.7045 0.0284 -0.7082>}
 ik_foot_r {parent=ik_foot_root, loc=<-15.2608 -4.4176 9.4432>, rot=<0.7082 0.0284 -0.7045 0.0357>}
 root_fp {parent=root, loc=<0 0 0>, rot=<0 0 0 1>}
 attach_fp {parent=root_fp, loc=<0 0 0>, rot=<0 0 0 1>}
 camera_root {parent=root_fp, loc=<0 0 0>, rot=<0 0 0 1>}
 camera {parent=camera_root, loc=<0 0 0>, rot=<0 0 0 1>}
 attach_emote_01 {parent=root, loc=<0 0 0>, rot=<0 0 0 1>}
 attach_emote_02 {parent=root, loc=<0 0 0>, rot=<0 0 0 1>}
 attach_emote_03 {parent=root, loc=<0 0 0>, rot=<0 0 0 1>}
 warp_point_01 {parent=root, loc=<0 0 0>, rot=<0 0 0 1>}
 warp_point_02 {parent=root, loc=<0 0 0>, rot=<0 0 0 1>}
 warp_point_03 {parent=root, loc=<0 0 0>, rot=<0 0 0 1>}
]
variables [
 CustomOuterThumbs {
  type="TArray<FRigElementKeyCollection>"
  default="((Keys=((Type=Bone,Name=\"dyn_L_custom_extraThumb_1\"),(Type=Bone,Name=\"dyn_L_custom_extraThumb_2\"),(Type=Bone,Name=\"dyn_L_custom_extraThumb_3\"))),(Keys=((Type=Bone,Name=\"dyn_R_custom_extraThumb_1\"),(Type=Bone,Name=\"dyn_R_custom_extraThumb_2\"),(Type=Bone,Nam ...(truncated)"
 }
 CustomOuterThumbs Index {type="TArray<FRigElementKey>"}
 Dyn_is_crouch {type=double, default=0}
 IsRigidBodyOn {type=bool, default=False}
 OriginalThumbs {
  type="TArray<FRigElementKeyCollection>"
  default=((Keys=((Type=Bone,Name="thumb_01_l"),(Type=Bone,Name="thumb_02_l"),(Type=Bone,Name="thumb_03_l"))),(Keys=((Type=Bone,Name="thumb_01_r"),(Type=Bone,Name="thumb_02_r"),(Type=Bone,Name="thumb_03_r"))))
 }
 OriginalThumbs Index {type="TArray<FRigElementKey>"}
]
