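# Physics asset for the laces and skirt of the GasChin male medium body cosmetic, providing capsule bodies and joints for their secondary motion.
# The pelvis anchors four chains, left and right for skirt and laces, each a tiny deform root body followed by three dyn segments.
# Roots rotate freely against the pelvis, while the dyn joints carry per-joint swing and twist limits that differ slightly between sides.
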PhysicsAsset /BRCosmetics/Characters/Player/Male/Medium/Bodies/M_MED_MuffinLadle_GasChin/Meshes/M_MED_MuffinLadle_GasChin_Laces_Physics
bodies {
 deform_lace_l [{shapeType=Capsule, length=1, radius=1, shapeName=deform_lace_l_capsule, rotation=<rot 90 0 0>}]
 deform_lace_r [{shapeType=Capsule, length=1, radius=1, shapeName=deform_lace_r_capsule, rotation=<rot 90 0 0>}]
 deform_skirt_l [{shapeType=Capsule, length=1, radius=1, shapeName=deform_skirt_l_capsule, rotation=<rot 90 0 0>}]
 deform_skirt_r [{shapeType=Capsule, length=1, radius=1, shapeName=deform_skirt_r_capsule, rotation=<rot 90 0 0>}]
 dyn_lace_1_l [{shapeType=Capsule, length=8.5246, radius=2.6809, shapeName=dyn_lace_1_l_capsule, center=<4.3127 -1.9753 0>, rotation=<rot 90 180 175.0008>}]
 dyn_lace_1_r [{
   shapeType=Capsule
   length=8.5246
   radius=2.6809
   shapeName=dyn_lace_1_r_capsule
   center=<4.3127 -1.9753 0>
   rotation=<rot 90 179.4032 174.4041>
  }]
 dyn_lace_2_l [{shapeType=Capsule, length=5.369, radius=2.6809, shapeName=dyn_lace_2_l_capsule, center=<4.1242 -2.3436 0>, rotation=<rot 90 0 0>}]
 dyn_lace_2_r [{shapeType=Capsule, length=5.369, radius=2.6809, shapeName=dyn_lace_2_r_capsule, center=<4.1242 -2.3436 0>, rotation=<rot 90 0 0>}]
 dyn_lace_l [{
   shapeType=Capsule
   length=8.5246
   radius=2.6809
   shapeName=dyn_lace_l_capsule
   center=<3.6545 -3.0242 0>
   rotation=<rot 90 178.2101 -171.7897>
  }]
 dyn_lace_r [{
   shapeType=Capsule
   length=8.5246
   radius=2.6809
   shapeName=dyn_lace_r_capsule
   center=<3.6545 -3.0242 0>
   rotation=<rot 90 -159.444 -149.4437>
  }]
 dyn_skirt_down_1_l [{shapeType=Capsule, length=12.5224, radius=7.0722, shapeName=dyn_skirt_down_1_l_capsule, center=<0 4.723 0>}]
 dyn_skirt_down_1_r [{shapeType=Capsule, length=12.5224, radius=7.0722, shapeName=dyn_skirt_down_1_r_capsule, center=<0 4.723 0>}]
 dyn_skirt_down_l [{shapeType=Capsule, length=10.4669, radius=7.0722, shapeName=dyn_skirt_down_l_capsule, center=<5.8748 5.1975 0>}]
 dyn_skirt_down_r [{shapeType=Capsule, length=10.4669, radius=7.0722, shapeName=dyn_skirt_down_r_capsule, center=<5.8748 5.1975 0>}]
 dyn_skirt_l [{shapeType=Capsule, length=9.7663, radius=7.0722, shapeName=dyn_skirt_l_capsule, center=<5.8748 5.1277 0>}]
 dyn_skirt_r [{shapeType=Capsule, length=9.7663, radius=7.0722, shapeName=dyn_skirt_r_capsule, center=<5.8748 5.1277 0>}]
 pelvis [{shapeType=Capsule, length=8.1137, radius=10.559, shapeName=pelvis_capsule, center=<-0.3369 -0.6442 0>}]
}
constraints [
 {
  bone1Name=deform_skirt_l
  bone2Name=pelvis
  swing1LimitDegrees=45
  swing1Motion=Free
  swing2LimitDegrees=45
  swing2Motion=Free
  twistLimitDegrees=45
  twistMotion=Free
 }
 {bone1Name=dyn_skirt_l, bone2Name=deform_skirt_l, swing1LimitDegrees=45, swing2LimitDegrees=10, twistLimitDegrees=30}
 {bone1Name=dyn_skirt_down_l, bone2Name=dyn_skirt_l, swing1LimitDegrees=20, swing2LimitDegrees=20, twistLimitDegrees=30}
 {bone1Name=dyn_skirt_down_1_l, bone2Name=dyn_skirt_down_l, swing1LimitDegrees=20, swing2LimitDegrees=20, twistLimitDegrees=20}
 {
  bone1Name=deform_skirt_r
  bone2Name=pelvis
  swing1LimitDegrees=45
  swing1Motion=Free
  swing2LimitDegrees=45
  swing2Motion=Free
  twistLimitDegrees=45
  twistMotion=Free
 }
 {bone1Name=dyn_skirt_r, bone2Name=deform_skirt_r, swing1LimitDegrees=45, swing2LimitDegrees=10, twistLimitDegrees=5}
 {bone1Name=dyn_skirt_down_r, bone2Name=dyn_skirt_r, swing1LimitDegrees=20, swing2LimitDegrees=20, twistLimitDegrees=20}
 {bone1Name=dyn_skirt_down_1_r, bone2Name=dyn_skirt_down_r, swing1LimitDegrees=20, swing2LimitDegrees=20, twistLimitDegrees=20}
 {
  bone1Name=deform_lace_l
  bone2Name=pelvis
  swing1LimitDegrees=45
  swing1Motion=Free
  swing2LimitDegrees=45
  swing2Motion=Free
  twistLimitDegrees=45
  twistMotion=Free
 }
 {bone1Name=dyn_lace_l, bone2Name=deform_lace_l, swing1LimitDegrees=45, swing2LimitDegrees=7.5, twistLimitDegrees=5}
 {bone1Name=dyn_lace_1_l, bone2Name=dyn_lace_l, swing1LimitDegrees=30, swing2LimitDegrees=20, twistLimitDegrees=1}
 {bone1Name=dyn_lace_2_l, bone2Name=dyn_lace_1_l, swing1LimitDegrees=25, swing2LimitDegrees=15, twistLimitDegrees=5}
 {
  bone1Name=deform_lace_r
  bone2Name=pelvis
  swing1LimitDegrees=45
  swing1Motion=Free
  swing2LimitDegrees=45
  swing2Motion=Free
  twistLimitDegrees=45
  twistMotion=Free
 }
 {bone1Name=dyn_lace_r, bone2Name=deform_lace_r, swing1LimitDegrees=45, swing2LimitDegrees=10, twistLimitDegrees=5}
 {bone1Name=dyn_lace_1_r, bone2Name=dyn_lace_r, swing1LimitDegrees=30, swing2LimitDegrees=20, twistLimitDegrees=1}
 {bone1Name=dyn_lace_2_r, bone2Name=dyn_lace_1_r, swing1LimitDegrees=25, swing2LimitDegrees=15, twistLimitDegrees=5}
]
