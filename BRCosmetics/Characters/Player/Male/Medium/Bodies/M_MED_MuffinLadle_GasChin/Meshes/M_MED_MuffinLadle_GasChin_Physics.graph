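# Physics asset for the M_MED_MuffinLadle_GasChin male player cosmetic body, set up for skirt and belt secondary motion.
# Capsules cover pelvis, thighs, custom calves and the dynamic chains.
# Four skirt chains use angular limits below free-swinging deform roots on the pelvis, with the lower left and right non-back segments cross-linked.
# Front and back belt chains hang from deform anchors with tight limits, their free-rotating tips constrained together.
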PhysicsAsset /BRCosmetics/Characters/Player/Male/Medium/Bodies/M_MED_MuffinLadle_GasChin/Meshes/M_MED_MuffinLadle_GasChin_Physics
bodies {
 deform_belt_b [{shapeType=Capsule, length=0.505, radius=0.505, shapeName=deform_belt_b_capsule, rotation=<rot 0 -90 0>}]
 deform_belt_f [{shapeType=Capsule, length=0.505, radius=0.505, shapeName=deform_belt_f_capsule, rotation=<rot 0 -90 0>}]
 deform_skirt_b_l [{shapeType=Capsule, length=0.505, radius=0.505, shapeName=deform_skirt_b_l_capsule, rotation=<rot 0 -90 0>}]
 deform_skirt_b_r [{shapeType=Capsule, length=0.505, radius=0.505, shapeName=deform_skirt_b_r_capsule, rotation=<rot 0 -90 0>}]
 deform_skirt_l [{shapeType=Capsule, length=1, radius=1, shapeName=deform_skirt_l_capsule, rotation=<rot 90 0 0>}]
 deform_skirt_r [{shapeType=Capsule, length=1, radius=1, shapeName=deform_skirt_r_capsule, rotation=<rot 90 0 0>}]
 dyn_L_custom_calf [{shapeType=Capsule, length=18.1268, radius=8.9604, shapeName=dyn_L_custom_calf_capsule, center=<-7.3334 0 0>, rotation=<rot 90 0 0>}]
 dyn_R_custom_calf [{shapeType=Capsule, length=18.1268, radius=8.9604, shapeName=dyn_R_custom_calf_capsule, center=<7.3334 0 0>, rotation=<rot 90 0 180>}]
 dyn_belt_1_b [{
   shapeType=Capsule
   length=12.6418
   radius=1.24
   shapeName=dyn_belt_1_b_capsule
   center=<7.9621 -0.4777 0.6548>
   rotation=<rot 0 92.8961 -89.689>
  }]
 dyn_belt_1_f [{
   shapeType=Capsule
   length=14.3465
   radius=1.0689
   shapeName=dyn_belt_1_f_capsule
   center=<7.0266 -2.3707 -1.5721>
   rotation=<rot 2.512 -90 -90.834>
  }]
 dyn_belt_2_b [{shapeType=Capsule, length=0.505, radius=0.505, shapeName=dyn_belt_2_b_capsule, rotation=<rot 0 -90 0>}]
 dyn_belt_2_f [{shapeType=Capsule, length=0.505, radius=0.505, shapeName=dyn_belt_2_f_capsule, rotation=<rot 0 -90 0>}]
 dyn_belt_b [{
   shapeType=Capsule
   length=7.7525
   radius=1.3161
   shapeName=dyn_belt_b_capsule
   center=<6.8927 -0.5338 -0.119>
   rotation=<rot 0 97.6375 -89.0033>
  }]
 dyn_belt_f [{
   shapeType=Capsule
   length=5.6589
   radius=1.3205
   shapeName=dyn_belt_f_capsule
   center=<6.7294 -0.3284 0.067>
   rotation=<rot 0 90.753 -83.9367>
  }]
 dyn_skirt_b_1_l [{
   shapeType=Capsule
   length=5.5686
   radius=3.0862
   shapeName=dyn_skirt_b_1_l_capsule
   center=<6.6623 -0.8164 0.246>
   rotation=<rot 3.0682 -89.9989 178.7751>
  }]
 dyn_skirt_b_1_r [{
   shapeType=Capsule
   length=6.1757
   radius=3.1061
   shapeName=dyn_skirt_b_1_r_capsule
   center=<6.6383 -0.6444 -1.9331>
   rotation=<rot 0 -88.4679 2.5748>
  }]
 dyn_skirt_b_2_l [{
   shapeType=Capsule
   length=7.6439
   radius=3.532
   shapeName=dyn_skirt_b_2_l_capsule
   center=<6.6433 0.261 0.5854>
   rotation=<rot 2.9943 -89.9997 -0.3377>
  }]
 dyn_skirt_b_2_r [{
   shapeType=Capsule
   length=6.9729
   radius=3.425
   shapeName=dyn_skirt_b_2_r_capsule
   center=<7.1222 -0.2407 -0.1089>
   rotation=<rot 0 85.3516 3.5082>
  }]
 dyn_skirt_b_l [{shapeType=Capsule, length=6, radius=4, shapeName=dyn_skirt_b_l_capsule, center=<6.2163 -2.0985 0>, rotation=<rot 90 0 0>}]
 dyn_skirt_b_r [{shapeType=Capsule, length=6, radius=4, shapeName=dyn_skirt_b_r_capsule, center=<6.2163 -2.0985 0>, rotation=<rot 90 0 0>}]
 dyn_skirt_down_1_l [{shapeType=Capsule, length=13.0291, radius=3.5, shapeName=dyn_skirt_down_1_l_capsule, center=<3.3866 -0.7244 0>, rotation=<rot 90 0 0>}]
 dyn_skirt_down_1_r [{shapeType=Capsule, length=12.5291, radius=4, shapeName=dyn_skirt_down_1_r_capsule, center=<3.3866 -0.7244 0.4466>, rotation=<rot 90 0 0>}]
 dyn_skirt_down_l [{
   shapeType=Capsule
   length=11.8834
   radius=2.5
   shapeName=dyn_skirt_down_l_capsule
   center=<6.5601 -0.738 2.2115>
   rotation=<rot 0 -0.0023 -0.0001>
  }]
 dyn_skirt_down_r [{
   shapeType=Capsule
   length=11.3834
   radius=3
   shapeName=dyn_skirt_down_r_capsule
   center=<7.1508 -0.7381 -2.8893>
   rotation=<rot 10 -0.0015 0.0001>
  }]
 dyn_skirt_l [{shapeType=Capsule, length=11, radius=2.5, shapeName=dyn_skirt_l_capsule, center=<7.7329 -0.2104 2.1998>, rotation=<rot 0 -0.0012 -0.0001>}]
 dyn_skirt_r [{shapeType=Capsule, length=10.5, radius=3, shapeName=dyn_skirt_r_capsule, center=<7.7329 -0.2104 -2.2723>, rotation=<rot -5 -0.0012 0.0001>}]
 pelvis [{shapeType=Capsule, length=8.1137, radius=10.559, shapeName=pelvis_capsule, center=<-0.3369 -0.6442 0>}]
 thigh_l [{shapeType=Capsule, length=30.6459, radius=9, shapeName=thigh_l_capsule, center=<-17.7098 2.6337 0>, rotation=<rot 90 89.6476 98.8129>}]
 thigh_r [{
   shapeType=Capsule
   length=31.1459
   radius=8.5
   shapeName=thigh_r_capsule
   center=<17.6965 -2.6316 -0.8505>
   rotation=<rot 85 170.8343 4.9998>
  }]
}
constraints [
 {
  bone1Name=deform_skirt_l
  bone2Name=pelvis
  swing1LimitDegrees=45
  swing1Motion=Free
  swing2LimitDegrees=45
  swing2Motion=Free
  twistLimitDegrees=45
  twistMotion=Free
 }
 {bone1Name=dyn_skirt_l, bone2Name=deform_skirt_l, swing1LimitDegrees=45, swing2LimitDegrees=10, twistLimitDegrees=30}
 {bone1Name=dyn_skirt_down_l, bone2Name=dyn_skirt_l, swing1LimitDegrees=20, swing2LimitDegrees=20, twistLimitDegrees=30}
 {bone1Name=dyn_skirt_down_1_l, bone2Name=dyn_skirt_down_l, swing1LimitDegrees=20, swing2LimitDegrees=20, twistLimitDegrees=20}
 {
  bone1Name=deform_skirt_r
  bone2Name=pelvis
  swing1LimitDegrees=45
  swing1Motion=Free
  swing2LimitDegrees=45
  swing2Motion=Free
  twistLimitDegrees=45
  twistMotion=Free
 }
 {bone1Name=dyn_skirt_r, bone2Name=deform_skirt_r, swing1LimitDegrees=45, swing2LimitDegrees=10, twistLimitDegrees=5}
 {bone1Name=dyn_skirt_down_r, bone2Name=dyn_skirt_r, swing1LimitDegrees=20, swing2LimitDegrees=20, twistLimitDegrees=20}
 {bone1Name=dyn_skirt_down_1_r, bone2Name=dyn_skirt_down_r, swing1LimitDegrees=20, swing2LimitDegrees=20, twistLimitDegrees=20}
 {bone1Name=dyn_skirt_down_l, bone2Name=dyn_skirt_down_r, swing1LimitDegrees=35, swing2LimitDegrees=20, twistLimitDegrees=25}
 {bone1Name=dyn_skirt_down_1_l, bone2Name=dyn_skirt_down_1_r, swing1LimitDegrees=25, swing2LimitDegrees=15, twistLimitDegrees=25}
 {bone1Name=dyn_skirt_b_l, bone2Name=deform_skirt_b_l, swing1LimitDegrees=30, swing2LimitDegrees=10, twistLimitDegrees=20}
 {
  bone1Name=deform_skirt_b_l
  bone2Name=pelvis
  swing1LimitDegrees=45
  swing1Motion=Free
  swing2LimitDegrees=45
  swing2Motion=Free
  twistLimitDegrees=45
  twistMotion=Free
 }
 {
  bone1Name=deform_skirt_b_r
  bone2Name=pelvis
  swing1LimitDegrees=45
  swing1Motion=Free
  swing2LimitDegrees=45
  swing2Motion=Free
  twistLimitDegrees=45
  twistMotion=Free
 }
 {bone1Name=dyn_skirt_b_1_l, bone2Name=dyn_skirt_b_l, swing1LimitDegrees=25, swing2LimitDegrees=20, twistLimitDegrees=20}
 {bone1Name=dyn_skirt_b_2_l, bone2Name=dyn_skirt_b_1_l, swing1LimitDegrees=25, swing2LimitDegrees=20, twistLimitDegrees=20}
 {bone1Name=dyn_skirt_b_1_r, bone2Name=dyn_skirt_b_r, swing1LimitDegrees=25, swing2LimitDegrees=20, twistLimitDegrees=20}
 {bone1Name=dyn_skirt_b_2_r, bone2Name=dyn_skirt_b_1_r, swing1LimitDegrees=25, swing2LimitDegrees=20, twistLimitDegrees=20}
 {bone1Name=dyn_skirt_b_r, bone2Name=deform_skirt_b_r, swing1LimitDegrees=30, swing2LimitDegrees=10, twistLimitDegrees=20}
 {bone1Name=dyn_belt_f, bone2Name=deform_belt_f, swing1LimitDegrees=10, swing2LimitDegrees=10, twistLimitDegrees=5}
 {bone1Name=dyn_belt_1_f, bone2Name=dyn_belt_f, swing1LimitDegrees=10, swing2LimitDegrees=10, twistLimitDegrees=5}
 {bone1Name=dyn_belt_b, bone2Name=deform_belt_b, swing1LimitDegrees=10, swing2LimitDegrees=10, twistLimitDegrees=5}
 {bone1Name=dyn_belt_1_b, bone2Name=dyn_belt_b, swing1LimitDegrees=10, swing2LimitDegrees=10, twistLimitDegrees=5}
 {
  bone1Name=dyn_belt_2_b
  bone2Name=dyn_belt_1_b
  swing1LimitDegrees=45
  swing1Motion=Free
  swing2LimitDegrees=45
  swing2Motion=Free
  twistLimitDegrees=45
  twistMotion=Free
 }
 {
  bone1Name=dyn_belt_2_f
  bone2Name=dyn_belt_1_f
  swing1LimitDegrees=45
  swing1Motion=Free
  swing2LimitDegrees=45
  swing2Motion=Free
  twistLimitDegrees=45
  twistMotion=Free
 }
 {
  bone1Name=dyn_belt_2_b
  bone2Name=dyn_belt_2_f
  swing1LimitDegrees=45
  swing1Motion=Free
  swing2LimitDegrees=45
  swing2Motion=Free
  twistLimitDegrees=45
  twistMotion=Free
 }
]
