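# Animation Blueprint for the MuffinLadle GasChin player body, adding secondary motion to the pose copied from the parent mesh.
# A crouch-aware Control Rig feeds a cached pose into two chained RigidBody simulations, for body and laces, with state-driven gravity override.
# Layered blends restore shoulder plates from the cached pose and apply curve-weighted lobby-idle and emote slots to custom arm bones.
# Two lace constraint nodes are disconnected from the output.
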AnimBlueprint /BRCosmetics/Characters/Player/Male/Medium/Bodies/M_MED_MuffinLadle_GasChin/Meshes/M_MED_MuffinLadle_GasChin_AnimBP
target_skeleton=&/M_MED_MuffinLadle_GasChin_Skeleton
graphs [
 AnimGraph {nodes [
   n1 = "Constraint - dyn_lace_1_l (T)"(Alpha=0.25, ConstraintWeights_0=1) {
    alphaBoolBlend {blendOption=Linear}
    boneToModify {boneName=dyn_lace_1_l}
    constraintSetup [{transformType=Translation, targetBone {boneName=dyn_skirt_down_l}}]
    constraintWeights [1]
   }
   "Constraint - dyn_lace_1_r (T)"(ComponentPose=@n1, Alpha=0.25, ConstraintWeights_0=1) {
    alphaBoolBlend {blendOption=Linear}
    boneToModify {boneName=dyn_lace_1_r}
    constraintSetup [{transformType=Translation, targetBone {boneName=dyn_skirt_down_r}}]
    constraintWeights [1]
   }
   n2 = Get Gravity Override()
   n3 = RigidBody!(ComponentPose=Local To Component(LocalPose=Use cached pose 'BasePose'()), Alpha, OverrideWorldGravity:Vector=@n2, ExternalForce) -> (Pose) {
    bClampLinearTranslationLimitToRefPose=true
    bOverrideWorldGravity=true
    overridePhysicsAsset=&/M_MED_MuffinLadle_GasChin_Physics
    simulationSpace=BaseBoneSpace
    alphaBoolBlend {blendInTime=1.7, blendOption=Linear, blendOutTime=1.7}
    alphaScaleBiasClamp {bInterpResult=true, interpSpeedDecreasing=5, interpSpeedIncreasing=5}
    baseBoneRef {boneName=pelvis}
    overrideWorldGravity {z=-980}
    simSpaceSettings {
     dampingAlpha=0.75
     maxAngularAcceleration=35
     maxAngularVelocity=5
     maxLinearAcceleration=1550
     maxLinearVelocity=350
     velocityScaleZ=0.6
     worldAlpha=0.75
     externalLinearDragV=<2 2 2>
    }
   }
   n4 = RigidBody(ComponentPose=@n3, OverrideWorldGravity=@n2) {
    bClampLinearTranslationLimitToRefPose=true
    bOverrideWorldGravity=true
    overridePhysicsAsset=&/M_MED_MuffinLadle_GasChin_Laces_Physics
    simulationSpace=BaseBoneSpace
    alphaBoolBlend {blendInTime=1.7, blendOption=Linear, blendOutTime=1.7}
    alphaScaleBiasClamp {bInterpResult=true, interpSpeedDecreasing=5, interpSpeedIncreasing=5}
    baseBoneRef {boneName=pelvis}
    overrideWorldGravity {z=-980}
    simSpaceSettings {
     dampingAlpha=0.75
     maxAngularAcceleration=35
     maxAngularVelocity=5
     maxLinearAcceleration=1550
     maxLinearVelocity=350
     velocityScaleZ=0.6
     worldAlpha=0.75
     externalLinearDragV=<2 2 2>
    }
   }
   n5 = Layered blend per bone(
    BasePose=Component To Local(ComponentPose=@n4)
    BlendPoses_0=Slot 'DefaultSlot'()
    BlendWeights_0=Get FE Custom Lobby Idle Is Playing() -> (FE Custom Lobby Idle Is Playing:float)
   ) {layerSetup [{branchFilters [
       {boneName=dyn_L_customBackArm_clavicle}
       {boneName=dyn_R_customBackArm_clavicle}
       {boneName=dyn_L_custom_extraThumb_1}
       {boneName=dyn_L_customFrontArm_clavicle}
       {boneName=dyn_R_customFrontArm_clavicle}
      ]}]}
   n6 = Layered blend per bone(BasePose=@n5, BlendPoses_0=Use cached pose 'BasePose'()) {layerSetup [{branchFilters [{boneName=dyn_L_shoulderPlate_main}, {boneName=dyn_R_shoulderPlate_main}]}]}
   n7 = Layered blend per bone(
    BasePose=@n6
    BlendPoses_0=Slot 'EmoteSlot'() {slotName=EmoteSlot}
    BlendWeights_0=Get is_Playing_Custom_Emote() -> (is_Playing_Custom_Emote:float)
   ) {layerSetup [{branchFilters [
       {boneName=dyn_L_customBackArm_clavicle}
       {boneName=dyn_L_customFrontArm_clavicle}
       {boneName=dyn_R_customBackArm_clavicle}
       {boneName=dyn_R_customFrontArm_clavicle}
      ]}]}
   Output Pose(Result=@n7) {layerGroup=None}
   n8 = Control Rig!(
    Source=Copy Pose From Mesh() {bCopyCurves=true, bUseAttachedParent=true}
    Alpha
    Dyn_is_crouch:double=Get Dyn Is Crouch() -> (Dyn Is Crouch:double)
    IsRigidBodyOn:bool=Get bRigidBodyEnabled()
   ) -> (Pose) {
    alphaBoolBlend {blendOption=Linear}
    controlRigAssetReference {blueprintRigClass=&/M_MED_MuffinLadle_GasChin_CtrlRig.M_MED_MuffinLadle_GasChin_CtrlRig_C}
   }
   Save cached pose 'BasePose'(Pose=@n8) {cacheName=BasePose}
  ]}
 EventGraph {nodes [
   Event BlueprintUpdateAnimation()
   > n1 = Sequence!(execute) -> (then_0, then_1, then_2:Exec, then_3:Exec, then_4:Exec)
   > n2 = Gravity Override State Function DEPRECATED(CharacterPartAnimInstance=Self-Reference(), GameTimeInSeconds=GetGameTimeInSeconds(), GravityOverrideParams=Get Gravity Override Params())
   > Set Gravity Override(Gravity Override=@n2)
   n3 = Branch(execute=@n1.then_1, Condition=Get bIsCrouching())
   > Set Dyn Is Crouch(Dyn Is Crouch:double=1) -> (Output_Get:double)
   Set Dyn Is Crouch(execute=@n3.else, Dyn Is Crouch=0)
   n4 = Branch(execute=@n1.then_2, Condition=Get bIsPlayingEmote())
   > Set is_Playing_Custom_Emote(is_Playing_Custom_Emote:float=GetCurveValue(CurveName=emote_BuiltIn)) -> (Output_Get:float)
   Set is_Playing_Custom_Emote(execute=@n4.else, is_Playing_Custom_Emote=0)
   Branch(execute=@n1.then_3, Condition=Get bIsInFrontEnd())
   > Set FE Custom Lobby Idle Is Playing(FE Custom Lobby Idle Is Playing:float=GetCurveValue(CurveName=Frontend_Idle_Bypass)) -> (Output_Get:float)
   GetOwnerFortPlayerAnimInstance!(self:Custom Character Part Anim Instance Object Reference) -> (ReturnValue:Fort Player Anim Instance Object Reference)
  ]}
]
variables [
 Dyn Is Crouch {type=number}
 FE Custom Lobby Idle Is Playing {type=number}
 Gravity Override {type=Vector}
 Gravity Override Params {
  type=GravityOverrideParamsStruct
  default {
   dBNO {gravityOverride=<-600 980 0>}
   skydive_FloatingDown {gravityOverride=<0 -450 -800>, windFrequency=<1.5 7 10>, windOffset=<0 -450 200>}
   skydive_GoingUp {gravityOverride=<0 -200 -600>}
  }
 }
 is_Playing_Custom_Emote {type=number}
]
