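# Physics asset for the MuffinLadle GasChin medium male player body, named as the collider set for its cape.
# Single capsules cover the spine, pelvis, head, clavicles, IK leg chains, four custom front and back arms, rear skirt chains and rear belt.
# Constraints link them into a tree through the spine, with all swing and twist motion free, so the joints impose no angular limits.
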PhysicsAsset /BRCosmetics/Characters/Player/Male/Medium/Bodies/M_MED_MuffinLadle_GasChin/Meshes/M_MED_MuffinLadle_GasChin_CapeColliders
bodies {
 clavicle_l [{shapeType=Capsule, length=2.7387, radius=6.595, shapeName=clavicle_l_capsule, center=<5 0 0>, rotation=<rot 0 90 90>}]
 clavicle_r [{shapeType=Capsule, length=1.6435, radius=7.6902, shapeName=clavicle_r_capsule, center=<-5 0 0>, rotation=<rot 0 90 90>}]
 dyn_L_customBackArm_lowerarm [{shapeType=Capsule, length=28.8427, radius=5.1573, shapeName=dyn_L_customBackArm_lowerarm_capsule, center=<14 0 0>, rotation=<rot 0 90 90>}]
 dyn_L_customBackArm_upperarm [{shapeType=Capsule, length=26.8427, radius=2.1573, shapeName=dyn_L_customBackArm_upperarm_capsule, center=<13 0 0>, rotation=<rot 0 90 90>}]
 dyn_L_customBackArm_wrist [{shapeType=Capsule, length=5.8427, radius=6.1573, shapeName=dyn_L_customBackArm_wrist_capsule, center=<10 0 0>, rotation=<rot 0 90 90>}]
 dyn_L_customFrontArm_lowerarm [{shapeType=Capsule, length=28.8427, radius=5.1573, shapeName=dyn_L_customFrontArm_lowerarm_capsule, center=<14 0 0>, rotation=<rot 0 90 90>}]
 dyn_L_customFrontArm_upperarm [{shapeType=Capsule, length=26.8427, radius=2.1573, shapeName=dyn_L_customFrontArm_upperarm_capsule, center=<13 0 0>, rotation=<rot 0 90 90>}]
 dyn_L_customFrontArm_wrist [{shapeType=Capsule, length=5.8427, radius=6.1573, shapeName=dyn_L_customFrontArm_wrist_capsule, center=<10 0 0>, rotation=<rot 0 90 90>}]
 dyn_L_custom_calf_1_IK [{shapeType=Capsule, length=20, radius=5, shapeName=dyn_L_custom_calf_1_IK_capsule, center=<-10 0 0>, rotation=<rot 0 90 90>}]
 dyn_L_custom_calf_IK [{shapeType=Capsule, length=16, radius=8, shapeName=dyn_L_custom_calf_IK_capsule, center=<-8 0 0>, rotation=<rot 0 90 90>}]
 dyn_L_custom_foot_IK [{shapeType=Capsule, length=14, radius=9, shapeName=dyn_L_custom_foot_IK_capsule, center=<-7 5 0>, rotation=<rot 0 0 90>}]
 dyn_R_customBackArm_lowerarm [{shapeType=Capsule, length=28.8427, radius=5.1573, shapeName=dyn_R_customBackArm_lowerarm_capsule, center=<-14 0 0>, rotation=<rot 0 90 90>}]
 dyn_R_customBackArm_upperarm [{shapeType=Capsule, length=26.8427, radius=2.1573, shapeName=dyn_R_customBackArm_upperarm_capsule, center=<-13 0 0>, rotation=<rot 0 90 90>}]
 dyn_R_customBackArm_wrist [{shapeType=Capsule, length=5.8427, radius=6.1573, shapeName=dyn_R_customBackArm_wrist_capsule, center=<-10 0 0>, rotation=<rot 0 90 90>}]
 dyn_R_customFrontArm_lowerarm [{shapeType=Capsule, length=28.8427, radius=5.1573, shapeName=dyn_R_customFrontArm_lowerarm_capsule, center=<-14 0 0>, rotation=<rot 0 90 90>}]
 dyn_R_customFrontArm_upperarm [{shapeType=Capsule, length=26.8427, radius=2.1573, shapeName=dyn_R_customFrontArm_upperarm_capsule, center=<-13 0 0>, rotation=<rot 0 90 90>}]
 dyn_R_customFrontArm_wrist [{shapeType=Capsule, length=5.8427, radius=6.1573, shapeName=dyn_R_customFrontArm_wrist_capsule, center=<-10 0 0>, rotation=<rot 0 90 90>}]
 dyn_R_custom_calf_1_IK [{shapeType=Capsule, length=20, radius=5, shapeName=dyn_R_custom_calf_1_IK_capsule, center=<10 0 0>, rotation=<rot 0 90 90>}]
 dyn_R_custom_calf_IK [{shapeType=Capsule, length=16, radius=8, shapeName=dyn_R_custom_calf_IK_capsule, center=<8 0 0>, rotation=<rot 0 90 90>}]
 dyn_R_custom_foot_IK [{shapeType=Capsule, length=14, radius=9, shapeName=dyn_R_custom_foot_IK_capsule, center=<7 -5 0>, rotation=<rot 0 0 90>}]
 dyn_belt_1_b [{shapeType=Capsule, length=10, radius=2, shapeName=dyn_belt_1_b_capsule, center=<5 0 0>, rotation=<rot 90 0 0>}]
 dyn_belt_b [{shapeType=Capsule, length=14, radius=2, shapeName=dyn_belt_b_capsule, center=<7 0 0>, rotation=<rot 90 0 0>}]
 dyn_skirt_b_1_l [{shapeType=Capsule, length=6, radius=6, shapeName=dyn_skirt_b_1_l_capsule, center=<5 4 0>, rotation=<rot 90 0 0>}]
 dyn_skirt_b_1_r [{shapeType=Capsule, length=6, radius=6, shapeName=dyn_skirt_b_1_r_capsule, center=<5 4 0>, rotation=<rot 90 0 0>}]
 dyn_skirt_b_2_l [{shapeType=Capsule, length=6, radius=7, shapeName=dyn_skirt_b_2_l_capsule, center=<5 4 0>, rotation=<rot 90 0 0>}]
 dyn_skirt_b_2_r [{shapeType=Capsule, length=6, radius=7, shapeName=dyn_skirt_b_2_r_capsule, center=<5 4 0>, rotation=<rot 90 0 0>}]
 dyn_skirt_b_l [{shapeType=Capsule, length=6, radius=6, shapeName=dyn_skirt_b_l_capsule, center=<5 4 0>, rotation=<rot 90 0 0>}]
 dyn_skirt_b_r [{shapeType=Capsule, length=6, radius=6, shapeName=dyn_skirt_b_r_capsule, center=<5 4 0>, rotation=<rot 90 0 0>}]
 head [{shapeType=Capsule, length=0, radius=12, shapeName=head_capsule, rotation=<rot 0 -90 0>}]
 pelvis [{shapeType=Capsule, length=7.817, radius=11.3687, shapeName=pelvis_capsule}]
 spine_01 [{shapeType=Capsule, length=10.9173, radius=11.1246, shapeName=spine_01_capsule}]
 spine_02 [{shapeType=Capsule, length=6.4956, radius=9.7481, shapeName=spine_02_capsule}]
 spine_03 [{shapeType=Capsule, length=7.0441, radius=9.1119, shapeName=spine_03_capsule}]
 spine_04 [{shapeType=Capsule, length=9.3221, radius=11.2313, shapeName=spine_04_capsule}]
 spine_05 [{shapeType=Capsule, length=20.2214, radius=9.711, shapeName=spine_05_capsule, center=<-0.1698 -0.5499 0>}]
 thigh_l_IK [{shapeType=Capsule, length=25, radius=9, shapeName=thigh_l_IK_capsule, center=<-15 0 0>, rotation=<rot 0 90 90>}]
 thigh_r_IK [{shapeType=Capsule, length=25, radius=9, shapeName=thigh_r_IK_capsule, center=<15 0 0>, rotation=<rot 0 90 90>}]
}
constraints [
 {
  bone1Name=spine_05
  bone2Name=spine_04
  swing1LimitDegrees=45
  swing1Motion=Free
  swing2LimitDegrees=45
  swing2Motion=Free
  twistLimitDegrees=45
  twistMotion=Free
 }
 {
  bone1Name=spine_04
  bone2Name=spine_03
  swing1LimitDegrees=45
  swing1Motion=Free
  swing2LimitDegrees=45
  swing2Motion=Free
  twistLimitDegrees=45
  twistMotion=Free
 }
 {
  bone1Name=spine_03
  bone2Name=spine_02
  swing1LimitDegrees=45
  swing1Motion=Free
  swing2LimitDegrees=45
  swing2Motion=Free
  twistLimitDegrees=45
  twistMotion=Free
 }
 {
  bone1Name=spine_02
  bone2Name=spine_01
  swing1LimitDegrees=45
  swing1Motion=Free
  swing2LimitDegrees=45
  swing2Motion=Free
  twistLimitDegrees=45
  twistMotion=Free
 }
 {
  bone1Name=spine_01
  bone2Name=pelvis
  swing1LimitDegrees=45
  swing1Motion=Free
  swing2LimitDegrees=45
  swing2Motion=Free
  twistLimitDegrees=45
  twistMotion=Free
 }
 {
  bone1Name=clavicle_l
  bone2Name=spine_05
  swing1LimitDegrees=45
  swing1Motion=Free
  swing2LimitDegrees=45
  swing2Motion=Free
  twistLimitDegrees=45
  twistMotion=Free
 }
 {
  bone1Name=clavicle_r
  bone2Name=spine_05
  swing1LimitDegrees=45
  swing1Motion=Free
  swing2LimitDegrees=45
  swing2Motion=Free
  twistLimitDegrees=45
  twistMotion=Free
 }
 {
  bone1Name=head
  bone2Name=spine_05
  swing1LimitDegrees=45
  swing1Motion=Free
  swing2LimitDegrees=45
  swing2Motion=Free
  twistLimitDegrees=45
  twistMotion=Free
 }
 {
  bone1Name=dyn_L_customBackArm_upperarm
  bone2Name=spine_05
  swing1LimitDegrees=45
  swing1Motion=Free
  swing2LimitDegrees=45
  swing2Motion=Free
  twistLimitDegrees=45
  twistMotion=Free
 }
 {
  bone1Name=dyn_L_customFrontArm_upperarm
  bone2Name=spine_05
  swing1LimitDegrees=45
  swing1Motion=Free
  swing2LimitDegrees=45
  swing2Motion=Free
  twistLimitDegrees=45
  twistMotion=Free
 }
 {
  bone1Name=dyn_R_customBackArm_upperarm
  bone2Name=spine_05
  swing1LimitDegrees=45
  swing1Motion=Free
  swing2LimitDegrees=45
  swing2Motion=Free
  twistLimitDegrees=45
  twistMotion=Free
 }
 {
  bone1Name=dyn_R_customFrontArm_upperarm
  bone2Name=spine_05
  swing1LimitDegrees=45
  swing1Motion=Free
  swing2LimitDegrees=45
  swing2Motion=Free
  twistLimitDegrees=45
  twistMotion=Free
 }
 {
  bone1Name=dyn_L_customBackArm_lowerarm
  bone2Name=dyn_L_customBackArm_upperarm
  swing1LimitDegrees=45
  swing1Motion=Free
  swing2LimitDegrees=45
  swing2Motion=Free
  twistLimitDegrees=45
  twistMotion=Free
 }
 {
  bone1Name=dyn_L_customFrontArm_lowerarm
  bone2Name=dyn_L_customFrontArm_upperarm
  swing1LimitDegrees=45
  swing1Motion=Free
  swing2LimitDegrees=45
  swing2Motion=Free
  twistLimitDegrees=45
  twistMotion=Free
 }
 {
  bone1Name=dyn_R_customBackArm_lowerarm
  bone2Name=dyn_R_customBackArm_upperarm
  swing1LimitDegrees=45
  swing1Motion=Free
  swing2LimitDegrees=45
  swing2Motion=Free
  twistLimitDegrees=45
  twistMotion=Free
 }
 {
  bone1Name=dyn_L_customFrontArm_wrist
  bone2Name=dyn_L_customFrontArm_lowerarm
  swing1LimitDegrees=45
  swing1Motion=Free
  swing2LimitDegrees=45
  swing2Motion=Free
  twistLimitDegrees=45
  twistMotion=Free
 }
 {
  bone1Name=dyn_L_customBackArm_wrist
  bone2Name=dyn_L_customBackArm_lowerarm
  swing1LimitDegrees=45
  swing1Motion=Free
  swing2LimitDegrees=45
  swing2Motion=Free
  twistLimitDegrees=45
  twistMotion=Free
 }
 {
  bone1Name=dyn_R_customBackArm_wrist
  bone2Name=dyn_R_customBackArm_lowerarm
  swing1LimitDegrees=45
  swing1Motion=Free
  swing2LimitDegrees=45
  swing2Motion=Free
  twistLimitDegrees=45
  twistMotion=Free
 }
 {
  bone1Name=dyn_R_customFrontArm_lowerarm
  bone2Name=dyn_R_customFrontArm_upperarm
  swing1LimitDegrees=45
  swing1Motion=Free
  swing2LimitDegrees=45
  swing2Motion=Free
  twistLimitDegrees=45
  twistMotion=Free
 }
 {
  bone1Name=dyn_R_customFrontArm_wrist
  bone2Name=dyn_R_customFrontArm_lowerarm
  swing1LimitDegrees=45
  swing1Motion=Free
  swing2LimitDegrees=45
  swing2Motion=Free
  twistLimitDegrees=45
  twistMotion=Free
 }
 {
  bone1Name=thigh_l_IK
  bone2Name=pelvis
  swing1LimitDegrees=45
  swing1Motion=Free
  swing2LimitDegrees=45
  swing2Motion=Free
  twistLimitDegrees=45
  twistMotion=Free
 }
 {
  bone1Name=dyn_L_custom_calf_IK
  bone2Name=thigh_l_IK
  swing1LimitDegrees=45
  swing1Motion=Free
  swing2LimitDegrees=45
  swing2Motion=Free
  twistLimitDegrees=45
  twistMotion=Free
 }
 {
  bone1Name=dyn_L_custom_calf_1_IK
  bone2Name=dyn_L_custom_calf_IK
  swing1LimitDegrees=45
  swing1Motion=Free
  swing2LimitDegrees=45
  swing2Motion=Free
  twistLimitDegrees=45
  twistMotion=Free
 }
 {
  bone1Name=dyn_L_custom_foot_IK
  bone2Name=dyn_L_custom_calf_1_IK
  swing1LimitDegrees=45
  swing1Motion=Free
  swing2LimitDegrees=45
  swing2Motion=Free
  twistLimitDegrees=45
  twistMotion=Free
 }
 {
  bone1Name=thigh_r_IK
  bone2Name=pelvis
  swing1LimitDegrees=45
  swing1Motion=Free
  swing2LimitDegrees=45
  swing2Motion=Free
  twistLimitDegrees=45
  twistMotion=Free
 }
 {
  bone1Name=dyn_R_custom_calf_IK
  bone2Name=thigh_r_IK
  swing1LimitDegrees=45
  swing1Motion=Free
  swing2LimitDegrees=45
  swing2Motion=Free
  twistLimitDegrees=45
  twistMotion=Free
 }
 {
  bone1Name=dyn_R_custom_calf_1_IK
  bone2Name=dyn_R_custom_calf_IK
  swing1LimitDegrees=45
  swing1Motion=Free
  swing2LimitDegrees=45
  swing2Motion=Free
  twistLimitDegrees=45
  twistMotion=Free
 }
 {
  bone1Name=dyn_R_custom_foot_IK
  bone2Name=dyn_R_custom_calf_1_IK
  swing1LimitDegrees=45
  swing1Motion=Free
  swing2LimitDegrees=45
  swing2Motion=Free
  twistLimitDegrees=45
  twistMotion=Free
 }
 {
  bone1Name=dyn_skirt_b_l
  bone2Name=spine_01
  swing1LimitDegrees=45
  swing1Motion=Free
  swing2LimitDegrees=45
  swing2Motion=Free
  twistLimitDegrees=45
  twistMotion=Free
 }
 {
  bone1Name=dyn_skirt_b_1_l
  bone2Name=dyn_skirt_b_l
  swing1LimitDegrees=45
  swing1Motion=Free
  swing2LimitDegrees=45
  swing2Motion=Free
  twistLimitDegrees=45
  twistMotion=Free
 }
 {
  bone1Name=dyn_skirt_b_2_l
  bone2Name=dyn_skirt_b_1_l
  swing1LimitDegrees=45
  swing1Motion=Free
  swing2LimitDegrees=45
  swing2Motion=Free
  twistLimitDegrees=45
  twistMotion=Free
 }
 {
  bone1Name=dyn_skirt_b_r
  bone2Name=spine_01
  swing1LimitDegrees=45
  swing1Motion=Free
  swing2LimitDegrees=45
  swing2Motion=Free
  twistLimitDegrees=45
  twistMotion=Free
 }
 {
  bone1Name=dyn_skirt_b_1_r
  bone2Name=dyn_skirt_b_r
  swing1LimitDegrees=45
  swing1Motion=Free
  swing2LimitDegrees=45
  swing2Motion=Free
  twistLimitDegrees=45
  twistMotion=Free
 }
 {
  bone1Name=dyn_skirt_b_2_r
  bone2Name=dyn_skirt_b_1_r
  swing1LimitDegrees=45
  swing1Motion=Free
  swing2LimitDegrees=45
  swing2Motion=Free
  twistLimitDegrees=45
  twistMotion=Free
 }
 {
  bone1Name=dyn_belt_b
  bone2Name=spine_01
  swing1LimitDegrees=45
  swing1Motion=Free
  swing2LimitDegrees=45
  swing2Motion=Free
  twistLimitDegrees=45
  twistMotion=Free
 }
 {
  bone1Name=dyn_belt_1_b
  bone2Name=dyn_belt_b
  swing1LimitDegrees=45
  swing1Motion=Free
  swing2LimitDegrees=45
  swing2Motion=Free
  twistLimitDegrees=45
  twistMotion=Free
 }
]
