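# Physics asset for the Cape face accessory of the MuffinLadle GasChin male medium player body, giving its left, right and back strap bones physics shapes.
# Capsules cover the upper spine, clavicles, neck and head alongside the straps.
# Side straps form chains from the head with tight twist limits below their free-swinging roots.
# The limited back strap branches in two, with cross-constraints between sibling bones that lock selected axes to couple them.
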PhysicsAsset /BRCosmetics/Characters/Player/Male/Medium/Bodies/M_MED_MuffinLadle_GasChin/Meshes/Parts/M_MED_MuffinLadle_GasChin_FaceAcc_Cape_Physics
bodies {
 clavicle_l [{
   shapeType=Capsule
   length=13.5865
   radius=8.9753
   shapeName=clavicle_l_capsule
   center=<9.2152 0.6357 -1.8649>
   rotation=<rot 76.0907 -0.001 -0.001>
  }]
 clavicle_r [{
   shapeType=Capsule
   length=13.056
   radius=9.5058
   shapeName=clavicle_r_capsule
   center=<-8.9984 0.6371 2.7402>
   rotation=<rot 76.0907 -0.001 179.999>
  }]
 deform_strap_b [{
   shapeType=Capsule
   length=3
   radius=2
   shapeName=deform_strap_b_capsule
   center=<2.6998 -0.396 0.8573>
   rotation=<rot 72.0362 122.9458 124.2654>
  }]
 deform_strap_l [{shapeType=Capsule, length=5, radius=2, shapeName=deform_strap_l_capsule, center=<2.851 -0.9657 0>, rotation=<rot 90 0 0>}]
 deform_strap_r [{shapeType=Capsule, length=5, radius=2, shapeName=deform_strap_r_capsule, center=<2.851 -0.9657 0>, rotation=<rot 90 0 0>}]
 dyn_strap_1_b [{shapeType=Capsule, length=5, radius=2, shapeName=dyn_strap_1_b_capsule, center=<2.851 -0.3512 0.4109>, rotation=<rot 90 0 0>}]
 dyn_strap_1_l [{
   shapeType=Capsule
   length=5.5
   radius=1.5
   shapeName=dyn_strap_1_l_capsule
   center=<2.851 -0.067 0.3271>
   rotation=<rot 70 -90.0008 -90.0008>
  }]
 dyn_strap_1_r [{
   shapeType=Capsule
   length=5.5
   radius=1.5
   shapeName=dyn_strap_1_r_capsule
   center=<2.851 -0.1562 -0.2946>
   rotation=<rot 70 89.9991 89.9992>
  }]
 dyn_strap_2_b [{
   shapeType=Capsule
   length=1
   radius=2.5
   shapeName=dyn_strap_2_b_capsule
   center=<2.851 -1.2466 0.0495>
   rotation=<rot -0.0001 0.0004 9.9999>
  }]
 dyn_strap_2_l [{shapeType=Capsule, length=5.5, radius=1.5, shapeName=dyn_strap_2_l_capsule, center=<2.851 -0.316 0.5302>, rotation=<rot 90 0 0>}]
 dyn_strap_2_r [{shapeType=Capsule, length=5, radius=2, shapeName=dyn_strap_2_r_capsule, center=<2.851 -0.316 0.5302>, rotation=<rot 90 0 0>}]
 dyn_strap_3_b [{shapeType=Capsule, length=5, radius=2.5, shapeName=dyn_strap_3_b_capsule, center=<3.6045 -1.1382 -1.7757>, rotation=<rot 90 0 0>}]
 dyn_strap_3_l [{shapeType=Capsule, length=6.524, radius=2, shapeName=dyn_strap_3_l_capsule, center=<2.851 -0.3568 0>, rotation=<rot 90 0 0>}]
 dyn_strap_3_r [{shapeType=Capsule, length=5.4661, radius=3, shapeName=dyn_strap_3_r_capsule, center=<2.851 -0.3568 0>, rotation=<rot 90 0 0>}]
 dyn_strap_4_b [{shapeType=Capsule, length=1.5, radius=2.5, shapeName=dyn_strap_4_b_capsule, center=<4.223 -1.7686 0.2151>, rotation=<rot -0.0003 0.0006 15>}]
 dyn_strap_5_b [{
   shapeType=Capsule
   length=2.5
   radius=2.5
   shapeName=dyn_strap_5_b_capsule
   center=<2.851 -0.47 0.354>
   rotation=<rot -0.0003 0.0005 30.0001>
  }]
 dyn_strap_b [
  {shapeType=Capsule, length=3.5, radius=2, shapeName=dyn_strap_b_capsule, center=<2.851 -0.4412 -0.8558>, rotation=<rot 90 0 0>}
  {
   shapeType=Capsule
   length=3.5
   radius=2
   shapeName=dyn_strap_b_capsule
   center=<2.851 -0.4412 0.8679>
   rotation=<rot 80.4237 -179.9999 179.9999>
  }
 ]
 dyn_strap_l [{shapeType=Capsule, length=5.5, radius=1.5, shapeName=dyn_strap_l_capsule, center=<2.851 -0.9657 0>, rotation=<rot 90 0 0>}]
 dyn_strap_r [{shapeType=Capsule, length=5.5, radius=1.5, shapeName=dyn_strap_r_capsule, center=<2.851 -0.9657 0>, rotation=<rot 90 0 0>}]
 head [{shapeType=Capsule, length=13.8525, radius=6.1417, center=<-1.5836 3.9872 0>, rotation=<rot 90 180 -151.2132>}]
 neck_01 [{shapeType=Capsule, length=7.6415, radius=6.7819, shapeName=neck_01_capsule, center=<0.0838 3.9899 0>, rotation=<rot 90 -9.4623 40.5382>}]
 neck_02 [{shapeType=Capsule, length=1.626, radius=7.2901, shapeName=neck_02_capsule, center=<0 1.4833 0>, rotation=<rot 0 0 90>}]
 spine_04 [{shapeType=Capsule, length=10.1775, radius=13.7881, shapeName=spine_04_capsule, center=<6.0846 -1.6827 0>}]
 spine_05 [
  {shapeType=Capsule, length=14.5017, radius=10.1767, center=<2.1941 -0.5891 0>, rotation=<rot 0 175.5189 180>}
  {shapeType=Capsule, length=8.8065, radius=6.9285, shapeName=spine_05_capsule, center=<6.6333 0.1951 0>, rotation=<rot 90 0 0>}
  {
   shapeType=Capsule
   length=4.887
   radius=9.5998
   shapeName=spine_05_capsule
   center=<0.6988 -4.6562 13.1328>
   rotation=<rot 90 -176.4237 149.4068>
  }
 ]
}
constraints [
 {
  bone1Name=clavicle_l
  bone2Name=spine_05
  swing1LimitDegrees=45
  swing1Motion=Free
  swing2LimitDegrees=45
  swing2Motion=Free
  twistLimitDegrees=45
  twistMotion=Free
 }
 {
  bone1Name=clavicle_r
  bone2Name=spine_05
  swing1LimitDegrees=45
  swing1Motion=Free
  swing2LimitDegrees=45
  swing2Motion=Free
  twistLimitDegrees=45
  twistMotion=Free
 }
 {
  bone1Name=neck_01
  bone2Name=spine_05
  swing1LimitDegrees=45
  swing1Motion=Free
  swing2LimitDegrees=45
  swing2Motion=Free
  twistLimitDegrees=45
  twistMotion=Free
 }
 {bone1Name=deform_strap_b, bone2Name=head, swing1LimitDegrees=30, swing2LimitDegrees=20, twistLimitDegrees=20}
 {bone1Name=dyn_strap_1_b, bone2Name=dyn_strap_b, swing1LimitDegrees=45, swing2LimitDegrees=15, twistLimitDegrees=4}
 {bone1Name=dyn_strap_2_b, bone2Name=dyn_strap_1_b, swing1LimitDegrees=25, swing2LimitDegrees=10, twistLimitDegrees=4}
 {bone1Name=dyn_strap_3_b, bone2Name=dyn_strap_b, swing1LimitDegrees=45, swing2LimitDegrees=25, twistLimitDegrees=4}
 {bone1Name=dyn_strap_4_b, bone2Name=dyn_strap_3_b, swing1LimitDegrees=35, swing2LimitDegrees=25, twistLimitDegrees=4}
 {bone1Name=dyn_strap_5_b, bone2Name=dyn_strap_3_b, swing1LimitDegrees=35, swing2LimitDegrees=10, twistLimitDegrees=4}
 {
  bone1Name=deform_strap_l
  bone2Name=head
  swing1LimitDegrees=45
  swing1Motion=Free
  swing2LimitDegrees=45
  swing2Motion=Free
  twistLimitDegrees=45
  twistMotion=Free
 }
 {bone1Name=dyn_strap_l, bone2Name=deform_strap_l, swing1LimitDegrees=20, swing2LimitDegrees=20, twistLimitDegrees=4}
 {bone1Name=dyn_strap_1_l, bone2Name=dyn_strap_l, swing1LimitDegrees=25, swing2LimitDegrees=25, twistLimitDegrees=4}
 {bone1Name=dyn_strap_2_l, bone2Name=dyn_strap_1_l, swing1LimitDegrees=25, swing2LimitDegrees=25, twistLimitDegrees=4}
 {bone1Name=dyn_strap_3_l, bone2Name=dyn_strap_2_l, swing1LimitDegrees=25, swing2LimitDegrees=25, twistLimitDegrees=4}
 {
  bone1Name=deform_strap_r
  bone2Name=head
  swing1LimitDegrees=45
  swing1Motion=Free
  swing2LimitDegrees=45
  swing2Motion=Free
  twistLimitDegrees=45
  twistMotion=Free
 }
 {bone1Name=dyn_strap_r, bone2Name=deform_strap_r, swing1LimitDegrees=20, swing2LimitDegrees=20, twistLimitDegrees=4}
 {bone1Name=dyn_strap_1_r, bone2Name=dyn_strap_r, swing1LimitDegrees=25, swing2LimitDegrees=25, twistLimitDegrees=4}
 {bone1Name=dyn_strap_2_r, bone2Name=dyn_strap_1_r, swing1LimitDegrees=25, swing2LimitDegrees=25, twistLimitDegrees=4}
 {bone1Name=dyn_strap_3_r, bone2Name=dyn_strap_2_r, swing1LimitDegrees=25, swing2LimitDegrees=25, twistLimitDegrees=4}
 {bone1Name=dyn_strap_5_b, bone2Name=dyn_strap_4_b, swing1LimitDegrees=15, swing2LimitDegrees=45, twistLimitDegrees=45, twistMotion=Locked}
 {
  bone1Name=spine_05
  bone2Name=spine_04
  swing1LimitDegrees=45
  swing1Motion=Free
  swing2LimitDegrees=45
  swing2Motion=Free
  twistLimitDegrees=45
  twistMotion=Free
 }
 {
  bone1Name=dyn_strap_3_b
  bone2Name=dyn_strap_1_b
  swing1LimitDegrees=45
  swing1Motion=Locked
  swing2LimitDegrees=45
  twistLimitDegrees=45
  twistMotion=Free
 }
 {
  bone1Name=neck_02
  bone2Name=neck_01
  swing1LimitDegrees=45
  swing1Motion=Free
  swing2LimitDegrees=45
  swing2Motion=Free
  twistLimitDegrees=45
  twistMotion=Free
 }
 {bone1Name=head, bone2Name=neck_02, swing1LimitDegrees=45, swing2LimitDegrees=45, twistLimitDegrees=45}
 {bone1Name=dyn_strap_b, bone2Name=deform_strap_b, swing1LimitDegrees=70, swing2LimitDegrees=35, twistLimitDegrees=1}
]
