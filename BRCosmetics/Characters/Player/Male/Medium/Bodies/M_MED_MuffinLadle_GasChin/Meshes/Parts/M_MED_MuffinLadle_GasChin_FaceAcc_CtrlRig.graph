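# Control Rig for the face accessory of the MuffinLadle GasChin character, correcting the back strap bone deform_strap_b.
# It aims the bone at a control under spine_05, clamps its position against a plane in neck_01 space with mirrored Y, then applies a small fixed rotation offset.
# The playing_FE_emote flag selects the clamp minimum, but both options are identical.
# Side straps and dyn bones appear in the hierarchy but are not directly driven.
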ControlRigBlueprint /BRCosmetics/Characters/Player/Male/Medium/Bodies/M_MED_MuffinLadle_GasChin/Meshes/Parts/M_MED_MuffinLadle_GasChin_FaceAcc_CtrlRig
graphs [
 RigVMFunctionLibrary {links [], nodes []}
 RigVMModel {nodes [
   RigUnit_BeginExecution = &RigUnit_BeginExecution()
   &RigUnit_AimConstraintLocalSpaceOffset(
    AdvancedSettings=(DebugSettings=(bEnabled=true,Scale=10,WorldOffset=(Rotation=(X=0,Y=0,Z=0,W=1),Translation=(X=0,Y=0,Z=0),Scale3D=(X=1,Y=1,Z=1))),RotationOrderForFilter=XZY)
    Child=(Type=Bone,Name="deform_strap_b")
    Filter=(bX=true,bY=true,bZ=True)
    Parents=((Item=(Type=Control,Name="spine_05_ctrl"),Weight=1))
    UpAxis=(X=0,Y=-1,Z=0)
    Weight=1
    WorldUp=(Target=(X=0,Y=0,Z=1),Kind=Location,Space=(Type=Bone,Name="deform_strap_b"))
    ExecutePin=@RigUnit_BeginExecution.ExecutePin
   )
   n1 = &RigUnit_GetTransform(Item=(Type=Bone,Name="neck_01"), bInitial=False)
   RigVMFunction_MathBoolToInteger = &RigVMFunction_MathBoolToInteger(Value=RigVMVariableNode() {variable=playing_FE_emote}.Value)
   ClampSpatially = /Script/RigVM.RigVMFunction_MathTransformClampSpatially(
    Axis=Y
    DebugColor=(R=1,G=0,B=0,A=1)
    DebugThickness=1
    Maximum=15
    Space=(Rotation=(X=0,Y=0,Z=0,W=1),Translation=(X=0,Y=0,Z=0),Scale3D=(X=1,Y=-1,Z=1))
    Type=Plane
    bDrawDebug=true
    Value=&RigUnit_GetTransform(Item=(Type=Bone,Name="deform_strap_b"), bInitial=False).Transform
    Space.Rotation=@n1.Transform.Rotation
    Space.Translation=@n1.Transform.Translation
    Minimum=RigVMDispatchNode(Values=(4,4), Index=@RigVMFunction_MathBoolToInteger.Result) {template="DISPATCH_RigVMDispatch_SelectInt32(in Index,in Values,out Result)"}.Result
   ) {
    is=struct RigVMUnitNode
    template="ClampSpatially::Execute(in Value,in Axis,in Type,in Minimum,in Maximum,in Space,in bDrawDebug,in DebugColor,in DebugThickness,out Result)"
   }
   > Set Transform_8 = /Script/ControlRig.RigUnit_SetTranslation(
    Item=(Type=Bone,Name="deform_strap_b")
    Space=GlobalSpace
    Weight=1
    bInitial=false
    bPropagateToChildren=true
    Value=@ClampSpatially.Result.Translation
   ) {is=struct RigVMUnitNode, template="Set Transform::Execute(in Item,in Space,in bInitial,in Value,in Weight,in bPropagateToChildren)"}
   > &RigUnit_ModifyTransforms(
    ItemToModify=((Item=(Type=Bone,Name="deform_strap_b"),Transform=(Rotation=(X=0,Y=0,Z=0.0872,W=0.9962),Translation=(X=0,Y=0,Z=0),Scale3D=(X=1,Y=1,Z=1))))
    Weight=1
    WeightMinimum=0
   )
  ]}
]
hierarchy [
 root {loc=<0 0 0>, rot=<0 0 0 1>}
 pelvis {parent=root, loc=<0 -0.7071 93.1191>, rot=<0.0127 0.707 -0.0127 -0.707>}
 spine_01 {parent=pelvis, loc=<6.7494 -0.6388 0>, rot=<0 0 0.034 -0.9994>}
 spine_02 {parent=spine_01, loc=<7.252 -0.035 0>, rot=<0 0 0.0238 -0.9997>}
 spine_03 {parent=spine_02, loc=<6.0106 0 0>, rot=<0 0 0.0149 -0.9999>}
 spine_04 {parent=spine_03, loc=<8.1023 0 0>, rot=<0 0 -0.0132 -0.9999>}
 spine_05 {parent=spine_04, loc=<10.0142 0 0>, rot=<0 0 0 1>}
 clavicle_l {parent=spine_05, loc=<4.3579 -0.1336 -2.3223>, rot=<0.0476 -0.7289 -0.01 -0.6829>}
 clavicle_r {parent=spine_05, loc=<4.3579 -0.1335 2.3223>, rot=<0.7289 0.0476 -0.6829 0.01>}
 neck_01 {parent=spine_05, loc=<11.3133 0.1867 0>, rot=<0 0 -0.2017 -0.9794>}
 neck_02 {parent=neck_01, loc=<4.5627 0 0>, rot=<0 0 0.0504 -0.9987>}
 head {parent=neck_02, loc=<5.3914 0 0>, rot=<0 0 0.115 -0.9934>}
 deform_strap_b {parent=head, loc=<1.251 -4.7692 2.9598>, rot=<0.2092 -0.9599 0.1861 0.0127>}
 dyn_strap_b {parent=deform_strap_b, loc=<0 0 0>, rot=<0 0 0 1>}
 dyn_strap_1_b {parent=dyn_strap_b, loc=<4.7443 0.0077 1.2477>, rot=<0.0188 0.0851 0.0104 -0.9961>}
 dyn_strap_2_b {parent=dyn_strap_1_b, loc=<3.2633 0 0>, rot=<0.0209 0.0195 -0.0727 -0.9969>}
 dyn_strap_3_b {parent=dyn_strap_b, loc=<4.7857 -0.0443 -0.396>, rot=<-0.0338 0.0101 -0.0034 -0.9994>}
 dyn_strap_4_b {parent=dyn_strap_3_b, loc=<5.5502 0.0007 0.5192>, rot=<-0.0047 0.0067 -0.0082 -0.9999>}
 dyn_strap_5_b {parent=dyn_strap_3_b, loc=<5.5988 -0.2737 -0.8203>, rot=<-0.0938 0.0022 0.0308 -0.9951>}
 deform_strap_r {parent=head, loc=<2.2707 3.5588 6.0928>, rot=<0.0092 -0.6182 0.7826 0.0728>}
 dyn_strap_r {parent=deform_strap_r, loc=<0 0 0>, rot=<0 0 0 1>}
 dyn_strap_1_r {parent=dyn_strap_r, loc=<5.8218 0 0>, rot=<-0.0776 -0.0595 0.0185 -0.995>}
 dyn_strap_2_r {parent=dyn_strap_1_r, loc=<6.615 0 0>, rot=<-0.0826 -0.0803 0.065 -0.9912>}
 dyn_strap_3_r {parent=dyn_strap_2_r, loc=<7.2589 -0.0002 -0.0001>, rot=<-0.2744 -0.0163 -0.0254 -0.9611>}
 deform_strap_l {parent=head, loc=<2.3624 3.5495 -6.1035>, rot=<-0.0092 0.6182 0.7826 0.0728>}
 dyn_strap_l {parent=deform_strap_l, loc=<0 0 0>, rot=<0 0 0 1>}
 dyn_strap_1_l {parent=dyn_strap_l, loc=<5.8217 0 0>, rot=<0.0768 0.0491 0.028 -0.9954>}
 dyn_strap_2_l {parent=dyn_strap_1_l, loc=<6.6148 0 0>, rot=<0.0848 0.078 0.0372 -0.9926>}
 dyn_strap_3_l {parent=dyn_strap_2_l, loc=<8.8147 0.0901 0.0271>, rot=<0.2744 0.0163 -0.0254 -0.9611>}
 spine_05_locator {parent=spine_05, loc=<0 0 0>, rot=<0 0 0 1>}
 spine_05_ctrl {controlType=EULER_TRANSFORM, parent=spine_05, type=Control, loc=<0 -7.9013 0>, rot=<0 0 0 1>}
]
variables [playing_FE_emote {type=bool, default=False}]
