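# Post-process style animation Blueprint for the face accessory part of the MuffinLadle GasChin character, adding secondary motion to the copied parent pose.
# The copied pose feeds a Control Rig informed by a lobby-idle curve flag, then two neck-based RigidBody nodes.
# A customization-load check for the matching backpack sets isWearingCape, which blends between the standard and cape physics assets.
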AnimBlueprint /BRCosmetics/Characters/Player/Male/Medium/Bodies/M_MED_MuffinLadle_GasChin/Meshes/Parts/M_MED_MuffinLadle_GasChin_FaceAcc_AnimBP
target_skeleton=&/M_MED_MuffinLadle_GasChin_FaceAcc_Skeleton
graphs [
 AnimGraph {nodes [
   n1 = Control Rig!(
    Source=Copy Pose From Mesh() {bUseAttachedParent=true}
    Alpha
    playing_FE_emote:bool=Get FE Custom Lobby Idle Is Playing() -> (FE Custom Lobby Idle Is Playing:bool)
   ) -> (Pose) {
    alphaBoolBlend {blendOption=Linear}
    controlRigAssetReference {blueprintRigClass=&/M_MED_MuffinLadle_GasChin_FaceAcc_CtrlRig.M_MED_MuffinLadle_GasChin_FaceAcc_CtrlRig_C}
   }
   n2 = Get isWearingCape() -> (isWearingCape:bool)
   n3 = RigidBody!(
    ComponentPose=Local To Component(LocalPose=@n1)
    bAlphaBoolEnabled:bool=NOT Boolean(A=@n2)
    OverrideWorldGravity:Vector="0,0,0"
    ExternalForce
   ) -> (Pose) {
    alphaInputType=Bool
    bClampLinearTranslationLimitToRefPose=true
    bOverrideWorldGravity=true
    overridePhysicsAsset=&/M_MED_MuffinLadle_GasChin_FaceAcc_Physics
    simulationSpace=BaseBoneSpace
    alphaBoolBlend {blendInTime=1.7, blendOption=Linear, blendOutTime=1.7}
    alphaScaleBiasClamp {bInterpResult=true, interpSpeedDecreasing=5, interpSpeedIncreasing=5}
    baseBoneRef {boneName=neck_02}
    componentAppliedLinearAccClamp=<3000 3000 3000>
    overrideWorldGravity {z=-980}
    simSpaceSettings {worldAlpha=0.4}
   }
   n4 = RigidBody(ComponentPose=@n3, bAlphaBoolEnabled=@n2, OverrideWorldGravity="0,0,0") {
    alphaInputType=Bool
    bClampLinearTranslationLimitToRefPose=true
    bOverrideWorldGravity=true
    overridePhysicsAsset=&/M_MED_MuffinLadle_GasChin_FaceAcc_Cape_Physics
    simulationSpace=BaseBoneSpace
    alphaBoolBlend {blendInTime=1.7, blendOption=Linear, blendOutTime=1.7}
    alphaScaleBiasClamp {bInterpResult=true, interpSpeedDecreasing=5, interpSpeedIncreasing=5}
    baseBoneRef {boneName=neck_02}
    componentAppliedLinearAccClamp=<3000 3000 3000>
    overrideWorldGravity {z=-980}
    simSpaceSettings {worldAlpha=0.4}
   }
   Output Pose(Result=Component To Local(ComponentPose=@n4)) {layerGroup=None}
  ]}
 EventGraph {nodes [
   Event BlueprintUpdateAnimation()
   > n1 = Gravity Override State Function DEPRECATED(CharacterPartAnimInstance=Self-Reference(), GameTimeInSeconds=GetGameTimeInSeconds())
   > Set Gravity Override(Gravity Override=@n1)
   > Branch(Condition=AND Boolean(A=Get bIsInFrontEnd(), B=Get isWearingCape()))
   n2 = GetOwnerFortPlayerAnimInstance!(self:Custom Character Part Anim Instance Object Reference) -> (ReturnValue:Fort Player Anim Instance Object Reference)
   > Set FE Custom Lobby Idle Is Playing(FE Custom Lobby Idle Is Playing:bool="Equal (Float)"(A=GetCurveValue(self=@n2, CurveName=Frontend_Idle_Bypass), B=1)) -> (Output_Get:bool)
   n3 = Event OnCharacterCustomizationLoaded()
   > Set isWearingCape(isWearingCape="Equal (Object)"(
     A=Get Array Item with EFort Custom Part Type(Array:Array of Custom Character Part Object References=@n3.CustomParts, PartType=Backpack) -> (Output:Custom Character Part Object Reference)
     B=/BRCosmetics/Characters/CharacterParts/Backpacks/CP_Backpack_MuffinLadle_GasChin
    ))
   CheckForBackbling()
  ]}
]
variables [
 FE Custom Lobby Idle Is Playing {type=boolean}
 Gravity Override {type=Vector}
 isWearingCape {type=boolean}
]
